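# Animation Blueprint for the Caretaker monster NPC, driving its locomotion, aiming and limb IK.
# A state machine covers scrunched spawn pose, idle, walk and foot-phase-matched stops, then feeds montage slots, additive breathing and an aim offset into a cached pose.
# Predictive stride and slope warping, leg IK and hand attack IK refine that pose.
# An in-water input event bypasses the IK chain, with inertialization smoothing the switch.
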AnimBlueprint /Caretaker/Animations/Caretaker_AnimBP
target_skeleton=/Caretaker/Meshes/M_LRG_Caretaker_Skeleton
graphs [
 AnimGraph {nodes [
   n1 = Use cached pose 'PreIKPose'()
   n2 = Copy Bone - Source Bone: R_BipedLeg_A_Ankle_Jnt - Target Bone: IK_Foot_R!(
    ComponentPose:Component Space Pose Link Structure=Copy Bone - Source Bone: L_BipedLeg_A_Ankle_Jnt - Target Bone: IK_Foot_L!(
     ComponentPose:Component Space Pose Link Structure=Local To Component(LocalPose=@n1)
     Alpha:float=1
     bCopyTranslation:bool=true
     bCopyRotation:bool=true
     bCopyScale:bool=False
    ) -> (Pose:Component Space Pose Link Structure) {
     is=title AnimGraphNode_CopyBone
     alphaBoolBlend {blendOption=Linear}
     sourceBone {boneName=L_BipedLeg_A_Ankle_Jnt}
     targetBone {boneName=IK_Foot_L}
    }
    Alpha:float=1
    bCopyTranslation:bool=true
    bCopyRotation:bool=true
    bCopyScale:bool=False
   ) -> (Pose:Component Space Pose Link Structure) {
    is=title AnimGraphNode_CopyBone
    alphaBoolBlend {blendOption=Linear}
    sourceBone {boneName=R_BipedLeg_A_Ankle_Jnt}
    targetBone {boneName=IK_Foot_R}
   }
   n3 = Predictive Stride Warping!(
    ComponentPose:Component Space Pose Link Structure=Copy Bone - Source Bone: R_BipedArm_A_Wrist_Jnt - Target Bone: IK_Hand_R!(
     ComponentPose:Component Space Pose Link Structure=Copy Bone - Source Bone: L_BipedArm_A_Wrist_Jnt - Target Bone: IK_Hand_L!(
      ComponentPose:Component Space Pose Link Structure=@n2
      Alpha:float=1
      bCopyTranslation:bool=true
      bCopyRotation:bool=true
      bCopyScale:bool=False
     ) -> (Pose:Component Space Pose Link Structure) {
      is=title AnimGraphNode_CopyBone
      alphaBoolBlend {blendOption=Linear}
      sourceBone {boneName=L_BipedArm_A_Wrist_Jnt}
      targetBone {boneName=IK_Hand_L}
     }
     Alpha:float=1
     bCopyTranslation:bool=true
     bCopyRotation:bool=true
     bCopyScale:bool=False
    ) -> (Pose:Component Space Pose Link Structure) {
     is=title AnimGraphNode_CopyBone
     alphaBoolBlend {blendOption=Linear}
     sourceBone {boneName=R_BipedArm_A_Wrist_Jnt}
     targetBone {boneName=IK_Hand_R}
    }
    AlphaCurveName:Name=IsWalking
    SpeedScaling:float=1
   ) -> (Pose:Component Space Pose Link Structure) {
    is=title FortAnimGraphNode_PredictiveStrideWarping
    alphaCurveName=IsWalking
    alphaInputType=Curve
    predictiveStrideWarpingAxisMode=IKFootRootLocalY
    alphaBoolBlend {blendOption=Linear}
    feetDefinitions [
     {
      numBonesInLimb=2
      fKFootBone {boneName=L_BipedLeg_A_Ankle_Jnt}
      iKFootBone {boneName=IK_Foot_L}
      iKPredictionFootBone {boneName=IK_Foot_Prediction_L}
     }
     {
      numBonesInLimb=2
      fKFootBone {boneName=R_BipedLeg_A_Ankle_Jnt}
      iKFootBone {boneName=IK_Foot_R}
      iKPredictionFootBone {boneName=IK_Foot_Prediciton_R}
     }
    ]
    iKFootRootBone {boneName=IK_Foot_Root}
    pelvisBone {boneName=C_BipedSpine_Main_Pelvis_Jnt}
   }
   n4 = Predictive Biped Slope Warping!(ComponentPose:Component Space Pose Link Structure=@n3, Alpha:float=1) -> (Pose:Component Space Pose Link Structure) {
    is=title FortAnimGraphNode_PredictiveSlopeWarping
    bCorrectForMeshOffset=false
    bKeepMeshInsideOfCapsule=false
    bUseWorldGravityDirectionForOffset=false
    maxStepHeight=250
    alphaBoolBlend {blendOption=Linear}
    feetDefinitions [
     {numBonesInLimb=2, fKFootBone {boneName=L_BipedLeg_A_Ankle_Jnt}, iKFootBone {boneName=IK_Foot_L}}
     {numBonesInLimb=2, fKFootBone {boneName=R_BipedLeg_A_Ankle_Jnt}, iKFootBone {boneName=IK_Foot_R}}
    ]
    iKFootRootBone {boneName=IK_Foot_Root}
    lFootPrediction {boneName=IK_Foot_Prediction_L}
    pelvisBone {boneName=C_BipedSpine_Main_Pelvis_Jnt}
    rFootPrediction {boneName=IK_Foot_Prediciton_R}
   }
   n5 = Leg IK!(ComponentPose:Component Space Pose Link Structure=@n4, Alpha:float=1) -> (Pose:Component Space Pose Link Structure) {
    is=title AnimGraphNode_LegIK
    alphaBoolBlend {blendOption=Linear}
    legsDefinition [
     {
      bEnableKneeTwistCorrection=true
      bEnableRotationLimit=false
      footBoneForwardAxis=Y
      hingeRotationAxis=None
      minRotationAngle=15
      numBonesInLimb=2
      twistOffsetCurveName=None
      fKFootBone {boneName=L_BipedLeg_A_Ankle_Jnt}
      iKFootBone {boneName=IK_Foot_L}
     }
     {
      bEnableKneeTwistCorrection=true
      bEnableRotationLimit=false
      footBoneForwardAxis=Y
      hingeRotationAxis=None
      minRotationAngle=15
      numBonesInLimb=2
      twistOffsetCurveName=None
      fKFootBone {boneName=R_BipedLeg_A_Ankle_Jnt}
      iKFootBone {boneName=IK_Foot_R}
     }
    ]
   }
   n6 = Attack IK!(ComponentPose:Component Space Pose Link Structure=@n5, Alpha:float=1) -> (Pose:Component Space Pose Link Structure) {
    is=title FortAnimGraphNode_AttackIK
    maxCastLength=300
    alphaBoolBlend {blendOption=Linear}
    handsDefinitions [
     {handOnGroundCurveName=LeftHandOnGround, targetBoneOffset=51, attackTargetBone {boneName=None}, iKHandBone {boneName=IK_Hand_L}}
     {handOnGroundCurveName=RightHandOnGround, targetBoneOffset=28, attackTargetBone {boneName=None}, iKHandBone {boneName=IK_Hand_R}}
    ]
    iKHandRootBone {boneName=IK_Hand_Root}
   }
   n7 = Leg IK(ComponentPose=@n6, Alpha=1) {
    alphaBoolBlend {blendOption=Linear}
    legsDefinition [
     {
      bEnableKneeTwistCorrection=true
      bEnableRotationLimit=false
      footBoneForwardAxis=Y
      hingeRotationAxis=None
      minRotationAngle=15
      numBonesInLimb=2
      twistOffsetCurveName=None
      fKFootBone {boneName=L_BipedArm_A_Wrist_Jnt}
      iKFootBone {boneName=IK_Hand_L}
     }
     {
      bEnableKneeTwistCorrection=true
      bEnableRotationLimit=false
      footBoneForwardAxis=Y
      hingeRotationAxis=None
      minRotationAngle=15
      numBonesInLimb=2
      twistOffsetCurveName=None
      fKFootBone {boneName=R_BipedArm_A_Wrist_Jnt}
      iKFootBone {boneName=IK_Hand_R}
     }
    ]
   }
   n8 = Component To Local(ComponentPose=@n7)
   n9 = Use cached pose 'PreIKPose'()
   n10 = Get Is Pawn In Water() -> (Is Pawn In Water:bool)
   Output Pose(Result=Inertialization!(Source:Pose Link Structure=Blend Poses by bool(BlendPose_0=@n9, BlendPose_1=@n8, bActiveValue=@n10) {blendType=Linear}) -> (Pose:Pose Link Structure) {is=title AnimGraphNode_Inertialization}) {layerGroup=None}
   n11 = "Sequence Player (additive) 'Caretaker_Idle_Breathe'"() {sequence=&/Caretaker_Idle_Breathe}
   AnimGraphNode_StateMachine_0 = Locomotion() {is=title AnimGraphNode_StateMachine}
   n12 = Slot 'FullBodyAdditiveSlot'(Source=Slot 'DefaultSlot'(Source=@AnimGraphNode_StateMachine_0)) {bAlwaysUpdateSourcePose=true, slotName=FullBodyAdditiveSlot}
   n13 = Get BreathingCurve() -> (BreathingCurve:float)
   Save cached pose 'PreIKPose'(Pose=AimOffset Player 'Caretaker_AO'!(
     X:float=Get AimYaw() -> (AimYaw:float)
     Y:float=Get AimPitch()
     BasePose:Pose Link Structure=Apply Additive(Base=@n12, Additive=@n11, Alpha=@n13) {alphaBoolBlend {blendOption=Linear}}
     Alpha:float=Get AimOffsetCurve() -> (AimOffsetCurve:float)
    ) -> (Pose:Pose Link Structure) {alphaCurveName=AimOffsetCurve, blendSpace=/Caretaker/Animations/AO/Caretaker_AO, alphaBoolBlend {blendOption=Linear}}) {cacheName=PreIKPose}
  ]}
 Locomotion {
  entry=HackState
  path=AnimGraph.AnimGraphNode_StateMachine_0.Locomotion
  ids=id
  nodes [
   AnimStateConduitNode_0 = Conduit()
   AnimStateNode_0 = HackState() {is=title AnimStateNode}
   AnimStateNode_1 = Idle()
   AnimStateNode_2 = Walk() {
    is=title AnimStateNode
    stateEntered {guid=#1}
    stateFullyBlended {guid=#2}
    stateLeft {guid=#3}
   }
   AnimStateNode_3 = WalkToIdle() {is=title AnimStateNode}
   AnimStateNode_5 = InitialSpawn() {is=title AnimStateNode, stateEntered {guid=#4, notifyName=SpawnEnter}}
   AnimStateTransitionNode_0 = AnimStateTransitionNode()
   AnimStateTransitionNode_1 = AnimStateTransitionNode() {crossfadeDuration=0.5, logicType=TLT_Inertialization, priorityOrder=100}
   AnimStateTransitionNode_10 = AnimStateTransitionNode() {crossfadeDuration=0, logicType=TLT_Inertialization}
   AnimStateTransitionNode_2 = AnimStateTransitionNode() {
    crossfadeDuration=0.5
    logicType=TLT_Inertialization
    priorityOrder=2
    transitionEnd {guid=#5}
    transitionInterrupt {guid=#6}
    transitionStart {guid=#7}
   }
   AnimStateTransitionNode_3 = AnimStateTransitionNode() {crossfadeDuration=0.35, logicType=TLT_Inertialization}
   AnimStateTransitionNode_4 = AnimStateTransitionNode() {bAutomaticRuleBasedOnSequencePlayerInState=true, crossfadeDuration=0.35, logicType=TLT_Inertialization}
   AnimStateTransitionNode_9 = AnimStateTransitionNode()
  ]
  links [
   {
    dst=Idle
    rule=AnimGraph.AnimGraphNode_StateMachine_0.Locomotion.AnimStateTransitionNode_1.Transition
    src=Conduit
    via=AnimStateTransitionNode_1
   }
   {
    dst=InitialSpawn
    rule=AnimGraph.AnimGraphNode_StateMachine_0.Locomotion.AnimStateTransitionNode_10.Transition
    src=Conduit
    via=AnimStateTransitionNode_10
   }
   {
    dst=Walk
    rule=AnimGraph.AnimGraphNode_StateMachine_0.Locomotion.AnimStateTransitionNode_2.Transition
    src=Conduit
    via=AnimStateTransitionNode_2
   }
   {
    dst=Conduit
    rule=AnimGraph.AnimGraphNode_StateMachine_0.Locomotion.AnimStateTransitionNode_0.Transition
    src=HackState
    via=AnimStateTransitionNode_0
   }
   {
    dst=Walk
    rule=AnimGraph.AnimGraphNode_StateMachine_0.Locomotion.AnimStateTransitionNode_9.Transition
    src=Idle
    via=AnimStateTransitionNode_9
   }
   {
    dst=WalkToIdle
    rule=AnimGraph.AnimGraphNode_StateMachine_0.Locomotion.AnimStateTransitionNode_3.Transition
    src=Walk
    via=AnimStateTransitionNode_3
   }
   {
    dst=Idle
    rule=AnimGraph.AnimGraphNode_StateMachine_0.Locomotion.AnimStateTransitionNode_4.Transition
    src=WalkToIdle
    via=AnimStateTransitionNode_4
   }
  ]
 }
 Conduit {path=AnimGraph.AnimGraphNode_StateMachine_0.Locomotion.AnimStateConduitNode_0.Conduit, nodes [Result(bCanEnterTransition=true)]}
 HackState {
  path=AnimGraph.AnimGraphNode_StateMachine_0.Locomotion.AnimStateNode_0.HackState
  nodes [Output Animation Pose(Result=Local Space Ref Pose())]
 }
 Idle {
  path=AnimGraph.AnimGraphNode_StateMachine_0.Locomotion.AnimStateNode_1.Idle
  nodes [Output Animation Pose(Result=Sequence Player 'Caretaker_Idle_Pose'() {sequence=&/Caretaker_Idle_Pose})]
 }
 Walk {
  path=AnimGraph.AnimGraphNode_StateMachine_0.Locomotion.AnimStateNode_2.Walk
  nodes [Output Animation Pose(Result=Sequence Player 'Caretaker_Walk_N'!(PlayRate:float=Get WalkPlayRate() -> (WalkPlayRate:float)) -> (Pose) {groupName=Locomotion, groupRole=AlwaysLeader, method=SyncGroup, sequence=&/Caretaker_Walk_N})]
 }
 WalkToIdle {
  path=AnimGraph.AnimGraphNode_StateMachine_0.Locomotion.AnimStateNode_3.WalkToIdle
  nodes [
   AnimGraphNode_StateMachine_0 = Stop() {is=title AnimGraphNode_StateMachine}
   Output Animation Pose(Result=@AnimGraphNode_StateMachine_0)
  ]
 }
 Stop {
  entry=HackState
  path=AnimGraph.AnimGraphNode_StateMachine_0.Locomotion.AnimStateNode_3.WalkToIdle.AnimGraphNode_StateMachine_0.Stop
  ids=id
  nodes [
   AnimStateConduitNode_0 = Conduit()
   AnimStateNode_0 = HackState()
   AnimStateNode_2 = StopRightPass() {is=title AnimStateNode}
   AnimStateNode_3 = StopLeftPlant() {is=title AnimStateNode}
   AnimStateNode_4 = StopLeftPass() {is=title AnimStateNode}
   AnimStateNode_5 = StopRightPlant() {is=title AnimStateNode}
   AnimStateTransitionNode_1 = AnimStateTransitionNode()
   AnimStateTransitionNode_2 = AnimStateTransitionNode() {crossfadeDuration=0.4, logicType=TLT_Inertialization}
   AnimStateTransitionNode_3 = AnimStateTransitionNode() {crossfadeDuration=0.4, logicType=TLT_Inertialization, priorityOrder=2}
   AnimStateTransitionNode_4 = AnimStateTransitionNode() {crossfadeDuration=0.4, logicType=TLT_Inertialization, priorityOrder=3}
   AnimStateTransitionNode_5 = AnimStateTransitionNode() {crossfadeDuration=0.4, logicType=TLT_Inertialization, priorityOrder=100}
  ]
  links [
   {
    dst=StopLeftPass
    rule=AnimGraph.AnimGraphNode_StateMachine_0.Locomotion.AnimStateNode_3.WalkToIdle.AnimGraphNode_StateMachine_0.Stop.AnimStateTransitionNode_4.Transition
    src=Conduit
    via=AnimStateTransitionNode_4
   }
   {
    dst=StopLeftPlant
    rule=AnimGraph.AnimGraphNode_StateMachine_0.Locomotion.AnimStateNode_3.WalkToIdle.AnimGraphNode_StateMachine_0.Stop.AnimStateTransitionNode_3.Transition
    src=Conduit
    via=AnimStateTransitionNode_3
   }
   {
    dst=StopRightPass
    rule=AnimGraph.AnimGraphNode_StateMachine_0.Locomotion.AnimStateNode_3.WalkToIdle.AnimGraphNode_StateMachine_0.Stop.AnimStateTransitionNode_2.Transition
    src=Conduit
    via=AnimStateTransitionNode_2
   }
   {
    dst=StopRightPlant
    rule=AnimGraph.AnimGraphNode_StateMachine_0.Locomotion.AnimStateNode_3.WalkToIdle.AnimGraphNode_StateMachine_0.Stop.AnimStateTransitionNode_5.Transition
    src=Conduit
    via=AnimStateTransitionNode_5
   }
   {
    dst=Conduit
    rule=AnimGraph.AnimGraphNode_StateMachine_0.Locomotion.AnimStateNode_3.WalkToIdle.AnimGraphNode_StateMachine_0.Stop.AnimStateTransitionNode_1.Transition
    src=HackState
    via=AnimStateTransitionNode_1
   }
  ]
 }
 Conduit {
  path=AnimGraph.AnimGraphNode_StateMachine_0.Locomotion.AnimStateNode_3.WalkToIdle.AnimGraphNode_StateMachine_0.Stop.AnimStateConduitNode_0.Conduit
  nodes [Result(bCanEnterTransition=true)]
 }
 HackState {
  path=AnimGraph.AnimGraphNode_StateMachine_0.Locomotion.AnimStateNode_3.WalkToIdle.AnimGraphNode_StateMachine_0.Stop.AnimStateNode_0.HackState
  nodes [Output Animation Pose(Result=Local Space Ref Pose())]
 }
 StopRightPass {
  path=AnimGraph.AnimGraphNode_StateMachine_0.Locomotion.AnimStateNode_3.WalkToIdle.AnimGraphNode_StateMachine_0.Stop.AnimStateNode_2.StopRightPass
  nodes [Output Animation Pose(Result=Sequence Player 'Caretaker_Walk_Stopzero_RightPass'() {sequence=&/Caretaker_Walk_Stopzero_RightPass})]
 }
 StopLeftPlant {
  path=AnimGraph.AnimGraphNode_StateMachine_0.Locomotion.AnimStateNode_3.WalkToIdle.AnimGraphNode_StateMachine_0.Stop.AnimStateNode_3.StopLeftPlant
  nodes [Output Animation Pose(Result=Sequence Player 'Caretaker_Walk_Stopzero_LeftPlant'() {sequence=&/Caretaker_Walk_Stopzero_LeftPlant})]
 }
 StopLeftPass {
  path=AnimGraph.AnimGraphNode_StateMachine_0.Locomotion.AnimStateNode_3.WalkToIdle.AnimGraphNode_StateMachine_0.Stop.AnimStateNode_4.StopLeftPass
  nodes [Output Animation Pose(Result=Sequence Player 'Caretaker_Walk_Stopzero_LeftPass'() {sequence=&/Caretaker_Walk_Stopzero_LeftPass})]
 }
 StopRightPlant {
  path=AnimGraph.AnimGraphNode_StateMachine_0.Locomotion.AnimStateNode_3.WalkToIdle.AnimGraphNode_StateMachine_0.Stop.AnimStateNode_5.StopRightPlant
  nodes [Output Animation Pose(Result=Sequence Player 'Caretaker_Walk_Stopzero_RightPlant'() {sequence=&/Caretaker_Walk_Stopzero_RightPlant})]
 }
 Transition {
  path=AnimGraph.AnimGraphNode_StateMachine_0.Locomotion.AnimStateNode_3.WalkToIdle.AnimGraphNode_StateMachine_0.Stop.AnimStateTransitionNode_1.Transition
  nodes [Result(bCanEnterTransition=true)]
 }
 Transition {
  path=AnimGraph.AnimGraphNode_StateMachine_0.Locomotion.AnimStateNode_3.WalkToIdle.AnimGraphNode_StateMachine_0.Stop.AnimStateTransitionNode_2.Transition
  nodes [Result(bCanEnterTransition=Get bFootPhase_StopRightPass() -> (bFootPhase_StopRightPass:bool))]
 }
 Transition {
  path=AnimGraph.AnimGraphNode_StateMachine_0.Locomotion.AnimStateNode_3.WalkToIdle.AnimGraphNode_StateMachine_0.Stop.AnimStateTransitionNode_3.Transition
  nodes [Result(bCanEnterTransition=Get bFootPhase_StopLeftPlant() -> (bFootPhase_StopLeftPlant:bool))]
 }
 Transition {
  path=AnimGraph.AnimGraphNode_StateMachine_0.Locomotion.AnimStateNode_3.WalkToIdle.AnimGraphNode_StateMachine_0.Stop.AnimStateTransitionNode_4.Transition
  nodes [Result(bCanEnterTransition=Get bFootPhase_StopLeftPass() -> (bFootPhase_StopLeftPass:bool))]
 }
 Transition {
  path=AnimGraph.AnimGraphNode_StateMachine_0.Locomotion.AnimStateNode_3.WalkToIdle.AnimGraphNode_StateMachine_0.Stop.AnimStateTransitionNode_5.Transition
  nodes [Result(bCanEnterTransition=true)]
 }
 InitialSpawn {
  path=AnimGraph.AnimGraphNode_StateMachine_0.Locomotion.AnimStateNode_5.InitialSpawn
  nodes [Output Animation Pose(Result=Sequence Player 'Caretaker_ScrunchedPose'() {sequence=&/Caretaker_ScrunchedPose})]
 }
 Transition {path=AnimGraph.AnimGraphNode_StateMachine_0.Locomotion.AnimStateTransitionNode_0.Transition, nodes [Result(bCanEnterTransition=true)]}
 Transition {path=AnimGraph.AnimGraphNode_StateMachine_0.Locomotion.AnimStateTransitionNode_1.Transition, nodes [Result(bCanEnterTransition=true)]}
 Transition {
  path=AnimGraph.AnimGraphNode_StateMachine_0.Locomotion.AnimStateTransitionNode_10.Transition
  nodes [Result(bCanEnterTransition=Get IsInitialSpawn() -> (IsInitialSpawn:bool))]
 }
 Transition {
  path=AnimGraph.AnimGraphNode_StateMachine_0.Locomotion.AnimStateTransitionNode_2.Transition
  nodes [Result(bCanEnterTransition=Get bIsMoving() -> (bIsMoving:bool))]
 }
 Transition {
  path=AnimGraph.AnimGraphNode_StateMachine_0.Locomotion.AnimStateTransitionNode_3.Transition
  nodes [Result(bCanEnterTransition=NOT Boolean(A=Get bIsMoving()))]
 }
 Transition {path=AnimGraph.AnimGraphNode_StateMachine_0.Locomotion.AnimStateTransitionNode_4.Transition, nodes [Result()]}
 Transition {
  path=AnimGraph.AnimGraphNode_StateMachine_0.Locomotion.AnimStateTransitionNode_9.Transition
  nodes [Result(bCanEnterTransition=Get bIsMoving())]
 }
 EventGraph {nodes [
   Event BlueprintUpdateAnimation()
   > n1 = Cast To NPC_Pawn_Irwin_Monster_Caretaker!(execute, Object=GetOwningActor()) -> (then, CastFailed, AsNPC Pawn Irwin Monster Caretaker:NPC Pawn Irwin Monster Caretaker Object Reference) {is=title K2Node_DynamicCast}
   > Is Valid(InputObject=Get Mesh(self=@n1))
   > Delay(Duration=0)
   > SetDelayedMaterialParameters(self:Fort AIAnim Instance Caretaker Object Reference)
   n2 = Event BlueprintProcessAnimInputEvent()
   > Branch(Condition=IsNamed(self=@n2.AnimInputEvent, InEventName=aie.CaretakerInWater))
   > n3 = Cast To FortAnimInputEvent_Bool!(execute, Object=@n2.AnimInputEvent) -> (then, CastFailed, AsFort Anim Input Event Bool:Fort Anim Input Event Bool Object Reference) {is=title K2Node_DynamicCast}
   > Set Is Pawn In Water(Is Pawn In Water:bool=Get bValue!(self:Fort Anim Input Event Bool Object Reference=@n3) -> (bValue:bool)) -> (Output_Get:bool)
   Event BlueprintInitializeAnimation()
   > Set IsInitialSpawn(IsInitialSpawn:bool=true) -> (Output_Get:bool)
   AnimNotify_SpawnEnter() {is=title K2Node_Event}
   > Set IsInitialSpawn(IsInitialSpawn=false)
  ]}
]
variables [
 Is Pawn In Water {type=boolean}
 IsInitialSpawn {type=boolean}
]
