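; BP_Caretaker_CollectionVolume_Base is a base Blueprint for Caretaker collection volumes, laid out as a skeleton.
; It declares BeginPlay, actor begin-overlap and custom events, all with empty bodies, plus an empty construction script.
; GetOverlappingActorsFromTrackingVolume is a stub that only returns a default value, so no tracking or collection logic is present here.
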(event EventBeginPlay)

(event Collision|EventActorBeginOverlap
  (params (OtherActor :type Actor%20Object%20Reference)))

(event Custom|CustomEvent_0)

(fn GetOverlappingActorsFromTrackingVolume
  (return 0))

(fn ConstructionScript)
