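; Box-shaped variant of the Caretaker collection volume, using a Box component to detect actors inside it.
; Box begin and end overlap events forward the other actor to the OnTrackingVolumeBeginOverlap and OnTrackingVolumeEndOverlap dispatchers.
; GetOverlappingActorsFromTrackingVolume returns the actors currently overlapping the Box.
; The construction script defers to the parent, and BeginPlay, Tick and actor-level overlap events add no logic.
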(event EventBeginPlay
  (|CallParentFunction) ; This node is disabled and will not be called.\nDrag off pins to build functionality.
)

(event Collision|EventActorBeginOverlap
  (params (OtherActor :type Actor%20Object%20Reference))
  (|CallParentFunction OtherActor) ; This node is disabled and will not be called.\nDrag off pins to build functionality.
)

(event EventTick
  (params (DeltaSeconds :type Float%20%28single-precision%29)))

(event OnComponentBeginOverlap(Box)
  (params (OverlappedComponent :type Primitive%20Component%20Object%20Reference) (OtherActor :type Actor%20Object%20Reference) (OtherComp :type Primitive%20Component%20Object%20Reference) (OtherBodyIndex :type Integer) (bFromSweep :type Boolean) (SweepResult :type Hit%20Result%20Structure%20%28by%20ref%29))
  (Default|CallOnTrackingVolumeBeginOverlap OtherActor))

(event OnComponentEndOverlap(Box)
  (params (OverlappedComponent :type Primitive%20Component%20Object%20Reference) (OtherActor :type Actor%20Object%20Reference) (OtherComp :type Primitive%20Component%20Object%20Reference) (OtherBodyIndex :type Integer))
  (Default|CallOnTrackingVolumeEndOverlap OtherActor))

(fn GetOverlappingActorsFromTrackingVolume
  (return (Collision|GetOverlappingActors (Variables|Default|GetBox) 0 :!class /Script/Engine.PrimitiveComponent)))

(fn ConstructionScript
  (|CallParentFunction))
