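; Short-lived grabber projectile for the Caretaker monster that triggers a Sideways encounter when it hits a player.
; On begin play it offsets its core particle effect along a random vector lerp, sets a short lifespan and deactivates the effect after a delay.
; On hit, a player pawn with the matching tag leads to a one-time RequestSidewaysActivation call on the owning Caretaker.
; Vehicle and animal checks are only debug or empty stubs.
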(event Collision|EventHit
  (params (MyComp :type Primitive%20Component%20Object%20Reference) (Other :type Actor%20Object%20Reference) (OtherComp :type Primitive%20Component%20Object%20Reference) (bSelfMoved :type Boolean) (HitLocation :type Vector :default "0, 0, 0") (HitNormal :type Vector :default "0, 0, 0") (NormalImpulse :type Vector :default "0, 0, 0") (Hit :type Hit%20Result%20Structure%20%28by%20ref%29))
  ;[1 1 1] Hit - Trigger the Encounter
  (bind _asfort_player_pawn (Utilities|Casting|CastToFortPlayerPawn Other)
    (:then
      (if (GameplayTags|HasMatchingGameplayTag _asfort_player_pawn (Variables|Default|GetT_OstTag))
        (bind _asfort_vehicle_interface (Utilities|Casting|CastToFortVehicleInterface Other)
          (:then
            (if (Vehicle|ContainsPlayers _asfort_vehicle_interface)
              (Development|PrintString "Hit")))
          (:CastFailed
            (if (GameplayTags|HasMatchingGameplayTag (Variables|Default|GetT_Animal))
              _)))
        (Utilities|FlowControl|DoOnce false
          (:Completed
            (bind _asnpc_pawn_irwin_monster_caretaker (Utilities|Casting|CastToNPC_Pawn_Irwin_Monster_Caretaker (Actor|GetOwner)))
            (Default|CallRequestSidewaysActivation _asnpc_pawn_irwin_monster_caretaker)))
        (else
          (if (Vehicle|IsinVehicle _asfort_player_pawn)
            _
            (else
              _)))))
    (:CastFailed
      _)))

(event EventBeginPlay
  ;[1 1 1] Set the actors lifespawn from sheets
  (|CallParentFunction)
  (Effects|Components|ParticleSystem|SetVectorParameter (Variables|Default|GetNS_Sideways_Encounter_CoreGrabber) "RandomPosOffset" (Math|Vector|Lerp(Vector) "-300.000000,-300.000000,-100.000000" "300.000000,300.000000,300.000000" (Math|Random|RandomFloat)))
  (Actor|SetLifeSpan 2.0)
  (Utilities|FlowControl|Delay 1.0)
  (Components|Activation|Deactivate (Variables|Default|GetNS_Sideways_Encounter_CoreGrabber)))

(fn ConstructionScript
  (|CallParentFunction))
