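# Material function that controls partial visibility and edge glow for the Caretaker character, applied through the UseCaretaker static switch.
# Vertex-color limb and head masks, a pre-skinned chest sphere mask and attack/visibility scalars combine with distorted panning noise and Fresnel-based temporal dither.
# These build opacity mask and emissive overrides on the incoming material attributes.
# World position offset collapses vertices toward the object position, blended by masks including the local player's Sideways bubble sphere.
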MaterialFunction /Caretaker/Materials/MF_CT_Visibility
graph {nodes [
  n1 = NamedRerouteUsage()
  n2 = Constant() {r=1}
  n3 = Clamp(None=Multiply(A=Desaturation(None=@n1, Fraction=@n2), B=ScalarParameter() {defaultValue=1.5, parameterName=EmissiveOpacity}))
  n4 = GetMaterialAttributes(None=FunctionInput() {inputName=Materials, inputType=FunctionInput_MaterialAttributes}) {attributeGetTypes [
    #1
    #2
    #3
    #4
   ]}
  n5 = Fresnel(Normal=Transform(None=@n4.Normal)) {exponent=1}
  n6 = Subtract(A=Power(Base=OneMinus(None=@n5)), B=NamedRerouteUsage())
  n7 = NamedRerouteUsage()
  n8 = MaterialFunctionCall(Alpha Threshold=Clamp(None=Multiply(A=@n6, B=@n7))) {materialFunction=&DitherTemporalAA_Fortnite}
  Add(A=@n3, B=@n8.Result)
  n9 = NamedRerouteUsage()
  n10 = VertexColor()
  n11 = Max(A=NamedRerouteUsage(), B=@n10.G)
  n12 = NamedRerouteUsage()
  n13 = NamedRerouteUsage()
  n14 = LinearInterpolate(A=@n13, B=NamedRerouteUsage(), Alpha=NamedRerouteUsage())
  n15 = NamedRerouteUsage()
  n16 = NamedRerouteUsage()
  n17 = LinearInterpolate(
   A=LinearInterpolate(A=LinearInterpolate(A=@n14, B=@n15, Alpha=@n16), B=NamedRerouteUsage(), Alpha=NamedRerouteUsage())
   B=@n13
   Alpha=NamedRerouteUsage()
  )
  n18 = Subtract(A=Subtract(A=Multiply(A=@n9, B=@n11), B=@n12), B=Multiply(A=OneMinus(None=@n17)) {constB=2})
  n19 = ScalarParameter() {defaultValue=0.05, parameterName=EdgeGlowRadius}
  n20 = Constant()
  n21 = SphereMask(A=@n18, B=ScalarParameter() {parameterName=EdgeGlowPos}, Radius=@n19, Hardness=@n20)
  n22 = SphereMask(A=@n10.G, B=Constant() {r=0.2}, Radius=@n19, Hardness=@n20)
  n23 = Multiply(A=@n12) {constB=2}
  Add(A=@n21, B=Clamp(None=Subtract(A=@n22, B=@n23)))
  n24 = Subtract(A=ObjectPositionWS(), B=WorldPosition().XYZ)
  Add(A=@n24, B=Multiply(A=Constant() {r=350}, B=ObjectOrientation()))
  n25 = Max(A=NamedRerouteUsage(), B=NamedRerouteUsage())
  n26 = Max(A=NamedRerouteUsage(), B=NamedRerouteUsage())
  n27 = NamedRerouteUsage()
  n28 = NamedRerouteUsage()
  Ceil(None=Max(A=Max(A=@n25, B=@n26), B=Max(A=@n27, B=@n28)))
  ChannelMaskParameter(None=Divide(
    A=NamedRerouteUsage()
    B=MaterialFunctionCall() {materialFunction=&ObjectLocalBounds}.Local Bounds Max
   )) {maskChannel=Blue, defaultValue=<rgba 0 0 1 0>}
  CollectionParameter() {collection=&FortniteMaterialParameters, parameterName=PlayerPosition}
  Divide(
   A=NamedRerouteUsage()
   B=MaterialFunctionCall() {materialFunction=&ObjectLocalBounds}.Local Bounds Size
  )
  n29 = WorldPosition()
  n30 = CollectionParameter() {
   collection=&FortniteMaterialParameters
   parameterName=SidewaysBubble_Location_LocalP1
  }
  n31 = CollectionParameter() {
   collection=&FortniteMaterialParameters
   parameterName=SidewaysBubble_Scale_LocalP1
  }
  n32 = SphereMask(
   A=@n29.XYZ
   B=ComponentMask(None=@n30) {b=true, g=true, r=true}
   Radius=Subtract(A=@n31, B=ScalarParameter() {defaultValue=1200, parameterName=BubbleRadiusBias})
   Hardness=Constant() {r=1}
  )
  n33 = Add(A=Max(A=NamedRerouteUsage(), B=@n32), B=Round(None=NamedRerouteUsage()))
  n34 = NamedRerouteUsage()
  n35 = NamedRerouteUsage()
  n36 = LinearInterpolate(A=@n34, B=Add(A=@n34, B=NamedRerouteUsage()), Alpha=@n35)
  n37 = SetMaterialAttributes!(
   MaterialAttributes=@n4.MaterialAttributes
   WorldPositionOffset=LinearInterpolate(A=@n24, B=@n4.WorldPositionOffset, Alpha=Clamp(None=@n33))
   BaseColor=@n4.MaterialAttributes
   EmissiveColor=@n36
   OpacityMask=@n35
   Opacity
  ) {attributeSetTypes [
    #1
    #3
    #4
    #5
    #6
   ]}
  FunctionOutput(None=StaticSwitchParameter(True=@n37, False=@n4.MaterialAttributes) {parameterName=UseCaretaker})
  LinearInterpolate(A=NamedRerouteUsage(), B=@n35, Alpha=@n35)
  LinearInterpolate(A=NamedRerouteUsage(), B=@n36, Alpha=@n35)
  Max(A=@n3, B=@n8.Result)
  n38 = NamedRerouteUsage()
  Max(A=Ceil(None=@n38), B=@n35)
  Max(A=@n32, B=@n35)
  Multiply()
  Multiply(A=NamedRerouteUsage()) {constB=3}
  n39 = Multiply(A=TextureCoordinate(), B=ScalarParameter() {defaultValue=1, parameterName=WPONoiseTiling})
  n40 = TextureCoordinate()
  n41 = ScalarParameter() {defaultValue=1, parameterName=WPONoiseTiling}
  n42 = TextureSample(UVs=Panner(Coordinate=Multiply(A=@n40, B=@n41)) {speedX=-0.02, speedY=-0.02}) {texture=/Game/Characters/Player/Male/Medium/Bodies/M_MED_Viper/Textures/Glitch/T_SoftNoise}
  n43 = Subtract(A=Multiply(A=ComponentMask(None=@n42.RGB) {g=true, r=true}) {constB=2})
  n44 = ScalarParameter() {defaultValue=0.1, parameterName=WPODistortion_Strength}
  n45 = TextureSample(UVs=Add(A=Panner(Coordinate=@n39) {speedX=0.05, speedY=0.05}, B=Multiply(A=@n43, B=@n44))) {texture=&T_NoisePack_01}
  n46 = ScalarParameter() {defaultValue=32, parameterName=WPO_Strength}
  NamedRerouteDeclaration(None=Multiply(A=VertexNormalWS(), B=Multiply(A=@n45.RGB, B=@n46))) {name=WPO, nodeColor=<rgba 1 0.3281 0 1>}
  NamedRerouteDeclaration(None=ScalarParameter() {parameterName=snap_atk}) {name=snap_ATK, nodeColor=<rgba 1 0 0.75 1>}
  n47 = PreSkinnedPosition()
  NamedRerouteDeclaration(None=ShaderStageSwitch(PixelShader=VertexInterpolator(VS=@n47).PS, VertexShader=@n47)) {name=PreskinnedPos, nodeColor=<rgba 1 0.6797 0 1>}
  n48 = SphereMask(
   A=NamedRerouteUsage()
   B=VectorParameter() {parameterName=ChestMaskPosition, defaultValue=<rgba -41 61 881 1>}.RGB
   Radius=ScalarParameter() {defaultValue=256, parameterName=ChestGradientRadius}
   Hardness=ScalarParameter() {parameterName=ChestGradientHardness}
  )
  NamedRerouteDeclaration(None=@n48) {name=ChestGradient, nodeColor=<rgba 1 0.4219 0 1>}
  NamedRerouteDeclaration(None=ScalarParameter() {parameterName=ethereal_atk}) {name=etheral_ATK, nodeColor=<rgba 0 1 0.5312 1>}
  n49 = VectorParameter() {parameterName=EdgeGlowColor, defaultValue=<rgba 0.5098 0 1 1>}
  n50 = Multiply(A=@n49.RGB, B=@n38)
  n51 = ScalarParameter() {defaultValue=5, parameterName=EdgeGlowStrength}
  NamedRerouteDeclaration(None=Power(Base=Multiply(A=@n50, B=@n51))) {name=EdgeEmissive, nodeColor=<rgba 0 0.5547 1 1>}
  NamedRerouteDeclaration(None=ScalarParameter() {parameterName=sideways_atk}) {name=sideways_ATK, nodeColor=<rgba 1 0.1875 0 1>}
  NamedRerouteDeclaration(None=ScalarParameter() {parameterName=smash_atk}) {name=smash_ATK, nodeColor=<rgba 1 0.6562 0 1>}
  NamedRerouteDeclaration(None=ScalarParameter() {parameterName=swipe_atk}) {name=swipe_ATK, nodeColor=<rgba 1 0 0.5391 1>}
  n52 = Ceil(None=@n48)
  n53 = NamedRerouteUsage()
  NamedRerouteDeclaration(None=Multiply(A=@n52, B=Ceil(None=@n53))) {name=ChestMask, nodeColor=<rgba 0 0.5078 1 1>}
  n54 = TextureCoordinate()
  n55 = ScalarParameter() {defaultValue=8, parameterName=NoiseTiling}
  n56 = Panner(Coordinate=Multiply(A=TextureCoordinate(), B=ScalarParameter() {defaultValue=8, parameterName=NoiseTiling})) {speedX=-0.15, speedY=-0.15}
  n57 = Multiply(A=ComponentMask(None=TextureSample(UVs=@n56) {texture=&T_SoftNoise}.RGB) {g=true, r=true}) {constB=2}
  n58 = Add(
   A=Panner(Coordinate=Multiply(A=@n54, B=@n55)) {speedX=0.25, speedY=0.25}
   B=Multiply(A=Subtract(A=@n57), B=ScalarParameter() {defaultValue=0.1, parameterName=Distortion_Strength})
  )
  NamedRerouteDeclaration(None=Multiply(
    A=TextureSample(UVs=@n58) {texture=&T_NoisePack_01}.G
    B=ScalarParameter() {defaultValue=0.1, parameterName=NoiseStrength}
   )) {name=Noise, nodeColor=<rgba 1 0 0.9844 1>}
  n59 = NamedRerouteUsage()
  n60 = VertexColor()
  n61 = SphereMask(A=@n60.R, B=Constant() {r=0.25}, Radius=Constant() {desc=1, r=0.1})
  n62 = NamedRerouteDeclaration(None=Multiply(A=Ceil(None=@n59), B=@n61)) {name=RightArm_Mask, nodeColor=<rgba 0 1 0.8125 1>}
  n63 = SphereMask(A=@n60.R, B=Constant() {r=0.5}, Radius=Constant() {desc=1, r=0.1})
  n64 = Multiply(B=SphereMask(A=@n60.R, B=Constant() {r=0.75}, Radius=Constant() {desc=1, r=0.1}))
  n65 = NamedRerouteUsage()
  n66 = SphereMask(A=@n60.R, B=Constant() {r=1}, Radius=Constant() {desc=1, r=0.1})
  n67 = NamedRerouteDeclaration(None=Multiply(A=Ceil(None=@n65), B=@n66)) {name=LeftArm_Mask, nodeColor=<rgba 1 0.2109 0 1>}
  n68 = SphereMask(A=@n60.R, B=Constant() {r=0.125}, Radius=Constant() {r=0.1})
  n69 = Ceil(None=NamedRerouteUsage())
  n70 = Add(A=NamedRerouteUsage(), B=NamedRerouteDeclaration(None=Multiply(A=@n68, B=@n69)) {name=Head_Mask, nodeColor=<rgba 1 0.4453 0 1>})
  n71 = Add(
   A=NamedRerouteDeclaration(None=Multiply(B=@n63)) {name=RightLeg_Mask, nodeColor=<rgba 0 0.0859 1 1>}
   B=Add(A=NamedRerouteDeclaration(None=@n64) {name=LeftLeg_Mask, nodeColor=<rgba 0.875 1 0 1>}, B=Add(A=@n67, B=@n70))
  )
  NamedRerouteDeclaration(None=Clamp(None=Add(A=@n62, B=@n71))) {name=MasksCombined, nodeColor=<rgba 0.8516 1 0 1>}
  NamedRerouteDeclaration(None=ScalarParameter() {parameterName=head_vis}) {name=head_vis, nodeColor=<rgba 0 1 0.625 1>}
  NamedRerouteDeclaration(None=Multiply(A=@n21, B=NamedRerouteUsage())) {name=EdgeGlowMask, nodeColor=<rgba 1 0.7031 0 1>}
  n72 = Ceil(None=@n18)
  NamedRerouteDeclaration(None=Clamp(None=Clamp(None=@n72))) {name=Opacity, nodeColor=<rgba 0.3828 0 1 1>}
  n73 = Multiply(A=@n49.RGB, B=NamedRerouteUsage())
  NamedRerouteDeclaration(None=Power(Base=Multiply(A=@n73, B=@n51))) {name=Emissive, nodeColor=<rgba 1 0 0.0469 1>}
  NamedRerouteDeclaration(None=@n8.Result) {name=Base_Opacity, nodeColor=<rgba 1 0.2344 0 1>}
  n74 = NamedRerouteUsage()
  n75 = Multiply(A=Time(), B=VectorParameter() {parameterName=NoisePanSpeedAndDIrection, defaultValue=<rgba 64 64 -64 1>}.RGB)
  n76 = TextureObject() {texture=&T_SoftNoise}
  n77 = VectorParameter() {parameterName=Distortion_MacroNoise_Tiling, defaultValue=<rgba 512 512 512 1>}
  n78 = Add(
   A=NamedRerouteUsage()
   B=Multiply(A=Time(), B=VectorParameter() {parameterName=DistortionNoisePanSpeedAndDIrection, defaultValue=<rgba -32 -32 32 1>}.RGB)
  )
  n79 = Subtract(A=Multiply(A=MaterialFunctionCall(TextureObject=@n76, TextureSize=@n77.RGB, WorldPosition=@n78) {materialFunction=&WorldAlignedTexture}.XYZ Texture) {constB=2})
  n80 = ScalarParameter() {defaultValue=0.225, parameterName=DistortionStrength}
  n81 = MaterialFunctionCall(
   TextureObject=TextureObject() {texture=&T_SoftNoise}
   TextureSize=VectorParameter() {parameterName=MacroNoise_Tiling, defaultValue=<rgba 512 512 512 1>}.RGB
   WorldPosition=Add(A=Add(A=@n74, B=@n75), B=Multiply(A=@n79, B=@n80))
  ) {materialFunction=&WorldAlignedTexture}
  NamedRerouteDeclaration(None=Power(Base=ComponentMask(None=@n81.XYZ Texture) {r=true}) {constExponent=1}) {name=MacroMaskNoise, nodeColor=<rgba 1 0 0.9141 1>}
  n82 = PreSkinnedPosition()
  n83 = VertexInterpolator(VS=@n82)
  NamedRerouteDeclaration(None=FunctionInput(Preview=ShaderStageSwitch(PixelShader=@n83.PS, VertexShader=@n82)) {bUsePreviewValueAsDefault=true, inputName=PreskinnedPos}) {name=PreskinnedPosition, nodeColor=<rgba 0 1 0.6719 1>}
  NamedRerouteDeclaration(None=@n4.BaseColor) {name=M_Basecolor, nodeColor=<rgba 0.7344 0 1 1>}
  NamedRerouteDeclaration(None=@n4.EmissiveColor) {name=M_EmissiveColor, nodeColor=<rgba 0.7109 1 0 1>}
  NamedRerouteDeclaration(None=@n61) {name=RA_Mask, nodeColor=<rgba 0 1 0.4375 1>}
  NamedRerouteDeclaration(None=@n66) {name=LA_Mask, nodeColor=<rgba 0 0.1328 1 1>}
  NamedRerouteDeclaration(None=@n68) {name=H_Mask, nodeColor=<rgba 1 0.6563 0 1>}
  NamedRerouteDeclaration(None=@n52) {name=C_Mask, nodeColor=<rgba 0.1016 0 1 1>}
  NamedRerouteUsage()
  NamedRerouteUsage()
  NamedRerouteUsage()
  SphereMask(
   A=@n29.XYZ
   B=ComponentMask(None=CameraPositionWS()) {b=true, g=true, r=true}
   Radius=ScalarParameter() {defaultValue=2048, parameterName=PlayerVisibilityRadius}
   Hardness=Constant() {r=1}
  )
  Subtract(A=Multiply(A=@n45.RGB) {constB=2})
  VectorParameter() {parameterName=ChestSocketLocation, defaultValue=<rgba 0 0 0 1>}
  WorldPosition()
 ]}
