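# Material function for the Caretaker death effect, which twists, collapses, lights and erodes the mesh around the chest socket.
# Death_Curve scales chest-centered sphere masks that drive a vertex twist blended with an inward pull, a purple emissive glow and an opacity cutout.
# Distorted panning noise is declared as reroutes, likely to break up the mask edges.
# Charring, timer and static-switch nodes do not reach the output.
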MaterialFunction /Caretaker/Materials/MF_CaretakerDeathFX
graph {nodes [
  n1 = WorldPosition()
  n2 = VectorParameter() {parameterName=ChestSocketLocation, defaultValue=<rgba 0 0 0 1>}
  n3 = WorldPosition()
  n4 = ScalarParameter() {parameterName=Death_Curve}
  n5 = SphereMask(
   A=@n3.XYZ
   B=@n2.RGB
   Radius=Multiply(A=ScalarParameter() {defaultValue=2400, parameterName=ChestGradientRadius}, B=@n4)
   Hardness=ScalarParameter() {defaultValue=0.5, parameterName=ChestGradientHardness}
  )
  n6 = Clamp(None=Subtract(A=@n5, B=NamedRerouteUsage()))
  n7 = SphereMask(A=@n3.XYZ, B=@n2.RGB, Radius=ScalarParameter() {defaultValue=756, parameterName=TwistRadius, sliderMin=1}, Hardness=Constant())
  n8 = Multiply(A=ScalarParameter() {defaultValue=2.5, parameterName=TwistAmt}, B=@n5)
  n9 = LinearInterpolate(
   A=RotateAboutAxis(
    NormalizedRotationAxis=VectorParameter() {parameterName=ChestSocketVector, defaultValue=<rgba 0 1 0 1>}.RGB
    RotationAngle=Multiply(A=@n7, B=@n8)
    PivotPoint=@n2.RGB
    Position=WorldPosition().XYZ
   )
   B=LinearInterpolate(B=Subtract(A=@n2.RGB, B=@n1.XYZ), Alpha=NamedRerouteDeclaration(None=@n6) {name=ChestGradient, nodeColor=<rgba 1 0.4219 0 1>}) {constB=0}
   Alpha=ScalarParameter() {defaultValue=0.5, parameterName=MixVal}
  )
  Add(B=@n9)
  Constant() {r=1}
  n10 = GetMaterialAttributes(None=FunctionInput() {inputName=Material, inputType=FunctionInput_MaterialAttributes}) {attributeGetTypes [
    #1
    #2
    #3
    #4
   ]}
  n11 = ScalarParameter() {defaultValue=1, parameterName=SamplePos}
  n12 = ScalarParameter() {defaultValue=1, parameterName=Radius}
  n13 = Constant()
  n14 = Multiply(
   A=ScalarParameter() {defaultValue=10, parameterName=GlowStrength}
   B=Multiply(A=Constant3Vector() {constant=<rgba 0.5733 0 1 0>}, B=SphereMask(A=@n6, B=@n11, Radius=@n12, Hardness=@n13))
  )
  n15 = NamedRerouteUsage()
  n16 = SphereMask(
   A=@n3.XYZ
   B=@n2.RGB
   Radius=Multiply(A=ScalarParameter() {defaultValue=2048, parameterName=ChestOpacityGradientRadius}, B=@n4)
   Hardness=Constant()
  )
  n17 = NamedRerouteUsage()
  n18 = OneMinus(None=Ceil(None=Subtract(A=@n16, B=@n17)))
  n19 = SetMaterialAttributes!(
   MaterialAttributes=@n10.MaterialAttributes
   BaseColor=@n10.MaterialAttributes
   EmissiveColor=Add(A=@n10.EmissiveColor, B=Power(Base=@n14))
   OpacityMask=LinearInterpolate(A=@n10.OpacityMask, B=NamedRerouteDeclaration(None=@n18) {name=ChestMask, nodeColor=<rgba 0 0.5078 1 1>}, Alpha=@n15)
   WorldPositionOffset=@n9
  ) {attributeSetTypes [
    #1
    #2
    #3
    #4
   ]}
  FunctionOutput(None=@n19)
  LinearInterpolate(A=@n10.WorldPositionOffset, B=@n9, Alpha=@n15)
  n20 = Multiply(A=TextureCoordinate(), B=ScalarParameter() {defaultValue=8, parameterName=NoiseTiling})
  n21 = TextureCoordinate()
  n22 = ScalarParameter() {parameterName=NoiseTiling}
  n23 = TextureSample(UVs=Panner(Coordinate=Multiply(A=@n21, B=@n22)) {speedX=-0.15, speedY=-0.15}) {texture=/Game/Characters/Player/Male/Medium/Bodies/M_MED_Viper/Textures/Glitch/T_SoftNoise}
  n24 = Subtract(A=Multiply(A=ComponentMask(None=@n23.RGB) {g=true, r=true}) {constB=2})
  n25 = ScalarParameter() {defaultValue=0.1, parameterName=Distortion_Strength}
  n26 = TextureSample(UVs=Add(A=Panner(Coordinate=@n20) {speedX=0.25, speedY=0.25}, B=Multiply(A=@n24, B=@n25))) {texture=&T_NoisePack_01}
  NamedRerouteDeclaration(None=Multiply(A=@n26.G, B=ScalarParameter() {defaultValue=0.5, parameterName=NoiseStrength})) {name=Noise, nodeColor=<rgba 1 0 0.9844 1>}
  NamedRerouteDeclaration(None=Multiply(A=@n26.R, B=ScalarParameter() {defaultValue=0.5, parameterName=WPONosie})) {name=Noise2, nodeColor=<rgba 1 0 0.7734 1>}
  NamedRerouteDeclaration(None=Ceil(None=@n4)) {name=DeathLerp, nodeColor=<rgba 0 0.7656 1 1>}
  n27 = ScalarParameter() {defaultValue=0.7, parameterName=CharSamplePos}
  n28 = ScalarParameter() {defaultValue=1.2, parameterName=CharRadius}
  n29 = Constant() {r=0.5}
  Reroute(None=LinearInterpolate(A=@n10.BaseColor, B=Constant(), Alpha=SphereMask(A=@n5, B=@n27, Radius=@n28, Hardness=@n29)))
  n30 = Time()
  n31 = ScalarParameter() {defaultValue=999999986991104, parameterName=GameTimeFromBP}
  Saturate(None=Divide(A=Subtract(A=@n30, B=@n31), B=ScalarParameter() {defaultValue=2, parameterName=DeathEffectDuration}))
  StaticSwitchParameter() {parameterName=UseCaretakerDeathFX}
 ]}
