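# Material function that supplies the base texture-driven material attributes for Deimos Husk enemy characters.
# Samples overridable Diffuse, S Mask, M Mask and Normal parameters in the Deimos group, using shared wrap samplers and Husk Deimos defaults.
# Diffuse and Normal feed their matching attributes, while S Mask green drives Metallic, Specular and Roughness alike.
# M Mask blue drives both Opacity and Ambient Occlusion, and a scalar parameter sets Opacity Mask.
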MaterialFunction /Game/Characters/Enemies/MaterialAttributes/MF_DeimosEnemyTextures
graph {nodes [
  n1 = TextureSampleParameter2D() {
   group=Deimos
   parameterName=Diffuse
   samplerSource=SSM_Wrap_WorldGroupSettings
   texture=/Game/Characters/Enemies/Husk_Deimos/Textures/T_Husk_Deimos_01_Base_D
  }
  n2 = TextureSampleParameter2D() {
   group=Deimos
   parameterName=S Mask
   samplerSource=SSM_Wrap_WorldGroupSettings
   samplerType=SAMPLERTYPE_Masks
   texture=/Game/Characters/Enemies/Husk_Deimos/Textures/T_Husk_Deimos_01_Base_S
  }
  n3 = TextureSampleParameter2D() {
   group=Deimos
   parameterName=M Mask
   samplerSource=SSM_Wrap_WorldGroupSettings
   samplerType=SAMPLERTYPE_Masks
   texture=/Game/Characters/Enemies/Husk_Deimos/Textures/T_Husk_Deimos_01_Base_M
  }
  n4 = TextureSampleParameter2D() {
   group=Deimos
   parameterName=Normal
   samplerSource=SSM_Wrap_WorldGroupSettings
   samplerType=SAMPLERTYPE_Normal
   texture=/Game/Characters/Enemies/Husk_Deimos/Textures/T_Husk_Deimos_01_Base_N
  }
  FunctionOutput(None=MakeMaterialAttributes!(
    BaseColor=@n1.RGB
    Metallic=@n2.G
    Specular=@n2.G
    Roughness=@n2.G
    Anisotropy
    EmissiveColor
    Opacity=@n3.B
    OpacityMask=ScalarParameter() {defaultValue=1, group=Base Material, parameterName=Opacity Mask}
    Normal=@n4.RGB
    Tangent
    WorldPositionOffset
    SubsurfaceColor
    ClearCoat
    ClearCoatRoughness
    AmbientOcclusion=@n3.B
    Refraction
    CustomizedUVs_0
    CustomizedUVs_1
    CustomizedUVs_2
    CustomizedUVs_3
    CustomizedUVs_4
    CustomizedUVs_5
    CustomizedUVs_6
    CustomizedUVs_7
    PixelDepthOffset
    ShadingModel
    Displacement
   ))
 ]}
