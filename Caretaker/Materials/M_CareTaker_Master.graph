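# Master material for the Caretaker enemy's skeletal mesh, using material attributes with Default Lit shading.
# Shared Deimos enemy texture and emissive functions build the surface attributes, which pass through the Caretaker visibility function.
# That result branches into the Deimos final-attribute function followed by the Creative move tool, a Caretaker death FX function, and a cinematic character post-FX function.
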Material /Caretaker/Materials/M_CareTaker_Master
graph {
 properties {
  bUseMaterialAttributes=true
  bUsedWithMeshDeformer=true
  bUsedWithSkeletalMesh=true
  pixelDepthOffsetMode=PDOM_Legacy
  usedShadingModels=Default Lit
 }
 nodes [
  n1 = BreakMaterialAttributes!(Attr=MaterialFunctionCall() {materialFunction=/Game/Characters/Enemies/MaterialAttributes/MF_DeimosEnemyTextures}.Result)
  n2 = MaterialFunctionCall!(
   PreskinnedPos
   Materials=MakeMaterialAttributes!(
    BaseColor=@n1.BaseColor
    Metallic=@n1.BaseColor
    Specular=@n1.BaseColor
    Roughness=@n1.BaseColor
    Anisotropy
    EmissiveColor=MaterialFunctionCall() {materialFunction=/Game/Characters/Enemies/MaterialAttributes/MF_DeimosEnemyEmissives}.Result
    Opacity
    OpacityMask
    Normal=@n1.BaseColor
    Tangent
    WorldPositionOffset
    SubsurfaceColor
    ClearCoat
    ClearCoatRoughness
    AmbientOcclusion
    Refraction
    CustomizedUVs_0
    CustomizedUVs_1
    CustomizedUVs_2
    CustomizedUVs_3
    CustomizedUVs_4
    CustomizedUVs_5
    CustomizedUVs_6
    CustomizedUVs_7
    PixelDepthOffset
    ShadingModel
    Displacement
   )
  ) {materialFunction=&/MF_CT_Visibility}
  MaterialFunctionCall(MA In=MaterialFunctionCall!(Material=@n2.Result) {materialFunction=/Game/Characters/Enemies/MaterialAttributes/MF_DeimosEnemyFinalMaterialAttribute}.Result) {materialFunction=&MF_CreativeMoveTool}
  MaterialFunctionCall!(Material=@n2.Result) {materialFunction=&/MF_CaretakerDeathFX}
  MaterialFunctionCall(In=@n2.Result) {materialFunction=&CINE_MF_CharPostFX}
 ]
}
