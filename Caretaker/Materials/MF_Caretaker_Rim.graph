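# Material function that adds a colored rim light to incoming material attributes, apparently for the Caretaker character.
# The mask multiplies ambient occlusion, upward-facing vertex normals, Fresnel and an offset gradient along a selectable axis (Z by default).
# The gradient divides pre-skinned position by local bounds size, so it stays fixed to the mesh during animation.
# The mask blends Base Color toward a brightened LightColor tint and adds LightColor glow to Emissive.
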MaterialFunction /Caretaker/Materials/MF_Caretaker_Rim
graph {nodes [
  n1 = GetMaterialAttributes(None=FunctionInput() {inputName=Material, inputType=FunctionInput_MaterialAttributes}) {attributeGetTypes [#1, #2, #3]}
  n2 = VectorParameter() {parameterName=LightColor, defaultValue=<rgba 1 0.3857 0 1>}
  n3 = Multiply(A=Multiply(A=@n1.BaseColor, B=@n2.RGB), B=ScalarParameter() {defaultValue=15, parameterName=LightBrightness})
  n4 = DotProduct(A=Constant3Vector() {constant=<rgba 0 0 1 0>}, B=VertexNormalWS())
  n5 = Clamp(None=@n4)
  n6 = Fresnel() {exponent=1}
  n7 = PreSkinnedPosition()
  n8 = Add(
   A=ShaderStageSwitch(PixelShader=VertexInterpolator(VS=@n7).PS, VertexShader=@n7)
   B=VectorParameter() {parameterName=GradientZ_Offset, defaultValue=<rgba 0 0 -512 1>}.RGB
  )
  n9 = MaterialFunctionCall() {materialFunction=&ObjectLocalBounds}
  n10 = Multiply(
   A=Multiply(A=Multiply(A=@n1.AmbientOcclusion, B=@n5), B=Power(Base=@n6))
   B=ChannelMaskParameter(None=Divide(A=@n8, B=@n9.Local Bounds Size)) {maskChannel=Blue, parameterName=ZMask, defaultValue=<rgba 0 0 1 0>}.R
  )
  n11 = Multiply(A=Multiply(A=@n2.RGB, B=@n10), B=ScalarParameter() {defaultValue=4, parameterName=LightBrightness2})
  FunctionOutput(None=SetMaterialAttributes!(
    MaterialAttributes=@n1.MaterialAttributes
    BaseColor=LinearInterpolate(A=@n1.BaseColor, B=@n3, Alpha=@n10)
    EmissiveColor=Add(A=@n1.EmissiveColor, B=@n11)
   ) {attributeSetTypes [#1, #2]})
  PixelNormalWS()
  Power(Base=@n4)
  TextureSample() {texture=/Caretaker/Textures/T_Caretaker_Body_D}
 ]}
