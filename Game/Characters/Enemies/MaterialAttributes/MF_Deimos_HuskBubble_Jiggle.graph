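# Material function for Deimos enemy husk bubbles that layers animated jiggle and glow onto incoming material attributes.
# BubbleMask blue offsets sine timing and scales normal-direction displacement, while green masks the bubble areas.
# A view-dot-normal term, cut by a parallax web pattern distorted with panning Perlin noise, adds emissive and darkens base color, blended by UseBubbleEmissve.
# The UseBubbleJiggle static switch passes the input through unchanged when off.
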MaterialFunction /Game/Characters/Enemies/MaterialAttributes/MF_Deimos_HuskBubble_Jiggle
graph {nodes [
  n1 = GetMaterialAttributes(None=FunctionInput() {inputName=Material, inputType=FunctionInput_MaterialAttributes}) {attributeGetTypes [
    #1
    #2
    #3
    #4
   ]}
  n2 = TextureSampleParameter2D() {parameterName=BubbleMask, texture=&T_BlacksRGB}
  n3 = Add(
   A=Multiply(A=@n2.B, B=ScalarParameter() {defaultValue=1, group=HuskBubble, parameterName=TimingOffsetSpread})
   B=Multiply(A=Time(), B=ScalarParameter() {defaultValue=0.25, group=HuskBubble, parameterName=AnimSpeed})
  )
  n4 = ScalarParameter() {defaultValue=1, group=HuskBubble, parameterName=DisplacementOffsetScale}
  n5 = Multiply(
   A=ConstantBiasScale(None=Sine(None=@n3))
   B=Multiply(A=Multiply(A=@n2.B, B=@n4), B=ScalarParameter() {defaultValue=4, group=HuskBubble, parameterName=Displacement_Amt})
  )
  n6 = VertexNormalWS()
  n7 = Add(A=@n1.WorldPositionOffset, B=Multiply(A=@n2.G, B=Multiply(A=@n5, B=@n6)))
  n8 = CameraVectorWS()
  n9 = Transform(None=@n1.Normal)
  n10 = Power(Base=Subtract(A=DotProduct(A=@n8, B=@n9), B=ScalarParameter() {defaultValue=0.4, group=HuskBubble, parameterName=FresnelFalloff}))
  n11 = TextureCoordinate()
  n12 = ScalarParameter() {defaultValue=4, group=HuskBubble, parameterName=HuskBubbleWenTiling}
  n13 = Panner(
   Coordinate=Multiply(A=@n11, B=ScalarParameter() {defaultValue=4, group=HuskBubble, parameterName=NoiseTiling})
   Speed=ComponentMask(None=VectorParameter() {group=HuskBubble, parameterName=DistortionSpeed, defaultValue=<rgba 0 0.1 0 1>}.RGB) {g=true, r=true}
  )
  n14 = Subtract(A=Multiply(A=TextureSample(UVs=@n13) {texture=/Game/Global/Textures/Default/Noise/3xPerlin_M}.RGB) {constB=2})
  n15 = Add(
   A=BumpOffset(
    Coordinate=Multiply(A=@n11, B=@n12)
    Height=ScalarParameter() {defaultValue=-1, parameterName=" Height"}
    HeightRatioInput=ScalarParameter() {defaultValue=0.1, parameterName=HeightRatios}
   )
   B=Multiply(
    A=ComponentMask(None=@n14) {g=true, r=true}
    B=ScalarParameter() {defaultValue=0.025, group=HuskBubble, parameterName=HuskWebDistortionStrength}
   )
  )
  n16 = OneMinus(None=Multiply(
    A=TextureSampleParameter2D(UVs=@n15) {parameterName=BubbleMask, texture=&T_BlacksRGB}.RGB
    B=ScalarParameter() {defaultValue=0.75, group=HuskBubble, parameterName=WebOpacity}
   ))
  n17 = Multiply(A=@n2.G, B=Multiply(A=Clamp(None=@n10), B=ComponentMask(None=@n16) {r=true}))
  n18 = Multiply(
   A=Multiply(A=@n17, B=VectorParameter() {parameterName=HuskBubbleEmissiveColor, defaultValue=<rgba 0 0 0 1>}.RGB)
   B=ScalarParameter() {defaultValue=2, group=HuskBubble, parameterName=BubbleEmissveStrength}
  )
  n19 = ScalarParameter() {parameterName=UseBubbleEmissve}
  n20 = LinearInterpolate(A=@n1.EmissiveColor, B=Add(A=@n1.EmissiveColor, B=Power(Base=@n18)), Alpha=@n19)
  n21 = Add(A=OneMinus(None=@n2.G), B=@n17)
  n22 = ScalarParameter() {defaultValue=0.5, parameterName=FresnelMinDarknes}
  n23 = LinearInterpolate(A=@n1.BaseColor, B=Multiply(A=@n1.BaseColor, B=Clamp(None=@n21, Min=@n22)), Alpha=@n19)
  FunctionOutput(None=StaticSwitchParameter(
    True=SetMaterialAttributes!(MaterialAttributes=@n1.MaterialAttributes, WorldPositionOffset=@n7, EmissiveColor=@n20, BaseColor=@n23) {attributeSetTypes [#1, #3, #4]}
    False=@n1.MaterialAttributes
   ) {defaultValue=true, parameterName=UseBubbleJiggle})
  VertexNormalWS()
 ]}
