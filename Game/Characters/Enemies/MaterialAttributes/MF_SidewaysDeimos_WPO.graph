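# Material function for the Sideways Deimos enemy that adds an animated wobble to the incoming material's World Position Offset, gated by the UseWPOWobble static switch.
# Three time-driven sines, sped by red vertex color, form an XYZ offset scaled by WPO_Offset_Amt; only the X phase offset varies with red.
# That offset is added to, not multiplied by, the normalized object-center-to-pixel direction.
# Any non-zero red enables the full offset through a ceiled mask.
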MaterialFunction /Game/Characters/Enemies/MaterialAttributes/MF_SidewaysDeimos_WPO
graph {nodes [
  n1 = GetMaterialAttributes(None=FunctionInput() {inputName=Material, inputType=FunctionInput_MaterialAttributes}) {attributeGetTypes [#1]}
  n2 = Normalize(VectorInput=Subtract(A=WorldPosition().XYZ, B=ObjectPositionWS()))
  n3 = Time()
  n4 = VertexColor()
  n5 = Subtract(A=Multiply(A=@n4.R) {constB=2})
  n6 = ScalarParameter() {defaultValue=1, parameterName=Time_offsetSpread_01}
  n7 = Multiply(A=Add(A=@n3, B=Multiply(A=@n5, B=@n6)), B=Multiply(A=ScalarParameter() {defaultValue=0.2, parameterName=Speed_01}, B=@n4.R))
  n8 = Multiply(
   A=Add(A=@n3, B=ScalarParameter() {defaultValue=1.5, parameterName=Time_offsetSpread_02})
   B=Multiply(A=ScalarParameter() {defaultValue=0.2, parameterName=Speed_02}, B=@n4.R)
  )
  n9 = Multiply(
   A=Add(A=@n3, B=ScalarParameter() {defaultValue=2, parameterName=Time_offsetSpread_03})
   B=Multiply(A=ScalarParameter() {defaultValue=0.2, parameterName=Speed_03}, B=@n4.R)
  )
  n10 = Multiply(
   A=MaterialFunctionCall(X=Sine(None=@n7), Y=Sine(None=@n8), Z=Sine(None=@n9)) {materialFunction=&MakeFloat3}.Result
   B=VectorParameter() {parameterName=WPO_Offset_Amt, defaultValue=<rgba 2 2 2 1>}.RGB
  )
  n11 = LinearInterpolate(A=@n1.WorldPositionOffset, B=Add(A=@n1.WorldPositionOffset, B=Add(A=@n2, B=@n10)), Alpha=Clamp(None=Ceil(None=@n4.R)))
  FunctionOutput(None=SetMaterialAttributes!(
    MaterialAttributes=@n1.MaterialAttributes
    WorldPositionOffset=StaticSwitchParameter(True=@n11, False=@n1.WorldPositionOffset) {parameterName=UseWPOWobble}
   ) {attributeSetTypes [#1]})
 ]}
