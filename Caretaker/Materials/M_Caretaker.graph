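# M_Caretaker is the Default Lit skeletal mesh surface material for the Caretaker, built on shared Deimos enemy material functions.
# It combines the enemy texture and emissive functions into one material attributes set.
# That set passes through a chain of functions: Caretaker rim, Sideways Deimos world position offset, husk bubble jiggle, enemy final attributes, creative move tool and cinematic character post FX.
# It supports mesh deformers and uses the Purple_Flesh physical material.
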Material /Caretaker/Materials/M_Caretaker
graph {
 properties {
  bUseMaterialAttributes=true
  bUsedWithMeshDeformer=true
  bUsedWithSkeletalMesh=true
  physMaterial=/Game/FortressPhysicalMaterials/Purple_Flesh
  pixelDepthOffsetMode=PDOM_Legacy
  usedShadingModels=Default Lit
 }
 nodes [
  n1 = BreakMaterialAttributes!(Attr=MaterialFunctionCall() {materialFunction=/Game/Characters/Enemies/MaterialAttributes/MF_DeimosEnemyTextures}.Result)
  n2 = MakeMaterialAttributes!(
   BaseColor=@n1.BaseColor
   Metallic=@n1.BaseColor
   Specular=@n1.BaseColor
   Roughness=@n1.BaseColor
   Anisotropy
   EmissiveColor=MaterialFunctionCall() {materialFunction=/Game/Characters/Enemies/MaterialAttributes/MF_DeimosEnemyEmissives}.Result
   Opacity
   OpacityMask
   Normal=@n1.BaseColor
   Tangent
   WorldPositionOffset
   SubsurfaceColor
   ClearCoat
   ClearCoatRoughness
   AmbientOcclusion
   Refraction
   CustomizedUVs_0
   CustomizedUVs_1
   CustomizedUVs_2
   CustomizedUVs_3
   CustomizedUVs_4
   CustomizedUVs_5
   CustomizedUVs_6
   CustomizedUVs_7
   PixelDepthOffset
   ShadingModel
   Displacement
  )
  n3 = MaterialFunctionCall!(Material=MaterialFunctionCall!(Material=MaterialFunctionCall!(Material=@n2) {materialFunction=&/MF_Caretaker_Rim}.Result) {materialFunction=/Game/Characters/Enemies/MaterialAttributes/MF_SidewaysDeimos_WPO}.Result) {materialFunction=/Game/Characters/Enemies/MaterialAttributes/MF_Deimos_HuskBubble_Jiggle}
  MaterialFunctionCall(In=MaterialFunctionCall(MA In=MaterialFunctionCall!(Material=@n3.Result) {materialFunction=/Game/Characters/Enemies/MaterialAttributes/MF_DeimosEnemyFinalMaterialAttribute}.Result) {materialFunction=&MF_CreativeMoveTool}.MA Out) {materialFunction=&CINE_MF_CharPostFX}
 ]
}
