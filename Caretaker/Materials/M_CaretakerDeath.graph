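# Masked death material for the Caretaker that dissolves the body outward from the chest with a purple glow.
# Death_Curve grows sphere masks around ChestSocketLocation, eroded by noise, to drive the opacity cutout and emissive edge.
# World position offset blends a twist around ChestSocketVector with a pull toward the chest.
# Shared Deimos enemy texture and emissive functions supply the base look, with Death_Curve_02 boosting masked emissive.
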Material /Caretaker/Materials/M_CaretakerDeath
graph {
 properties {
  bCastRayTracedShadows=false
  bUseMaterialAttributes=true
  bUsedWithMeshDeformer=true
  bUsedWithNiagaraSprites=true
  bUsedWithSkeletalMesh=true
  blendMode=BLEND_Masked
  pixelDepthOffsetMode=PDOM_Legacy
  usedShadingModels=Default Lit
 }
 nodes [
  n1 = BreakMaterialAttributes!(Attr=MaterialFunctionCall() {materialFunction=/Game/Characters/Enemies/MaterialAttributes/MF_DeimosEnemyTextures}.Result)
  n2 = MakeMaterialAttributes!(
   BaseColor=@n1.BaseColor
   Metallic=@n1.BaseColor
   Specular=@n1.BaseColor
   Roughness=@n1.BaseColor
   Anisotropy
   EmissiveColor=MaterialFunctionCall() {materialFunction=/Game/Characters/Enemies/MaterialAttributes/MF_DeimosEnemyEmissives}.Result
   Opacity
   OpacityMask
   Normal=@n1.BaseColor
   Tangent
   WorldPositionOffset
   SubsurfaceColor
   ClearCoat
   ClearCoatRoughness
   AmbientOcclusion
   Refraction
   CustomizedUVs_0
   CustomizedUVs_1
   CustomizedUVs_2
   CustomizedUVs_3
   CustomizedUVs_4
   CustomizedUVs_5
   CustomizedUVs_6
   CustomizedUVs_7
   PixelDepthOffset
   ShadingModel
   Displacement
  )
  n3 = GetMaterialAttributes(None=@n2) {attributeGetTypes [#1]}
  n4 = ScalarParameter() {defaultValue=300, parameterName=GlowStr}
  n5 = TextureSampleParameter2D() {parameterName=FX_Mask, texture=/Caretaker/Textures/T_Caretaker_Body_E}
  n6 = ScalarParameter() {defaultValue=1, parameterName=Death_Curve_02}
  n7 = Multiply(A=@n3.Emissive Color, B=LinearInterpolate(B=@n4, Alpha=Multiply(A=@n5.R, B=@n6)) {constA=1})
  n8 = ObjectRadius()
  n9 = Multiply(A=SphereMask(
    A=WorldPosition().XYZ
    B=VectorParameter() {parameterName=ChestSocketLocation, defaultValue=<rgba 0 0 0 1>}.RGB
    Radius=Multiply(A=@n8, B=@n6)
    Hardness=ScalarParameter() {parameterName=0.5}
   ))
  n10 = Constant3Vector() {constant=<rgba 0.5733 0 1 0>}
  Add(A=@n7, B=Multiply(A=Multiply(A=@n6, B=@n4), B=Multiply(A=@n9, B=@n10)))
  MaterialFunctionCall!(Material=@n2) {materialFunction=&/MF_Caretaker_Rim}
  Multiply(A=TextureCoordinate(), B=ScalarParameter() {defaultValue=1, parameterName=Cracks_Tiling})
  n11 = Sine(None=Time())
  Multiply(A=@n3.Emissive Color, B=Multiply(A=@n4, B=ConstantBiasScale(None=@n11)))
  Multiply() {constB=0.5}
  n12 = Multiply(A=TextureCoordinate(), B=ScalarParameter() {defaultValue=8, parameterName=NoiseTiling})
  n13 = TextureCoordinate()
  n14 = ScalarParameter() {defaultValue=8, parameterName=NoiseTiling}
  n15 = TextureSample(UVs=Panner(Coordinate=Multiply(A=@n13, B=@n14)) {speedX=-0.15, speedY=-0.15}) {texture=/Game/Characters/Player/Male/Medium/Bodies/M_MED_Viper/Textures/Glitch/T_SoftNoise}
  n16 = Subtract(A=Multiply(A=ComponentMask(None=@n15.RGB) {g=true, r=true}) {constB=2})
  n17 = ScalarParameter() {defaultValue=0.1, parameterName=Distortion_Strength}
  n18 = TextureSample(UVs=Add(A=Panner(Coordinate=@n12) {speedX=0.25, speedY=0.25}, B=Multiply(A=@n16, B=@n17))) {texture=&T_NoisePack_01}
  NamedRerouteDeclaration(None=Multiply(A=@n18.R, B=ScalarParameter() {defaultValue=0.5, parameterName=WPONosie})) {name=Noise2, nodeColor=<rgba 0 0.0625 1 1>}
  NamedRerouteDeclaration(None=Multiply(A=@n18.G, B=ScalarParameter() {defaultValue=0.5, parameterName=NoiseStrength})) {name=Noise, nodeColor=<rgba 1 0 0.9844 1>}
  n19 = WorldPosition()
  n20 = VectorParameter() {parameterName=ChestSocketLocation, defaultValue=<rgba 0 0 0 1>}
  n21 = ScalarParameter() {parameterName=Death_Curve}
  n22 = SphereMask(
   A=@n19.XYZ
   B=@n20.RGB
   Radius=Multiply(A=ScalarParameter() {defaultValue=2400, parameterName=ChestGradientRadius}, B=@n21)
   Hardness=ScalarParameter() {defaultValue=0.5, parameterName=ChestGradientHardness}
  )
  n23 = ScalarParameter() {defaultValue=0.7, parameterName=CharSamplePos}
  n24 = ScalarParameter() {defaultValue=1.2, parameterName=CharRadius}
  n25 = Constant() {r=0.5}
  Reroute(None=LinearInterpolate(A=Constant() {r=1}, B=Constant(), Alpha=SphereMask(A=@n22, B=@n23, Radius=@n24, Hardness=@n25)))
  n26 = Time()
  n27 = ScalarParameter() {defaultValue=999999986991104, parameterName=GameTimeFromBP}
  Saturate(None=Divide(A=Subtract(A=@n26, B=@n27), B=ScalarParameter() {defaultValue=2, parameterName=DeathEffectDuration}))
  ScalarParameter() {parameterName=DeathGlowRadius_02}
  n28 = Clamp(None=Subtract(A=@n22, B=NamedRerouteUsage()))
  n29 = ScalarParameter() {defaultValue=1, parameterName=SamplePos}
  n30 = ScalarParameter() {defaultValue=1, parameterName=Radius}
  n31 = Constant()
  n32 = Multiply(
   A=ScalarParameter() {defaultValue=10, parameterName=GlowStrength}
   B=Multiply(A=@n10, B=SphereMask(A=@n28, B=@n29, Radius=@n30, Hardness=@n31))
  )
  n33 = Subtract(A=@n20.RGB, B=WorldPosition().XYZ)
  n34 = NamedRerouteDeclaration(None=@n28) {name=ChestGradient, nodeColor=<rgba 1 0.4219 0 1>}
  n35 = VectorParameter() {parameterName=ChestSocketVector, defaultValue=<rgba 0 1 0 1>}
  n36 = Multiply(
   A=SphereMask(A=@n19.XYZ, B=@n20.RGB, Radius=ScalarParameter() {defaultValue=756, parameterName=TwistRadius, sliderMin=1}, Hardness=Constant())
   B=Multiply(A=ScalarParameter() {defaultValue=2.5, parameterName=TwistAmt}, B=@n22)
  )
  n37 = WorldPosition() {worldPositionShaderOffset=WPT_ExcludeAllShaderOffsets}
  n38 = Multiply(A=ScalarParameter() {defaultValue=2048, parameterName=ChestOpacityGradientRadius}, B=@n21)
  n39 = Constant()
  n40 = Ceil(None=Subtract(A=SphereMask(A=@n19.XYZ, B=@n20.RGB, Radius=@n38, Hardness=@n39), B=NamedRerouteUsage()))
  SetMaterialAttributes!(
   MaterialAttributes=@n3.MaterialAttributes
   Emissive Color=Add(A=@n7, B=Power(Base=@n32))
   Opacity Mask=NamedRerouteDeclaration(None=OneMinus(None=@n40)) {name=ChestMask, nodeColor=<rgba 0 0.5078 1 1>}
   World Position Offset=LinearInterpolate(
    A=RotateAboutAxis(NormalizedRotationAxis=@n35.RGB, RotationAngle=@n36, PivotPoint=@n20.RGB, Position=@n37.XYZ)
    B=LinearInterpolate(B=@n33, Alpha=@n34) {constB=0}
    Alpha=ScalarParameter() {defaultValue=0.5, parameterName=MixVal}
   )
  ) {attributeSetTypes [#1, #2, #3]}
  VertexInterpolator(VS=PreSkinnedPosition())
 ]
}
