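# Translucent unlit skeletal-mesh material for the Caretaker body, with an emissive edge glow and attack-driven strength parameters.
# Base opacity combines normal-mapped Fresnel, camera-distance fade and a sphere mask that clears the local player's Sideways bubble.
# Vertex-color limb and head masks, a chest gradient and panning, distorted world-aligned noise support masked cutouts.
# Timestamped left/right pulses travel along local bounds, optionally adding opacity and normal-based vertex offset.
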Material /Caretaker/Materials/M_CaretakerTransparency_Body
graph {
 properties {
  bUsedWithMeshDeformer=true
  bUsedWithSkeletalMesh=true
  blendMode=BLEND_Translucent
  pixelDepthOffsetMode=PDOM_Legacy
  shadingModel=MSM_Unlit
  usedShadingModels=Unlit
 }
 nodes [
  n1 = NamedRerouteUsage()
  n2 = NamedRerouteUsage()
  n3 = Multiply(A=@n2, B=ScalarParameter() {defaultValue=2, parameterName=Attacking Opacity Strength})
  Clamp(None=Add(A=@n1, B=Multiply(A=@n1, B=@n3)))
  CollectionParameter() {collection=&FortniteMaterialParameters, parameterName=PlayerPosition}
  n4 = Time()
  n5 = Ceil(None=Divide(A=NamedRerouteUsage()) {constB=3})
  ConstantBiasScale(None=ChannelMaskParameter(None=MaterialFunctionCall(Seed=@n4, Vec3=@n5) {materialFunction=&MF_Hash}.Vec3) {parameterName=ChannelMask}.R)
  n6 = Time()
  n7 = ScalarParameter() {defaultValue=999999995904, parameterName=GameTimeFromBP_Left}
  n8 = ScalarParameter() {defaultValue=0.5, parameterName=PulseDuration}
  n9 = Saturate(None=Divide(A=Subtract(A=@n6, B=@n7), B=@n8))
  ConstantBiasScale(None=Sine(None=Subtract(A=@n9) {constB=0.25}))
  n10 = ScalarParameter() {defaultValue=999999995904, parameterName=GameTimeFromBP_Right}
  n11 = Saturate(None=Divide(A=Subtract(A=@n6, B=@n10), B=@n8))
  ConstantBiasScale(None=Sine(None=Subtract(A=@n11) {constB=0.25}))
  Constant()
  Frac(None=Multiply(A=Time()) {constB=0.25})
  n12 = ScalarParameter() {defaultValue=3, parameterName=AttackingGlowStrength}
  Multiply(A=NamedRerouteUsage(), B=Clamp(None=Multiply(A=@n2, B=@n12), Max=@n12) {minDefault=1})
  n13 = PreSkinnedPosition()
  NamedRerouteDeclaration(None=ShaderStageSwitch(PixelShader=VertexInterpolator(VS=@n13).PS, VertexShader=@n13)) {name=PreskinnedPosition, nodeColor=<rgba 0 1 0.6719 1>}
  n14 = Transform(None=TextureSampleParameter2D() {parameterName=Normal, samplerType=SAMPLERTYPE_Normal, texture=/Caretaker/Textures/T_Caretaker_Body_N}.RGB)
  n15 = Power(Base=OneMinus(None=Fresnel(Normal=@n14) {exponent=1}))
  n16 = NamedRerouteUsage()
  n17 = ChannelMaskParameter(None=Divide(A=@n16, B=MaterialFunctionCall() {materialFunction=&ObjectLocalBounds}.Local Bounds Size)) {maskChannel=Blue, parameterName=Z-Channel, defaultValue=<rgba 0 0 1 0>}
  n18 = ChannelMaskParameter(None=@n16) {parameterName=G_Channel}
  n19 = LinearInterpolate(A=@n11, B=@n9, Alpha=Clamp(None=Ceil(None=@n18.R)))
  n20 = Subtract(
   A=SphereMask(
    A=@n17.R
    B=@n19
    Radius=ScalarParameter() {defaultValue=0.35, parameterName=PulseRadius}
    Hardness=ScalarParameter() {defaultValue=0.15, parameterName=Pulsehardness}
   )
   B=Multiply(A=NamedRerouteUsage())
  )
  n21 = Multiply(A=Multiply(A=OneMinus(None=@n17.R), B=Clamp(None=@n20)), B=Ceil(None=@n19))
  NamedRerouteDeclaration(None=Multiply(A=@n15, B=@n21)) {name=PulseOpacity, nodeColor=<rgba 0 1 0.4375 1>}
  n22 = SphereMask(
   A=NamedRerouteUsage()
   B=VectorParameter() {parameterName=ChestMaskPosition, defaultValue=<rgba -41 61 881 1>}.RGB
   Radius=ScalarParameter() {defaultValue=256, parameterName=ChestGradientRadius}
   Hardness=ScalarParameter() {parameterName=ChestGradientHardness}
  )
  NamedRerouteDeclaration(None=@n22) {name=ChestGradient, nodeColor=<rgba 1 0.4219 0 1>}
  NamedRerouteDeclaration(None=ScalarParameter() {parameterName=ethereal_atk}) {name=etheral_ATK, nodeColor=<rgba 0 1 0.5312 1>}
  n23 = Ceil(None=@n22)
  n24 = NamedRerouteUsage()
  NamedRerouteDeclaration(None=Multiply(A=@n23, B=Ceil(None=@n24))) {name=ChestMask, nodeColor=<rgba 0 0.5078 1 1>}
  n25 = TextureCoordinate()
  n26 = ScalarParameter() {defaultValue=8, parameterName=NoiseTiling}
  n27 = Panner(Coordinate=Multiply(A=TextureCoordinate(), B=ScalarParameter() {parameterName=NoiseTiling})) {speedX=-0.15, speedY=-0.15}
  n28 = Multiply(A=ComponentMask(None=TextureSample(UVs=@n27) {texture=/Game/Characters/Player/Male/Medium/Bodies/M_MED_Viper/Textures/Glitch/T_SoftNoise}.RGB) {g=true, r=true}) {constB=2}
  n29 = Add(
   A=Panner(Coordinate=Multiply(A=@n25, B=@n26)) {speedX=0.25, speedY=0.25}
   B=Multiply(A=Subtract(A=@n28), B=ScalarParameter() {defaultValue=0.1, parameterName=Distortion_Strength})
  )
  NamedRerouteDeclaration(None=Multiply(
    A=TextureSample(UVs=@n29) {texture=&T_NoisePack_01}.G
    B=ScalarParameter() {defaultValue=0.1, parameterName=NoiseStrength}
   )) {name=Noise, nodeColor=<rgba 1 0 0.9844 1>}
  n30 = NamedRerouteUsage()
  n31 = VertexColor()
  n32 = SphereMask(A=@n31.R, B=Constant() {r=0.25}, Radius=Constant() {desc=1, r=0.1})
  n33 = NamedRerouteDeclaration(None=Multiply(A=Ceil(None=@n30), B=@n32)) {name=RightArm_Mask, nodeColor=<rgba 0 1 0.8125 1>}
  n34 = SphereMask(A=@n31.R, B=Constant() {r=0.5}, Radius=Constant() {desc=1, r=0.1})
  n35 = Multiply(B=SphereMask(A=@n31.R, B=Constant() {r=0.75}, Radius=Constant() {desc=1, r=0.1}))
  n36 = NamedRerouteUsage()
  n37 = SphereMask(A=@n31.R, B=Constant() {r=1}, Radius=Constant() {desc=1, r=0.1})
  n38 = NamedRerouteDeclaration(None=Multiply(A=Ceil(None=@n36), B=@n37)) {name=LeftArm_Mask, nodeColor=<rgba 1 0.2109 0 1>}
  n39 = SphereMask(A=@n31.R, B=Constant() {r=0.125}, Radius=Constant() {r=0.1})
  n40 = Ceil(None=NamedRerouteUsage())
  n41 = Add(A=NamedRerouteUsage(), B=NamedRerouteDeclaration(None=Multiply(A=@n39, B=@n40)) {name=Head_Mask, nodeColor=<rgba 1 0.4453 0 1>})
  n42 = Add(
   A=NamedRerouteDeclaration(None=Multiply(B=@n34)) {name=RightLeg_Mask, nodeColor=<rgba 0 0.0859 1 1>}
   B=Add(A=NamedRerouteDeclaration(None=@n35) {name=LeftLeg_Mask, nodeColor=<rgba 0.875 1 0 1>}, B=Add(A=@n38, B=@n41))
  )
  NamedRerouteDeclaration(None=Clamp(None=Add(A=@n33, B=@n42))) {name=MasksCombined, nodeColor=<rgba 0.8516 1 0 1>}
  NamedRerouteDeclaration(None=ScalarParameter() {parameterName=head_vis}) {name=head_vis, nodeColor=<rgba 0 1 0.625 1>}
  n43 = Multiply(A=VectorParameter() {parameterName=EdgeGlowColor, defaultValue=<rgba 0.5098 0 1 1>}.RGB, B=NamedRerouteUsage())
  n44 = ScalarParameter() {defaultValue=5, parameterName=EdgeGlowStrength}
  NamedRerouteDeclaration(None=Power(Base=Multiply(A=@n43, B=@n44))) {name=Emissive, nodeColor=<rgba 1 0 0.0469 1>}
  n45 = WorldPosition()
  n46 = CollectionParameter() {
   collection=&FortniteMaterialParameters
   parameterName=SidewaysBubble_Location_LocalP1
  }
  n47 = ComponentMask(None=@n46) {b=true, g=true, r=true}
  n48 = CollectionParameter() {
   collection=&FortniteMaterialParameters
   parameterName=SidewaysBubble_Scale_LocalP1
  }
  n49 = Subtract(A=@n48, B=ScalarParameter() {defaultValue=768, parameterName=BubbleRadiusBias})
  n50 = Constant() {r=1}
  n51 = WorldPosition()
  n52 = Distance(A=ComponentMask(None=CameraPositionWS()) {b=true, g=true, r=true}, B=@n51.XYZ)
  n53 = ScalarParameter() {defaultValue=4096, parameterName=PlayerVisibilityRadius}
  n54 = OneMinus(None=Divide(A=Subtract(A=@n52, B=@n53), B=ScalarParameter() {defaultValue=1024, parameterName=FadeBias}))
  n55 = Multiply(A=Subtract(A=@n15, B=NamedRerouteUsage()), B=NamedRerouteUsage())
  n56 = Multiply(A=OneMinus(None=SphereMask(A=@n45.XYZ, B=@n47, Radius=@n49, Hardness=@n50)), B=Multiply(A=Clamp(None=@n54), B=Clamp(None=@n55)))
  n57 = Max(A=NamedRerouteUsage(), B=@n56)
  n58 = NamedRerouteUsage()
  n59 = VertexColor()
  n60 = Subtract(A=Multiply(A=NamedRerouteUsage(), B=Max(A=@n58, B=@n59.G)), B=NamedRerouteUsage())
  n61 = NamedRerouteUsage()
  n62 = NamedRerouteUsage()
  n63 = NamedRerouteUsage()
  n64 = LinearInterpolate(
   A=LinearInterpolate(A=LinearInterpolate(A=@n61, B=@n62, Alpha=@n63), B=NamedRerouteUsage(), Alpha=NamedRerouteUsage())
   B=NamedRerouteUsage()
   Alpha=NamedRerouteUsage()
  )
  n65 = NamedRerouteUsage()
  n66 = Multiply(A=OneMinus(None=LinearInterpolate(A=@n64, B=@n61, Alpha=@n65))) {constB=2}
  n67 = Clamp(None=Ceil(None=Subtract(A=@n60, B=@n66)))
  NamedRerouteDeclaration(None=Multiply(
    A=StaticSwitchParameter(True=@n57, False=@n56) {defaultValue=true, parameterName=useLimbsPulse}
    B=LinearInterpolate(Alpha=@n67) {constA=1, constB=0}
   )) {name=Base_Opacity, nodeColor=<rgba 1 0.2344 0 1>}
  n68 = NamedRerouteUsage()
  n69 = Multiply(A=Time(), B=VectorParameter() {parameterName=NoisePanSpeedAndDIrection, defaultValue=<rgba 64 64 -64 1>}.RGB)
  n70 = TextureObject() {texture=&T_SoftNoise}
  n71 = VectorParameter() {parameterName=Distortion_MacroNoise_Tiling, defaultValue=<rgba 512 512 512 1>}
  n72 = Add(
   A=NamedRerouteUsage()
   B=Multiply(A=Time(), B=VectorParameter() {parameterName=DistortionNoisePanSpeedAndDIrection, defaultValue=<rgba -32 -32 32 1>}.RGB)
  )
  n73 = Subtract(A=Multiply(A=MaterialFunctionCall(TextureObject=@n70, TextureSize=@n71.RGB, WorldPosition=@n72) {materialFunction=&WorldAlignedTexture}.XYZ Texture) {constB=2})
  n74 = ScalarParameter() {defaultValue=0.225, parameterName=DistortionStrength}
  n75 = MaterialFunctionCall(
   TextureObject=TextureObject() {texture=&T_SoftNoise}
   TextureSize=VectorParameter() {parameterName=MacroNoise_Tiling, defaultValue=<rgba 512 512 512 1>}.RGB
   WorldPosition=Add(A=Add(A=@n68, B=@n69), B=Multiply(A=@n73, B=@n74))
  ) {materialFunction=&WorldAlignedTexture}
  NamedRerouteDeclaration(None=Power(Base=ComponentMask(None=@n75.XYZ Texture) {r=true}) {constExponent=1}) {name=MacroMaskNoise, nodeColor=<rgba 1 0 0.9141 1>}
  NamedRerouteDeclaration(None=@n32) {name=RA_Mask, nodeColor=<rgba 0 1 0.4375 1>}
  NamedRerouteDeclaration(None=@n37) {name=LA_Mask, nodeColor=<rgba 0 0.1328 1 1>}
  NamedRerouteDeclaration(None=@n39) {name=H_Mask, nodeColor=<rgba 1 0.6562 0 1>}
  NamedRerouteDeclaration(None=@n23) {name=C_Mask, nodeColor=<rgba 0.1016 0 1 1>}
  n76 = Max(
   A=NamedRerouteDeclaration(None=ScalarParameter() {parameterName=sideways_atk}) {name=sideways_ATK, nodeColor=<rgba 1 0.1875 0 1>}
   B=NamedRerouteDeclaration(None=ScalarParameter() {parameterName=smash_atk}) {name=smash_ATK, nodeColor=<rgba 1 0.6562 0 1>}
  )
  n77 = NamedRerouteDeclaration(None=ScalarParameter() {parameterName=swipe_atk}) {name=swipe_ATK, nodeColor=<rgba 1 0 0.5391 1>}
  n78 = ScalarParameter() {parameterName=snap_atk}
  NamedRerouteDeclaration(None=Max(A=Max(A=@n76, B=@n77), B=NamedRerouteDeclaration(None=@n78) {name=snap_ATK, nodeColor=<rgba 1 0 0.75 1>})) {name=Attacking, nodeColor=<rgba 0.6172 0 1 1>}
  SphereMask(A=@n51.XYZ, B=ComponentMask(None=CameraPositionWS()) {b=true, g=true, r=true}, Radius=@n53, Hardness=Constant() {r=1})
  n79 = ScalarParameter() {defaultValue=32, parameterName=PulseWPO}
  StaticSwitchParameter(True=Multiply(A=VertexNormalWS(), B=Multiply(A=@n21, B=@n79)), False=Constant()) {defaultValue=true, parameterName=useLimbsPulse}
 ]
}
