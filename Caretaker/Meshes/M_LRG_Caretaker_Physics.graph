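# Physics asset for the Caretaker biped skeletal mesh (M_LRG_Caretaker), defining its per-bone collision bodies and joint constraints.
# Capsules cover head, spine and limbs, with spheres at both ankles and the right shoulder, and several shapes on some bones.
# Constraints with uniform swing and twist limits chain outward from pelvis and chest.
# The rig is asymmetric: only the left arm has a clavicle body, and the Spine3 body has no constraint.
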PhysicsAsset /Caretaker/Meshes/M_LRG_Caretaker_Physics
bodies {
 C_BipedHeadNeck_A_Head_Jnt [{shapeType=Capsule, length=70, radius=30, center=<34.5888 -4.3614 0.7006>, rotation=<rot -1.264 83.3226 -86.0115>}]
 C_BipedSpine_Main_Chest_Jnt [{shapeType=Capsule, length=80, radius=80, center=<59.2145 -6.1793 31.5163>, rotation=<rot 4.6316 66.3812 6.9164>}]
 C_BipedSpine_Main_Pelvis_Jnt [{shapeType=Capsule, length=90, radius=30, center=<-7.5629 5.4655 7.7494>, rotation=<rot 7.1272 -89.9998 -7.1121>}]
 C_BipedSpine_Main_Spine1_Jnt [{shapeType=Capsule, length=75, radius=30, center=<28.8935 0.9707 2.8918>, rotation=<rot 2.7478 -102.2622 6.5364>}]
 C_BipedSpine_Main_Spine2_Jnt [{shapeType=Capsule, length=75, radius=50, center=<54.3848 7.685 11.0978>, rotation=<rot -8.6038 60.1572 -8.258>}]
 C_BipedSpine_Main_Spine3_Jnt [
  {
   shapeType=Capsule
   length=50
   radius=40
   shapeName=C_BipedSpine_Main_Spine3_Jnt_capsule
   center=<40.8257 60.287 24.452>
   rotation=<rot 6.4098 -50.4317 82.3059>
  }
  {
   shapeType=Capsule
   length=84.3173
   radius=29.167
   shapeName=C_BipedSpine_Main_Spine3_Jnt_capsule
   center=<-5.5202 -27.8679 3.7193>
   rotation=<rot 79.9995 0 10.0009>
  }
 ]
 L_BipedArm_A_Clavicle_Jnt [{shapeType=Capsule, length=47.1741, radius=32, center=<46.6727 2.6302 4.6172>, rotation=<rot 10.1218 -61.3745 -88.9134>}]
 L_BipedArm_A_Elbow_Jnt [
  {shapeType=Capsule, length=125, radius=38.9978, center=<23.6579 2.6619 6.4732>, rotation=<rot 0 88.0363 -91.1223>}
  {shapeType=Capsule, length=200, radius=20, center=<209.3519 -3.6991 5.3616>, rotation=<rot 0 88.0363 -91.1223>}
 ]
 L_BipedArm_A_Shoulder_Jnt [{shapeType=Capsule, length=225, radius=20, center=<138.5317 -2.2522 -1.1028>, rotation=<rot 0 -89.6733 -86.3909>}]
 L_BipedArm_A_Wrist_Jnt [{shapeType=Capsule, length=75.042, radius=52.4936, center=<74.4578 -10.8621 -24.7051>, rotation=<rot -43.3625 -103.6375 100.5587>}]
 L_BipedLeg_A_Ankle_Jnt [{shapeType=Sphere, radius=55, shapeName=L_BipedLeg_A_Ankle_Jnt_sphere, center=<59.5875 -0.0002 -0.0015>}]
 L_BipedLeg_A_Knee_Jnt [
  {shapeType=Capsule, length=140, radius=20, center=<71.057 -9.54 1.2157>, rotation=<rot 0.0969 84.4375 -88.7981>}
  {
   shapeType=Capsule
   length=150
   radius=30
   shapeName=R_BipedLeg_A_Knee_Jnt_capsule
   center=<227.7753 5.6351 2.3294>
   rotation=<rot 88.4576 7.8633 -3.7835>
  }
  {
   shapeType=Capsule
   length=50
   radius=40
   shapeName=R_BipedLeg_A_Knee_Jnt_capsule
   center=<338.3792 28.6119 2.4776>
   rotation=<rot 89.2351 16.2441 7.661>
  }
 ]
 L_BipedLeg_A_Thigh_Jnt [
  {shapeType=Capsule, length=135, radius=25, center=<224.9824 13.2702 12.0014>, rotation=<rot 2.6685 -102.0461 -73.4808>}
  {
   shapeType=Capsule
   length=165
   radius=18
   shapeName=L_BipedLeg_A_Thigh_Jnt_capsule
   center=<68.5283 8.0261 19.4554>
   rotation=<rot -79.1867 0 12.5904>
  }
 ]
 R_BipedArm_A_Elbow_Jnt [
  {shapeType=Capsule, length=150, radius=42.838, center=<-57.7729 3.1513 -4.6971>, rotation=<rot 0 90.8862 -89.4206>}
  {shapeType=Capsule, length=225, radius=25, center=<-276.8676 -0.246 -6.9131>, rotation=<rot 0 90.8862 -89.4207>}
 ]
 R_BipedArm_A_Shoulder_Jnt [
  {shapeType=Sphere, radius=55, shapeName=R_BipedArm_A_Shoulder_Jnt_sphere, center=<-16.0467 0.0005 -9.0343>}
  {shapeType=Capsule, length=180, radius=20, center=<-159.3247 -4.3637 -6.6328>, rotation=<rot 0 -89.4499 -89.0217>}
 ]
 R_BipedArm_A_Wrist_Jnt [{shapeType=Capsule, length=73.1978, radius=52.062, center=<-81.5979 -2.8579 19.7642>, rotation=<rot -43.3635 -103.6359 -78.1184>}]
 R_BipedLeg_A_Ankle_Jnt [{shapeType=Sphere, radius=55, shapeName=R_BipedLeg_A_Ankle_Jnt_sphere, center=<-54.6634 0.0002 5.593>}]
 R_BipedLeg_A_Knee_Jnt [
  {shapeType=Capsule, length=125, radius=20, center=<-79.6462 0.8889 5.2647>, rotation=<rot 0.1909 79.9378 -88.8095>}
  {
   shapeType=Capsule
   length=150
   radius=30
   shapeName=R_BipedLeg_A_Knee_Jnt_capsule
   center=<-239.1068 -6.0981 5.2295>
   rotation=<rot 88.4579 7.8627 1.9373>
  }
  {
   shapeType=Capsule
   length=50
   radius=40
   shapeName=R_BipedLeg_A_Knee_Jnt_capsule
   center=<-330.6415 -17.184 3.2043>
   rotation=<rot 89.2351 16.2421 7.6604>
  }
 ]
 R_BipedLeg_A_Thigh_Jnt [
  {shapeType=Capsule, length=90, radius=25, center=<-243.2598 -24.4692 -9.6452>, rotation=<rot 2.4732 -100.5026 -75.6488>}
  {
   shapeType=Capsule
   length=200
   radius=15
   shapeName=R_BipedLeg_A_Thigh_Jnt_capsule
   center=<-99.1041 -23.8259 -17.1083>
   rotation=<rot -82.0147 0 7.9692>
  }
 ]
}
constraints [
 {
  bone1Name=C_BipedSpine_Main_Spine1_Jnt
  bone2Name=C_BipedSpine_Main_Pelvis_Jnt
  swing1LimitDegrees=45
  swing2LimitDegrees=45
  twistLimitDegrees=45
 }
 {
  bone1Name=C_BipedSpine_Main_Spine2_Jnt
  bone2Name=C_BipedSpine_Main_Spine1_Jnt
  swing1LimitDegrees=45
  swing2LimitDegrees=45
  twistLimitDegrees=45
 }
 {
  bone1Name=C_BipedSpine_Main_Chest_Jnt
  bone2Name=C_BipedSpine_Main_Spine2_Jnt
  swing1LimitDegrees=45
  swing2LimitDegrees=45
  twistLimitDegrees=45
 }
 {
  bone1Name=L_BipedArm_A_Clavicle_Jnt
  bone2Name=C_BipedSpine_Main_Chest_Jnt
  swing1LimitDegrees=45
  swing2LimitDegrees=45
  twistLimitDegrees=45
 }
 {bone1Name=L_BipedArm_A_Shoulder_Jnt, bone2Name=L_BipedArm_A_Clavicle_Jnt, swing1LimitDegrees=45, swing2LimitDegrees=45, twistLimitDegrees=45}
 {bone1Name=L_BipedArm_A_Elbow_Jnt, bone2Name=L_BipedArm_A_Shoulder_Jnt, swing1LimitDegrees=45, swing2LimitDegrees=45, twistLimitDegrees=45}
 {bone1Name=L_BipedArm_A_Wrist_Jnt, bone2Name=L_BipedArm_A_Elbow_Jnt, swing1LimitDegrees=45, swing2LimitDegrees=45, twistLimitDegrees=45}
 {
  bone1Name=R_BipedArm_A_Shoulder_Jnt
  bone2Name=C_BipedSpine_Main_Chest_Jnt
  swing1LimitDegrees=45
  swing2LimitDegrees=45
  twistLimitDegrees=45
 }
 {bone1Name=R_BipedArm_A_Elbow_Jnt, bone2Name=R_BipedArm_A_Shoulder_Jnt, swing1LimitDegrees=45, swing2LimitDegrees=45, twistLimitDegrees=45}
 {bone1Name=R_BipedArm_A_Wrist_Jnt, bone2Name=R_BipedArm_A_Elbow_Jnt, swing1LimitDegrees=45, swing2LimitDegrees=45, twistLimitDegrees=45}
 {
  bone1Name=C_BipedHeadNeck_A_Head_Jnt
  bone2Name=C_BipedSpine_Main_Chest_Jnt
  swing1LimitDegrees=45
  swing2LimitDegrees=45
  twistLimitDegrees=45
 }
 {bone1Name=L_BipedLeg_A_Thigh_Jnt, bone2Name=C_BipedSpine_Main_Pelvis_Jnt, swing1LimitDegrees=45, swing2LimitDegrees=45, twistLimitDegrees=45}
 {bone1Name=L_BipedLeg_A_Knee_Jnt, bone2Name=L_BipedLeg_A_Thigh_Jnt, swing1LimitDegrees=45, swing2LimitDegrees=45, twistLimitDegrees=45}
 {bone1Name=L_BipedLeg_A_Ankle_Jnt, bone2Name=L_BipedLeg_A_Knee_Jnt, swing1LimitDegrees=45, swing2LimitDegrees=45, twistLimitDegrees=45}
 {bone1Name=R_BipedLeg_A_Thigh_Jnt, bone2Name=C_BipedSpine_Main_Pelvis_Jnt, swing1LimitDegrees=45, swing2LimitDegrees=45, twistLimitDegrees=45}
 {bone1Name=R_BipedLeg_A_Knee_Jnt, bone2Name=R_BipedLeg_A_Thigh_Jnt, swing1LimitDegrees=45, swing2LimitDegrees=45, twistLimitDegrees=45}
 {bone1Name=R_BipedLeg_A_Ankle_Jnt, bone2Name=R_BipedLeg_A_Knee_Jnt, swing1LimitDegrees=45, swing2LimitDegrees=45, twistLimitDegrees=45}
]
