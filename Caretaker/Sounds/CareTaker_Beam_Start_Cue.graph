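# Sound cue for the start of the CareTaker's beam attack, playing a beam charge sound effect.
# A single wave player node feeds the output directly, with no randomization or modulation.
# It uses the CareTaker ability attenuation and the enemy attack sound class, with a boosted volume multiplier.
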SoundCue /Caretaker/Sounds/CareTaker_Beam_Start_Cue
root=WavePlayer_1
nodes [{id=WavePlayer_1, properties {soundWaveAssetPtr=&/Caretaker_Beam_SFX_Charge_01}}]
properties {
 attenuationSettings=/Caretaker/Sounds/Settings/CareTaker_Ability_Att
 maxDistance=11500
 soundClassObject=/Game/Sounds/FortSoundClassesAndModes/Fort_Enemy_Attack
 volumeMultiplier=1.5
}
