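# Sound cue that plays when the Caretaker's beam attack stops.
# It feeds a single beam-off wave through a Modulator at default settings, with boosted cue volume.
# It uses the Caretaker ability attenuation and routes to the enemy attack sound class.
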SoundCue /Caretaker/Sounds/CareTaker_Beam_Stop_Cue
root=Modulator_0
nodes [{id=Modulator_0, children [WavePlayer_1]}, {id=WavePlayer_1, properties {soundWaveAssetPtr=&/Caretaker_Beam_SFX_Off_01}}]
properties {
 attenuationSettings=/Caretaker/Sounds/Settings/CareTaker_Ability_Att
 maxDistance=11500
 soundClassObject=/Game/Sounds/FortSoundClassesAndModes/Fort_Enemy_Attack
 volumeMultiplier=2.25
}
