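# Sound cue for the Caretaker's beam swipe, routed through the Fortnite enemy attack sound class.
# A looping beam wave player feeds an Enveloper whose volume curve fades in from silence, slowly at first and then steeply to full level.
# The cue boosts overall volume and uses the shared Caretaker beam attenuation asset for distance falloff.
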SoundCue /Caretaker/Sounds/CareTaker_Beam_Swipe_Cue
root=Enveloper_1
nodes [
 {
  id=Enveloper_1
  children [WavePlayer_1]
  properties {pitchCurve {editorCurveData {keys [{time=0}]}}, volumeCurve {editorCurveData {keys [<at 0 0>,<at 0.4605 0.071>,<at 1 1>]}}}
 }
 {id=WavePlayer_1, properties {bLooping=true, soundWaveAssetPtr=&/Caretaker_Beam_SFX_Loop_01}}
]
properties {
 attenuationSettings=/Caretaker/Sounds/Settings/CareTaker_Beam_Att
 maxDistance=9000
 soundClassObject=/Game/Sounds/FortSoundClassesAndModes/Fort_Enemy_Attack
 volumeMultiplier=1.5
}
