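# Sound cue for the Caretaker enemy's beam swipe hitting the ground, routed to the enemy attack sound class.
# A mixer layers a one-shot ground impact over a quieter looping ground-impact bed.
# An Enveloper ramps the loop's volume up from silence, so the initial hit leads into the sustained sound.
# The cue uses the shared Caretaker beam attenuation and a boosted overall volume.
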SoundCue /Caretaker/Sounds/CareTaker_Beam_Swipe_Impacts_Cue
root=Mixer_0
nodes [
 {
  id=Enveloper_0
  children [WavePlayer_2]
  properties {pitchCurve {editorCurveData {keys [{time=0}]}}, volumeCurve {editorCurveData {keys [<at 0 0>,<at 1 1>]}}}
 }
 {id=Mixer_0, children [WavePlayer_1,Enveloper_0], properties {inputVolume [1.25,0.7]}}
 {id=WavePlayer_1, properties {soundWaveAssetPtr=&/Caretaker_Beam_SFX_Impt_Ground_01}}
 {id=WavePlayer_2, properties {bLooping=true, soundWaveAssetPtr=&/Caretaker_Beam_SFX_Impt_Ground_Loop_01}}
]
properties {
 attenuationSettings=/Caretaker/Sounds/Settings/CareTaker_Beam_Att
 maxDistance=9000
 soundClassObject=/Game/Sounds/FortSoundClassesAndModes/Fort_Enemy_Attack
 volumeMultiplier=3
}
