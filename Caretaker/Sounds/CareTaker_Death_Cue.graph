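# Sound cue for the Caretaker's death, layering a random death vocal with a fixed death SFX, each passed through its own Modulator into a Mixer.
# A Quality Level node sends the first two levels to a six-vocal pool and the last two to a three-vocal pool excluded from branch culling.
# It uses the Caretaker ability attenuation, the enemy attack sound class and two voice concurrency sets.
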SoundCue /Caretaker/Sounds/CareTaker_Death_Cue
root=QualityLevel_0
nodes [
 {id=Mixer_0, children [Modulator_0,Modulator_1], properties {inputVolume [1,1]}}
 {id=Mixer_1, children [Modulator_2,Modulator_3], properties {inputVolume [1,1]}}
 {id=Modulator_0, children [Random_0]}
 {id=Modulator_1, children [WavePlayer_2]}
 {id=Modulator_2, children [Random_1]}
 {id=Modulator_3, children [WavePlayer_8]}
 {id=QualityLevel_0, children [Mixer_0,Mixer_0,Mixer_1,Mixer_1]}
 {id=Random_0, children [WavePlayer_1,WavePlayer_3,WavePlayer_4,WavePlayer_5,WavePlayer_6,WavePlayer_0], properties {weights [1,1,1,1,1,1]}}
 {id=Random_1, children [WavePlayer_15, WavePlayer_16, WavePlayer_14], properties {bShouldExcludeFromBranchCulling=true, weights [1,1,1]}}
 {id=WavePlayer_0, properties {soundWaveAssetPtr=&/Caretaker_Death_Vox_01}}
 {id=WavePlayer_1, properties {soundWaveAssetPtr=&/Caretaker_Death_Vox_06}}
 {id=WavePlayer_14, properties {soundWaveAssetPtr=&Caretaker_Death_Vox_01}}
 {id=WavePlayer_15, properties {soundWaveAssetPtr=&/Caretaker_Death_Vox_02}}
 {id=WavePlayer_16, properties {soundWaveAssetPtr=&/Caretaker_Death_Vox_05}}
 {id=WavePlayer_2, properties {soundWaveAssetPtr=&/Caretaker_Death_SFx_01}}
 {id=WavePlayer_3, properties {soundWaveAssetPtr=&Caretaker_Death_Vox_02}}
 {id=WavePlayer_4, properties {soundWaveAssetPtr=&/Caretaker_Death_Vox_03}}
 {id=WavePlayer_5, properties {soundWaveAssetPtr=&/Caretaker_Death_Vox_04}}
 {id=WavePlayer_6, properties {soundWaveAssetPtr=&Caretaker_Death_Vox_05}}
 {id=WavePlayer_8, properties {soundWaveAssetPtr=&Caretaker_Death_SFx_01}}
]
properties {
 attenuationSettings=/Caretaker/Sounds/Settings/CareTaker_Ability_Att
 maxDistance=11500
 soundClassObject=/Game/Sounds/FortSoundClassesAndModes/Fort_Enemy_Attack
 volumeMultiplier=2.25
 concurrencySet [/Caretaker/Sounds/Settings/Caretaker_VO_Concurrency, /Caretaker/Sounds/Settings/Caretaker_Group_AggroVO_Concurrency]
}
