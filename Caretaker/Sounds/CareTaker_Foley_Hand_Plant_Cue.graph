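# Sound cue for the Caretaker's hand-plant foley, played as enemy attack audio.
# A Random node picks one of four hand-plant wave variations with equal weight, then a Modulator at the root processes it.
# The cue boosts volume and uses the Caretaker foley attenuation settings and the Fort enemy attack sound class.
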SoundCue /Caretaker/Sounds/CareTaker_Foley_Hand_Plant_Cue
root=Modulator_0
nodes [
 {id=Modulator_0, children [Random_1]}
 {id=Random_1, children [WavePlayer_0,WavePlayer_5,WavePlayer_6,WavePlayer_7], properties {weights [1,1,1,1]}}
 {id=WavePlayer_0, properties {soundWaveAssetPtr=&/Caretaker_Foley_Hand_Plant_01}}
 {id=WavePlayer_5, properties {soundWaveAssetPtr=&/Caretaker_Foley_Hand_Plant_02}}
 {id=WavePlayer_6, properties {soundWaveAssetPtr=&/Caretaker_Foley_Hand_Plant_03}}
 {id=WavePlayer_7, properties {soundWaveAssetPtr=&/Caretaker_Foley_Hand_Plant_04}}
]
properties {
 attenuationSettings=/Caretaker/Sounds/Settings/CareTaker_Foley_Att
 maxDistance=11000
 soundClassObject=/Game/Sounds/FortSoundClassesAndModes/Fort_Enemy_Attack
 volumeMultiplier=2.25
}
