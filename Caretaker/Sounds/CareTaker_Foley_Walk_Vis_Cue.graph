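# Sound cue for the Caretaker enemy's walk foley, using the Caretaker foley attenuation and the enemy attack sound class.
# A quality-level node at the root sends its first two slots to a random pick of ten walk recordings through a default modulator.
# The last two slots use a reduced set of three recordings, with a modulator that sets its own pitch variation range.
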SoundCue /Caretaker/Sounds/CareTaker_Foley_Walk_Vis_Cue
root=QualityLevel_0
nodes [
 {id=Modulator_0, children [Random_0]}
 {id=Modulator_1, children [Random_3], properties {pitchMax=1.1, pitchMin=0.9}}
 {id=QualityLevel_0, children [Modulator_0,Modulator_0,Modulator_1,Modulator_1]}
 {
  id=Random_0
  children [
   WavePlayer_1
   WavePlayer_2
   WavePlayer_3
   WavePlayer_4
   WavePlayer_17
   WavePlayer_18
   WavePlayer_19
   WavePlayer_20
   WavePlayer_21
   WavePlayer_22
  ]
  properties {weights [1,1,1,1,1,1,1,1,1,1]}
 }
 {id=Random_3, children [WavePlayer_23, WavePlayer_24, WavePlayer_25], properties {weights [1,1,1]}}
 {id=WavePlayer_1, properties {soundWaveAssetPtr=&/Caretaker_Foley_Walk_Vis_01}}
 {id=WavePlayer_17, properties {soundWaveAssetPtr=&/Caretaker_Foley_Walk_Vis_05}}
 {id=WavePlayer_18, properties {soundWaveAssetPtr=&/Caretaker_Foley_Walk_Vis_06}}
 {id=WavePlayer_19, properties {soundWaveAssetPtr=&/Caretaker_Foley_Walk_Vis_07}}
 {id=WavePlayer_2, properties {soundWaveAssetPtr=&/Caretaker_Foley_Walk_Vis_02}}
 {id=WavePlayer_20, properties {soundWaveAssetPtr=&/Caretaker_Foley_Walk_Vis_08}}
 {id=WavePlayer_21, properties {soundWaveAssetPtr=&/Caretaker_Foley_Walk_Vis_09}}
 {id=WavePlayer_22, properties {soundWaveAssetPtr=&/Caretaker_Foley_Walk_Vis_10}}
 {id=WavePlayer_23, properties {soundWaveAssetPtr=&Caretaker_Foley_Walk_Vis_01}}
 {id=WavePlayer_24, properties {soundWaveAssetPtr=&Caretaker_Foley_Walk_Vis_02}}
 {id=WavePlayer_25, properties {soundWaveAssetPtr=&/Caretaker_Foley_Walk_Vis_03}}
 {id=WavePlayer_3, properties {soundWaveAssetPtr=&Caretaker_Foley_Walk_Vis_03}}
 {id=WavePlayer_4, properties {soundWaveAssetPtr=&/Caretaker_Foley_Walk_Vis_04}}
]
properties {
 attenuationSettings=/Caretaker/Sounds/Settings/CareTaker_Foley_Att
 maxDistance=11000
 soundClassObject=/Game/Sounds/FortSoundClassesAndModes/Fort_Enemy_Attack
 volumeMultiplier=1.125
}
