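# Sound cue for the Caretaker "Res In" sound, likely an enemy ability effect given its attenuation and sound class.
# A random node picks one of four equally weighted wave variations and feeds a modulator node left at default settings.
# The cue uses the Caretaker ability attenuation, routes through the Fortnite enemy attack sound class, and plays with a boosted volume multiplier.
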SoundCue /Caretaker/Sounds/CareTaker_Res_In_Cue
root=Modulator_0
nodes [
 {id=Modulator_0, children [Random_0]}
 {id=Random_0, children [WavePlayer_1,WavePlayer_2,WavePlayer_3,WavePlayer_4], properties {weights [1,1,1,1]}}
 {id=WavePlayer_1, properties {soundWaveAssetPtr=&/Caretaker_Inv_Res_In_01}}
 {id=WavePlayer_2, properties {soundWaveAssetPtr=&/Caretaker_Inv_Res_In_02}}
 {id=WavePlayer_3, properties {soundWaveAssetPtr=&/Caretaker_Inv_Res_In_03}}
 {id=WavePlayer_4, properties {soundWaveAssetPtr=&/Caretaker_Inv_Res_In_04}}
]
properties {
 attenuationSettings=/Caretaker/Sounds/Settings/CareTaker_Ability_Att
 maxDistance=11500
 soundClassObject=/Game/Sounds/FortSoundClassesAndModes/Fort_Enemy_Attack
 volumeMultiplier=2.25
}
