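# Sound cue for the Caretaker enemy's Sideways pull sound effect.
# A single wave player feeds the pull sample through a modulator at the root, which varies playback.
# It routes through the enemy attack sound class with boosted volume and uses the Caretaker ability attenuation for spatial falloff.
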SoundCue /Caretaker/Sounds/CareTaker_Sideways_Pull_Cue
root=Modulator_0
nodes [
 {id=Modulator_0, children [WavePlayer_0]}
 {id=WavePlayer_0, properties {soundWaveAssetPtr=&/Caretaker_Sideways_SFX_Pull_01}}
]
properties {
 attenuationSettings=/Caretaker/Sounds/Settings/CareTaker_Ability_Att
 maxDistance=11500
 soundClassObject=/Game/Sounds/FortSoundClassesAndModes/Fort_Enemy_Attack
 volumeMultiplier=3
}
