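# Sound cue for the Caretaker's sideways pull, named as the player-side version and routed to the Fort_Enemy_Attack sound class.
# It has no root or sound nodes, so as written it has nothing to play.
# Its attenuation override turns off distance falloff and spatialization for non-positional playback, taking precedence over the referenced beam attenuation asset.
# Volume is slightly reduced.
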SoundCue /Caretaker/Sounds/CareTaker_Sideways_Pull_Player_Cue
root=null
nodes []
properties {
 attenuationSettings=/Caretaker/Sounds/Settings/CareTaker_Beam_Att
 bOverrideAttenuation=true
 maxDistance=2097152
 soundClassObject=/Game/Sounds/FortSoundClassesAndModes/Fort_Enemy_Attack
 volumeMultiplier=0.9
 attenuationOverrides {bAttenuate=false, bSpatialize=false}
}
