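# Sound cue for the swing sound of the Caretaker's Smash attack.
# A random node picks one of four swing wave variations with equal weight.
# It plays through the Caretaker ability attenuation and the enemy attack sound class, with a boosted volume.
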SoundCue /Caretaker/Sounds/CareTaker_Smash_Swing_Cue
root=Random_1
nodes [
 {id=Random_1, children [WavePlayer_1,WavePlayer_5,WavePlayer_6,WavePlayer_7], properties {weights [1,1,1,1]}}
 {id=WavePlayer_1, properties {soundWaveAssetPtr=&/Caretaker_Smash_SFX_Swing_01}}
 {id=WavePlayer_5, properties {soundWaveAssetPtr=&/Caretaker_Smash_SFX_Swing_02}}
 {id=WavePlayer_6, properties {soundWaveAssetPtr=&/Caretaker_Smash_SFX_Swing_03}}
 {id=WavePlayer_7, properties {soundWaveAssetPtr=&/Caretaker_Smash_SFX_Swing_04}}
]
properties {
 attenuationSettings=/Caretaker/Sounds/Settings/CareTaker_Ability_Att
 maxDistance=11500
 soundClassObject=/Game/Sounds/FortSoundClassesAndModes/Fort_Enemy_Attack
 volumeMultiplier=1.5
}
