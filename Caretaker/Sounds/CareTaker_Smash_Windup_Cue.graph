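# SoundCue for the windup sound of the Caretaker's Smash attack.
# A root Random node picks one of four windup wave variations with equal weight.
# The cue uses the Caretaker ability attenuation settings, plays in the enemy attack sound class and has a raised volume multiplier.
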SoundCue /Caretaker/Sounds/CareTaker_Smash_Windup_Cue
root=Random_0
nodes [
 {id=Random_0, children [WavePlayer_0,WavePlayer_2,WavePlayer_3,WavePlayer_4], properties {weights [1,1,1,1]}}
 {id=WavePlayer_0, properties {soundWaveAssetPtr=&/Caretaker_Smash_SFX_Windup_01}}
 {id=WavePlayer_2, properties {soundWaveAssetPtr=&/Caretaker_Smash_SFX_Windup_02}}
 {id=WavePlayer_3, properties {soundWaveAssetPtr=&/Caretaker_Smash_SFX_Windup_03}}
 {id=WavePlayer_4, properties {soundWaveAssetPtr=&/Caretaker_Smash_SFX_Windup_04}}
]
properties {
 attenuationSettings=/Caretaker/Sounds/Settings/CareTaker_Ability_Att
 maxDistance=11500
 soundClassObject=/Game/Sounds/FortSoundClassesAndModes/Fort_Enemy_Attack
 volumeMultiplier=1.5
}
