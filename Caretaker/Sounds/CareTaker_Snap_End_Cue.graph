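# Sound cue for the end of the Caretaker's Snap ability, categorized as an enemy attack sound.
# A single wave player feeds the Snap end SFX wave straight to the output, with no randomization or layering.
# The cue boosts volume, uses the shared Caretaker ability attenuation, and routes through the Fort enemy attack sound class.
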SoundCue /Caretaker/Sounds/CareTaker_Snap_End_Cue
root=WavePlayer_1
nodes [{id=WavePlayer_1, properties {soundWaveAssetPtr=&/Caretaker_Snap_SFX_End_01}}]
properties {
 attenuationSettings=/Caretaker/Sounds/Settings/CareTaker_Ability_Att
 maxDistance=11500
 soundClassObject=/Game/Sounds/FortSoundClassesAndModes/Fort_Enemy_Attack
 volumeMultiplier=3
}
