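# First-person sound cue for the Caretaker snap on the player, pairing an attach hit with a sustained loop.
# A mixer layers the boosted one-shot attach wave over a looping wave that an enveloper fades in along a volume curve, each branch passing through a modulator.
# Overridden attenuation disables distance falloff and spatialization, so it plays non-positionally in the enemy attack sound class with raised overall volume.
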SoundCue /Caretaker/Sounds/CareTaker_Snap_Player_Loop_1p_Cue
root=Mixer_0
nodes [
 {
  id=Enveloper_0
  children [WavePlayer_0]
  properties {pitchCurve {editorCurveData {keys [{time=0}]}}, volumeCurve {editorCurveData {keys [<at 0 0>,<at 0.2497 0.0402>,<at 1 1>]}}}
 }
 {id=Mixer_0, children [Modulator_1,Modulator_0], properties {inputVolume [1.5,1]}}
 {id=Modulator_0, children [Enveloper_0]}
 {id=Modulator_1, children [WavePlayer_1]}
 {id=WavePlayer_0, properties {bLooping=true, soundWaveAssetPtr=&/Caretaker_Snap_SFX_Loop_02}}
 {id=WavePlayer_1, properties {soundWaveAssetPtr=&/Caretaker_Snap_SFX_Attach_01}}
]
properties {
 attenuationSettings=/Caretaker/Sounds/Settings/CareTaker_Beam_Att
 bOverrideAttenuation=true
 maxDistance=2097152
 soundClassObject=/Game/Sounds/FortSoundClassesAndModes/Fort_Enemy_Attack
 volumeMultiplier=3
 attenuationOverrides {bAttenuate=false, bSpatialize=false}
}
