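# Sound cue for the third-person Caretaker snap on a player, on the enemy attack sound class with Caretaker beam attenuation.
# A mixer layers a one-shot attach sound, weighted louder, over a looping snap sound, each through its own modulator.
# An enveloper fades the loop in along a slow-starting volume curve, so the attach sound leads before the sustained layer takes over.
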SoundCue /Caretaker/Sounds/CareTaker_Snap_Player_Loop_3p_Cue
root=Mixer_0
nodes [
 {
  id=Enveloper_0
  children [WavePlayer_1]
  properties {pitchCurve {editorCurveData {keys [{time=0}]}}, volumeCurve {editorCurveData {keys [<at 0 0>,<at 0.2497 0.0402>,<at 1 1>]}}}
 }
 {id=Mixer_0, children [Modulator_2,Modulator_1], properties {inputVolume [1.5,1]}}
 {id=Modulator_1, children [Enveloper_0]}
 {id=Modulator_2, children [WavePlayer_2]}
 {id=WavePlayer_1, properties {bLooping=true, soundWaveAssetPtr=&/Caretaker_Snap_SFX_Loop_02}}
 {id=WavePlayer_2, properties {soundWaveAssetPtr=&/Caretaker_Snap_SFX_Attach_01}}
]
properties {
 attenuationSettings=/Caretaker/Sounds/Settings/CareTaker_Beam_Att
 maxDistance=9000
 soundClassObject=/Game/Sounds/FortSoundClassesAndModes/Fort_Enemy_Attack
 volumeMultiplier=0.9
}
