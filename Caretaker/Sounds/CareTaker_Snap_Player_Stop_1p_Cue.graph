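# Sound cue for the Caretaker snap attack's stop sound, apparently the first-person version heard by the affected player.
# A single detach wave plays through a Modulator node at default settings, with a boosted cue volume in the enemy attack sound class.
# Attenuation overrides turn off distance falloff and spatialization, superseding the referenced beam attenuation asset so the sound plays non-positionally.
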SoundCue /Caretaker/Sounds/CareTaker_Snap_Player_Stop_1p_Cue
root=Modulator_0
nodes [{id=Modulator_0, children [WavePlayer_0]}, {id=WavePlayer_0, properties {soundWaveAssetPtr=&/Caretaker_Snap_SFX_Detach_01}}]
properties {
 attenuationSettings=/Caretaker/Sounds/Settings/CareTaker_Beam_Att
 bOverrideAttenuation=true
 maxDistance=2097152
 soundClassObject=/Game/Sounds/FortSoundClassesAndModes/Fort_Enemy_Attack
 volumeMultiplier=2.25
 attenuationOverrides {bAttenuate=false, bSpatialize=false}
}
