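# Sound cue for the third-person detach stop sound of the Caretaker's snap attack on a player, going by its name.
# A single Wave Player plays the Snap SFX Detach sample through a Modulator node.
# It uses the Caretaker beam attenuation settings and routes to the enemy attack sound class at slightly reduced volume.
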SoundCue /Caretaker/Sounds/CareTaker_Snap_Playey_Stop_3p_Cue
root=Modulator_0
nodes [{id=Modulator_0, children [WavePlayer_0]}, {id=WavePlayer_0, properties {soundWaveAssetPtr=&/Caretaker_Snap_SFX_Detach_01}}]
properties {
 attenuationSettings=/Caretaker/Sounds/Settings/CareTaker_Beam_Att
 maxDistance=9000
 soundClassObject=/Game/Sounds/FortSoundClassesAndModes/Fort_Enemy_Attack
 volumeMultiplier=0.9
}
