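# Sound cue for the start of the Caretaker's snap ability, classed as an enemy attack sound with ability attenuation and a boosted volume.
# A mixer layers the one-shot start wave with the loop wave.
# An Enveloper shapes the loop wave's volume, swelling slowly from quiet to full, holding briefly, then fading to silence by about five and a half seconds.
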SoundCue /Caretaker/Sounds/CareTaker_Snap_Start_Cue
root=Mixer_0
nodes [
 {
  id=Enveloper_0
  children [WavePlayer_2]
  properties {
   pitchCurve {editorCurveData {keys [{time=0}]}}
   volumeCurve {editorCurveData {keys [<at -0.0214 0.074>,<at 0.1779 0.1111>,<at 2.1704 0.2075>,<at 3.5 1>,<at 5 1>,<at 5.5 0>]}}
  }
 }
 {id=Mixer_0, children [WavePlayer_0,Enveloper_0], properties {inputVolume [1,1]}}
 {id=WavePlayer_0, properties {soundWaveAssetPtr=&/Caretaker_Snap_SFX_Start_01}}
 {id=WavePlayer_2, properties {soundWaveAssetPtr=&/Caretaker_Snap_SFX_Loop_01}}
]
properties {
 attenuationSettings=/Caretaker/Sounds/Settings/CareTaker_Ability_Att
 maxDistance=11500
 soundClassObject=/Game/Sounds/FortSoundClassesAndModes/Fort_Enemy_Attack
 volumeMultiplier=2.25
}
