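# Sound cue for the Caretaker enemy's spawn, layering a vocal sound with a sound effect.
# A mixer blends two branches, with the voice layer louder than the effect layer.
# Each branch passes through a modulator into an equally weighted random pick of four wave variations.
# It plays in the enemy attack sound class with the Caretaker ability attenuation and a boosted overall volume.
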SoundCue /Caretaker/Sounds/CareTaker_Spawn_Cue
root=Mixer_0
nodes [
 {id=Mixer_0, children [Modulator_0,Modulator_1], properties {inputVolume [1.25,0.7]}}
 {id=Modulator_0, children [Random_0]}
 {id=Modulator_1, children [Random_1]}
 {id=Random_0, children [WavePlayer_1,WavePlayer_3,WavePlayer_4,WavePlayer_5], properties {weights [1,1,1,1]}}
 {id=Random_1, children [WavePlayer_6,WavePlayer_7,WavePlayer_8,WavePlayer_9], properties {weights [1,1,1,1]}}
 {id=WavePlayer_1, properties {soundWaveAssetPtr=&/Caretaker_Spawn_Vox_01}}
 {id=WavePlayer_3, properties {soundWaveAssetPtr=&/Caretaker_Spawn_Vox_02}}
 {id=WavePlayer_4, properties {soundWaveAssetPtr=&/Caretaker_Spawn_Vox_03}}
 {id=WavePlayer_5, properties {soundWaveAssetPtr=&/Caretaker_Spawn_Vox_04}}
 {id=WavePlayer_6, properties {soundWaveAssetPtr=&/Caretaker_Spawn_SFX_01}}
 {id=WavePlayer_7, properties {soundWaveAssetPtr=&/Caretaker_Spawn_SFX_02}}
 {id=WavePlayer_8, properties {soundWaveAssetPtr=&/Caretaker_Spawn_SFX_03}}
 {id=WavePlayer_9, properties {soundWaveAssetPtr=&/Caretaker_Spawn_SFX_04}}
]
properties {
 attenuationSettings=/Caretaker/Sounds/Settings/CareTaker_Ability_Att
 maxDistance=11500
 soundClassObject=/Game/Sounds/FortSoundClassesAndModes/Fort_Enemy_Attack
 volumeMultiplier=1.5
}
