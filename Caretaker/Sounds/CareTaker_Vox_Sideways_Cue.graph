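# Sound cue for the Caretaker enemy's "Sideways" vocalization, routed to the enemy attack sound class.
# A Random node picks one of three equally weighted Sideways voice waves and feeds it through a Modulator left at default settings.
# Playback uses Caretaker vox attenuation, a boosted volume multiplier and a shared Caretaker voice concurrency setting to limit overlapping lines.
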SoundCue /Caretaker/Sounds/CareTaker_Vox_Sideways_Cue
root=Modulator_0
nodes [
 {id=Modulator_0, children [Random_0]}
 {id=Random_0, children [WavePlayer_1,WavePlayer_2,WavePlayer_3], properties {weights [1,1,1]}}
 {id=WavePlayer_1, properties {soundWaveAssetPtr=&/Caretaker_Sideways_Vox_01}}
 {id=WavePlayer_2, properties {soundWaveAssetPtr=&/Caretaker_Sideways_Vox_02}}
 {id=WavePlayer_3, properties {soundWaveAssetPtr=&/Caretaker_Sideways_Vox_03}}
]
properties {
 attenuationSettings=/Caretaker/Sounds/Settings/CareTaker_Vox_Att
 maxDistance=11500
 soundClassObject=/Game/Sounds/FortSoundClassesAndModes/Fort_Enemy_Attack
 volumeMultiplier=3
 concurrencySet [/Caretaker/Sounds/Settings/Caretaker_VO_Concurrency]
}
