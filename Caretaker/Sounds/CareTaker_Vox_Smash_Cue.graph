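# Sound cue for the Caretaker enemy's vocalization during its smash attack.
# A Random node picks one of four equally weighted smash vox waves and passes it through a Modulator.
# It plays boosted in volume under the Fort enemy attack sound class, with Caretaker vox attenuation and two concurrency sets for Caretaker VO and group aggro VO.
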SoundCue /Caretaker/Sounds/CareTaker_Vox_Smash_Cue
root=Modulator_0
nodes [
 {id=Modulator_0, children [Random_0]}
 {id=Random_0, children [WavePlayer_1,WavePlayer_2,WavePlayer_3,WavePlayer_7], properties {weights [1,1,1,1]}}
 {id=WavePlayer_1, properties {soundWaveAssetPtr=&/Caretaker_Smash_Vox_01}}
 {id=WavePlayer_2, properties {soundWaveAssetPtr=&/Caretaker_Smash_Vox_02}}
 {id=WavePlayer_3, properties {soundWaveAssetPtr=&/Caretaker_Smash_Vox_03}}
 {id=WavePlayer_7, properties {soundWaveAssetPtr=&/Caretaker_Smash_Vox_04}}
]
properties {
 attenuationSettings=/Caretaker/Sounds/Settings/CareTaker_Vox_Att
 maxDistance=11500
 soundClassObject=/Game/Sounds/FortSoundClassesAndModes/Fort_Enemy_Attack
 volumeMultiplier=2.25
 concurrencySet [/Caretaker/Sounds/Settings/Caretaker_VO_Concurrency, /Caretaker/Sounds/Settings/Caretaker_Group_AggroVO_Concurrency]
}
