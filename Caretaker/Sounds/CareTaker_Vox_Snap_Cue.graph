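# Sound cue for the Caretaker's vocal snap, classed as an enemy attack sound.
# A random node picks one of four equally weighted snap vocal recordings, then a modulator node passes it to the output.
# The cue boosts volume and uses shared Caretaker voice attenuation and concurrency settings for long-range spatial playback and limiting overlapping lines.
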SoundCue /Caretaker/Sounds/CareTaker_Vox_Snap_Cue
root=Modulator_0
nodes [
 {id=Modulator_0, children [Random_1]}
 {id=Random_1, children [WavePlayer_0,WavePlayer_4,WavePlayer_5,WavePlayer_6], properties {weights [1,1,1,1]}}
 {id=WavePlayer_0, properties {soundWaveAssetPtr=&/Caretaker_Snap_Vox_01}}
 {id=WavePlayer_4, properties {soundWaveAssetPtr=&/Caretaker_Snap_Vox_02}}
 {id=WavePlayer_5, properties {soundWaveAssetPtr=&/Caretaker_Snap_Vox_03}}
 {id=WavePlayer_6, properties {soundWaveAssetPtr=&/Caretaker_Snap_Vox_04}}
]
properties {
 attenuationSettings=/Caretaker/Sounds/Settings/CareTaker_Vox_Att
 maxDistance=11500
 soundClassObject=/Game/Sounds/FortSoundClassesAndModes/Fort_Enemy_Attack
 volumeMultiplier=2.25
 concurrencySet [/Caretaker/Sounds/Settings/Caretaker_VO_Concurrency]
}
