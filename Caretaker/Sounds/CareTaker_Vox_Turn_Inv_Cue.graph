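# Sound cue for the Caretaker's turn groan vocal (Inv variant), routed to the enemy attack sound class with Caretaker distance attenuation and voice concurrency.
# A quality switch feeds the first two levels a randomized pick of six groans through a default modulator.
# The last two levels pick from only the first three groans, with a custom pitch range and that random node excluded from branch culling.
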SoundCue /Caretaker/Sounds/CareTaker_Vox_Turn_Inv_Cue
root=QualityLevel_0
nodes [
 {id=Modulator_0, children [Random_1]}
 {id=Modulator_1, children [Random_2], properties {pitchMax=1.1, pitchMin=0.9}}
 {id=QualityLevel_0, children [Modulator_0,Modulator_0,Modulator_1,Modulator_1]}
 {id=Random_1, children [WavePlayer_0,WavePlayer_5,WavePlayer_6,WavePlayer_7,WavePlayer_8,WavePlayer_9], properties {weights [1,1,1,1,1,1]}}
 {id=Random_2, children [WavePlayer_10, WavePlayer_11, WavePlayer_12], properties {bShouldExcludeFromBranchCulling=true, weights [1,1,1]}}
 {id=WavePlayer_0, properties {soundWaveAssetPtr=&/Caretaker_Vox_Groan_Turn_Inv_01}}
 {id=WavePlayer_10, properties {soundWaveAssetPtr=&Caretaker_Vox_Groan_Turn_Inv_01}}
 {id=WavePlayer_11, properties {soundWaveAssetPtr=&/Caretaker_Vox_Groan_Turn_Inv_02}}
 {id=WavePlayer_12, properties {soundWaveAssetPtr=&/Caretaker_Vox_Groan_Turn_Inv_03}}
 {id=WavePlayer_5, properties {soundWaveAssetPtr=&Caretaker_Vox_Groan_Turn_Inv_02}}
 {id=WavePlayer_6, properties {soundWaveAssetPtr=&Caretaker_Vox_Groan_Turn_Inv_03}}
 {id=WavePlayer_7, properties {soundWaveAssetPtr=&/Caretaker_Vox_Groan_Turn_Inv_04}}
 {id=WavePlayer_8, properties {soundWaveAssetPtr=&/Caretaker_Vox_Groan_Turn_Inv_05}}
 {id=WavePlayer_9, properties {soundWaveAssetPtr=&/Caretaker_Vox_Groan_Turn_Inv_06}}
]
properties {
 attenuationSettings=/Caretaker/Sounds/Settings/CareTaker_Vox_Dist_Att
 maxDistance=14000
 soundClassObject=/Game/Sounds/FortSoundClassesAndModes/Fort_Enemy_Attack
 volumeMultiplier=1.5
 concurrencySet [/Caretaker/Sounds/Settings/Caretaker_VO_Concurrency]
}
