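# Sound cue for the Caretaker's walk vocals, randomly playing breath and turn-groan waves in the enemy attack sound class.
# A quality-level switch sends the first two levels to a large wave pool and the last two to a smaller subset with wider pitch modulation.
# Both pools repeat some waves despite equal weights, biasing selection.
# Custom distance attenuation applies, and concurrency settings limit overlapping voices.
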SoundCue /Caretaker/Sounds/CareTaker_Vox_Walk_Inv_Cue
root=QualityLevel_0
nodes [
 {id=Modulator_0, children [Random_1]}
 {id=Modulator_1, children [Random_4], properties {pitchMax=1.1, pitchMin=0.9}}
 {id=QualityLevel_0, children [Modulator_0,Modulator_0,Modulator_1,Modulator_1]}
 {
  id=Random_1
  children [
   WavePlayer_1
   WavePlayer_2
   WavePlayer_3
   WavePlayer_4
   WavePlayer_17
   WavePlayer_18
   WavePlayer_19
   WavePlayer_20
   WavePlayer_21
   WavePlayer_22
   WavePlayer_0
   WavePlayer_5
   WavePlayer_6
   WavePlayer_7
   WavePlayer_8
   WavePlayer_9
   WavePlayer_10
   WavePlayer_11
   WavePlayer_12
   WavePlayer_13
   WavePlayer_14
   WavePlayer_15
   WavePlayer_16
   WavePlayer_26
   WavePlayer_27
   WavePlayer_28
  ]
  properties {weights [1,1,1,1,1,1,1,1,1,1,1,1,1,1,1,1,1,1,1,1,1,1,1,1,1,1]}
 }
 {
  id=Random_4
  children [
   WavePlayer_23
   WavePlayer_24
   WavePlayer_25
   WavePlayer_35
   WavePlayer_32
   WavePlayer_33
   WavePlayer_34
   WavePlayer_36
   WavePlayer_37
   WavePlayer_38
   WavePlayer_39
   WavePlayer_40
   WavePlayer_29
   WavePlayer_30
   WavePlayer_31
  ]
  properties {bShouldExcludeFromBranchCulling=true, weights [1,1,1,1,1,1,1,1,1,1,1,1,1,1,1]}
 }
 {id=WavePlayer_0, properties {soundWaveAssetPtr=&/Caretaker_Vox_Breath_Walk_Inv_10}}
 {id=WavePlayer_1, properties {soundWaveAssetPtr=&Caretaker_Vox_Breath_Walk_Inv_10}}
 {id=WavePlayer_10, properties {soundWaveAssetPtr=&/Caretaker_Vox_Breath_Walk_Inv_06}}
 {id=WavePlayer_11, properties {soundWaveAssetPtr=&/Caretaker_Vox_Breath_Walk_Inv_07}}
 {id=WavePlayer_12, properties {soundWaveAssetPtr=&/Caretaker_Vox_Breath_Walk_Inv_08}}
 {id=WavePlayer_13, properties {soundWaveAssetPtr=&/Caretaker_Vox_Breath_Walk_Inv_09}}
 {id=WavePlayer_14, properties {soundWaveAssetPtr=&/Caretaker_Vox_Groan_Turn_Inv_06}}
 {id=WavePlayer_15, properties {soundWaveAssetPtr=&/Caretaker_Vox_Groan_Turn_Inv_01}}
 {id=WavePlayer_16, properties {soundWaveAssetPtr=&/Caretaker_Vox_Groan_Turn_Inv_02}}
 {id=WavePlayer_17, properties {soundWaveAssetPtr=&/Caretaker_Vox_Breath_Walk_Inv_04}}
 {id=WavePlayer_18, properties {soundWaveAssetPtr=&/Caretaker_Vox_Breath_Walk_Inv_05}}
 {id=WavePlayer_19, properties {soundWaveAssetPtr=&Caretaker_Vox_Breath_Walk_Inv_06}}
 {id=WavePlayer_2, properties {soundWaveAssetPtr=&/Caretaker_Vox_Breath_Walk_Inv_01}}
 {id=WavePlayer_20, properties {soundWaveAssetPtr=&Caretaker_Vox_Breath_Walk_Inv_07}}
 {id=WavePlayer_21, properties {soundWaveAssetPtr=&Caretaker_Vox_Breath_Walk_Inv_08}}
 {id=WavePlayer_22, properties {soundWaveAssetPtr=&Caretaker_Vox_Breath_Walk_Inv_09}}
 {id=WavePlayer_23, properties {soundWaveAssetPtr=&Caretaker_Vox_Breath_Walk_Inv_10}}
 {id=WavePlayer_24, properties {soundWaveAssetPtr=&Caretaker_Vox_Breath_Walk_Inv_01}}
 {id=WavePlayer_25, properties {soundWaveAssetPtr=&/Caretaker_Vox_Breath_Walk_Inv_02}}
 {id=WavePlayer_26, properties {soundWaveAssetPtr=&/Caretaker_Vox_Groan_Turn_Inv_03}}
 {id=WavePlayer_27, properties {soundWaveAssetPtr=&/Caretaker_Vox_Groan_Turn_Inv_04}}
 {id=WavePlayer_28, properties {soundWaveAssetPtr=&/Caretaker_Vox_Groan_Turn_Inv_05}}
 {id=WavePlayer_29, properties {soundWaveAssetPtr=&Caretaker_Vox_Groan_Turn_Inv_01}}
 {id=WavePlayer_3, properties {soundWaveAssetPtr=&Caretaker_Vox_Breath_Walk_Inv_02}}
 {id=WavePlayer_30, properties {soundWaveAssetPtr=&Caretaker_Vox_Groan_Turn_Inv_03}}
 {id=WavePlayer_31, properties {soundWaveAssetPtr=&Caretaker_Vox_Groan_Turn_Inv_04}}
 {id=WavePlayer_32, properties {soundWaveAssetPtr=&Caretaker_Vox_Breath_Walk_Inv_10}}
 {id=WavePlayer_33, properties {soundWaveAssetPtr=&Caretaker_Vox_Breath_Walk_Inv_01}}
 {id=WavePlayer_34, properties {soundWaveAssetPtr=&Caretaker_Vox_Breath_Walk_Inv_02}}
 {id=WavePlayer_35, properties {soundWaveAssetPtr=&/Caretaker_Vox_Breath_Walk_Inv_03}}
 {id=WavePlayer_36, properties {soundWaveAssetPtr=&Caretaker_Vox_Breath_Walk_Inv_03}}
 {id=WavePlayer_37, properties {soundWaveAssetPtr=&Caretaker_Vox_Breath_Walk_Inv_10}}
 {id=WavePlayer_38, properties {soundWaveAssetPtr=&Caretaker_Vox_Breath_Walk_Inv_01}}
 {id=WavePlayer_39, properties {soundWaveAssetPtr=&Caretaker_Vox_Breath_Walk_Inv_02}}
 {id=WavePlayer_4, properties {soundWaveAssetPtr=&Caretaker_Vox_Breath_Walk_Inv_03}}
 {id=WavePlayer_40, properties {soundWaveAssetPtr=&Caretaker_Vox_Breath_Walk_Inv_03}}
 {id=WavePlayer_5, properties {soundWaveAssetPtr=&Caretaker_Vox_Breath_Walk_Inv_01}}
 {id=WavePlayer_6, properties {soundWaveAssetPtr=&Caretaker_Vox_Breath_Walk_Inv_02}}
 {id=WavePlayer_7, properties {soundWaveAssetPtr=&Caretaker_Vox_Breath_Walk_Inv_03}}
 {id=WavePlayer_8, properties {soundWaveAssetPtr=&Caretaker_Vox_Breath_Walk_Inv_04}}
 {id=WavePlayer_9, properties {soundWaveAssetPtr=&Caretaker_Vox_Breath_Walk_Inv_05}}
]
properties {
 attenuationSettings=/Caretaker/Sounds/Settings/CareTaker_Vox_Dist_Att
 bOverrideConcurrency=true
 maxDistance=14000
 soundClassObject=/Game/Sounds/FortSoundClassesAndModes/Fort_Enemy_Attack
 volumeMultiplier=1.5
 concurrencyOverrides {maxCount=2, resolutionRule=StopFarthestThenPreventNew}
 concurrencySet [/Caretaker/Sounds/Settings/Caretaker_VO_Concurrency]
}
