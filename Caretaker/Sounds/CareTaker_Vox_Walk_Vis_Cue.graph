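# Caretaker enemy sound cue that plays a randomly chosen walk breathing vocal (Walk_Vis variant).
# A quality-level node sends its first two slots to an equally weighted pool of eighteen waves and its last two to a four-wave subset.
# Each pool feeds its own modulator, with an explicit pitch range set for the smaller one.
# The cue uses Caretaker vocal attenuation and concurrency, under the enemy attack sound class.
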SoundCue /Caretaker/Sounds/CareTaker_Vox_Walk_Vis_Cue
root=QualityLevel_0
nodes [
 {id=Modulator_0, children [Random_1]}
 {id=Modulator_1, children [Random_2], properties {pitchMax=1.1, pitchMin=0.9}}
 {id=QualityLevel_0, children [Modulator_0,Modulator_0,Modulator_1,Modulator_1]}
 {
  id=Random_1
  children [
   WavePlayer_0
   WavePlayer_5
   WavePlayer_6
   WavePlayer_7
   WavePlayer_8
   WavePlayer_9
   WavePlayer_10
   WavePlayer_11
   WavePlayer_12
   WavePlayer_13
   WavePlayer_14
   WavePlayer_15
   WavePlayer_16
   WavePlayer_26
   WavePlayer_27
   WavePlayer_28
   WavePlayer_29
   WavePlayer_30
  ]
  properties {weights [1,1,1,1,1,1,1,1,1,1,1,1,1,1,1,1,1,1]}
 }
 {id=Random_2, children [WavePlayer_31, WavePlayer_32, WavePlayer_33, WavePlayer_34], properties {weights [1,1,1,1]}}
 {id=WavePlayer_0, properties {soundWaveAssetPtr=&/Caretaker_Vox_Breath_Walk_Vis_01}}
 {id=WavePlayer_10, properties {soundWaveAssetPtr=&/Caretaker_Vox_Breath_Walk_Vis_07}}
 {id=WavePlayer_11, properties {soundWaveAssetPtr=&/Caretaker_Vox_Breath_Walk_Vis_08}}
 {id=WavePlayer_12, properties {soundWaveAssetPtr=&/Caretaker_Vox_Breath_Walk_Vis_09}}
 {id=WavePlayer_13, properties {soundWaveAssetPtr=&/Caretaker_Vox_Breath_Walk_Vis_10}}
 {id=WavePlayer_14, properties {soundWaveAssetPtr=&/Caretaker_Vox_Breath_Walk_Vis_12}}
 {id=WavePlayer_15, properties {soundWaveAssetPtr=&/Caretaker_Vox_Breath_Walk_Vis_13}}
 {id=WavePlayer_16, properties {soundWaveAssetPtr=&/Caretaker_Vox_Breath_Walk_Vis_14}}
 {id=WavePlayer_26, properties {soundWaveAssetPtr=&/Caretaker_Vox_Breath_Walk_Vis_15}}
 {id=WavePlayer_27, properties {soundWaveAssetPtr=&/Caretaker_Vox_Breath_Walk_Vis_16}}
 {id=WavePlayer_28, properties {soundWaveAssetPtr=&/Caretaker_Vox_Breath_Walk_Vis_17}}
 {id=WavePlayer_29, properties {soundWaveAssetPtr=&/Caretaker_Vox_Breath_Walk_Vis_18}}
 {id=WavePlayer_30, properties {soundWaveAssetPtr=&/Caretaker_Vox_Breath_Walk_Vis_19}}
 {id=WavePlayer_31, properties {soundWaveAssetPtr=&Caretaker_Vox_Breath_Walk_Vis_01}}
 {id=WavePlayer_32, properties {soundWaveAssetPtr=&/Caretaker_Vox_Breath_Walk_Vis_02}}
 {id=WavePlayer_33, properties {soundWaveAssetPtr=&/Caretaker_Vox_Breath_Walk_Vis_03}}
 {id=WavePlayer_34, properties {soundWaveAssetPtr=&/Caretaker_Vox_Breath_Walk_Vis_04}}
 {id=WavePlayer_5, properties {soundWaveAssetPtr=&Caretaker_Vox_Breath_Walk_Vis_02}}
 {id=WavePlayer_6, properties {soundWaveAssetPtr=&Caretaker_Vox_Breath_Walk_Vis_03}}
 {id=WavePlayer_7, properties {soundWaveAssetPtr=&Caretaker_Vox_Breath_Walk_Vis_04}}
 {id=WavePlayer_8, properties {soundWaveAssetPtr=&/Caretaker_Vox_Breath_Walk_Vis_05}}
 {id=WavePlayer_9, properties {soundWaveAssetPtr=&/Caretaker_Vox_Breath_Walk_Vis_06}}
]
properties {
 attenuationSettings=/Caretaker/Sounds/Settings/CareTaker_Vox_Att
 maxDistance=11500
 soundClassObject=/Game/Sounds/FortSoundClassesAndModes/Fort_Enemy_Attack
 volumeMultiplier=1.125
 concurrencySet [/Caretaker/Sounds/Settings/Caretaker_VO_Concurrency]
}
