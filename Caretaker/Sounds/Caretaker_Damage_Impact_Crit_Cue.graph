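# Sound cue for the Caretaker's critical damage impact, classed as a 2D instigator hit notification sound.
# A random node picks one of four equally weighted crit impact waves, and a modulator applies slight upward pitch variation.
# Attenuation is overridden off so it plays without distance falloff, at reduced volume.
# A concurrency override caps it at three simultaneous instances, stopping the oldest.
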SoundCue /Caretaker/Sounds/Caretaker_Damage_Impact_Crit_Cue
root=Modulator_0
nodes [
 {id=Modulator_0, children [Random_0], properties {pitchMax=1.1, pitchMin=1}}
 {id=Random_0, children [WavePlayer_0,WavePlayer_1,WavePlayer_2,WavePlayer_3], properties {weights [1,1,1,1]}}
 {id=WavePlayer_0, properties {soundWaveAssetPtr=&/Caretaker_Damage_Impact_Crit_01}}
 {id=WavePlayer_1, properties {soundWaveAssetPtr=&/Caretaker_Damage_Impact_Crit_02}}
 {id=WavePlayer_2, properties {soundWaveAssetPtr=&/Caretaker_Damage_Impact_Crit_03}}
 {id=WavePlayer_3, properties {soundWaveAssetPtr=&/Caretaker_Damage_Impact_Crit_04}}
]
properties {
 bOverrideAttenuation=true
 bOverrideConcurrency=true
 maxDistance=2097152
 soundClassObject=&Fort_2DHitNotifications_Instigate_PSM
 volumeMultiplier=0.6
 attenuationOverrides {bAttenuate=false, falloffDistance=10000, attenuationShapeExtents {x=0}}
 concurrencyOverrides {maxCount=3, resolutionRule=StopOldest}
}
