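# Sound cue named for the Caretaker's damage impact while invulnerable, routed to the enemy attack sound class.
# It has no root or sound nodes, so it plays no audio as it stands.
# Its overrides turn off volume attenuation and cap concurrent instances at a small count, stopping the oldest first.
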SoundCue /Caretaker/Sounds/Caretaker_Damage_Impact_Invulnerable_Cue
root=null
nodes []
properties {
 bOverrideAttenuation=true
 bOverrideConcurrency=true
 maxDistance=2097152
 soundClassObject=/Game/Sounds/FortSoundClassesAndModes/Fort_Enemy_Attack
 attenuationOverrides {bAttenuate=false, falloffDistance=10000, attenuationShapeExtents {x=0}}
 concurrencyOverrides {maxCount=3, resolutionRule=StopOldest}
}
