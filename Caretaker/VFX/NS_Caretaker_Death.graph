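# Niagara death effect for the Caretaker, layering imploding embers, energy burst and delayed fizzle, all tracking the system's skeletal mesh.
# GPU embers spawn along bones, then vortex and attraction forces pull them into a sampled mesh point that kills them.
# Arcs, ring, radial-material sprites and meshes form the burst; FizzleGlow and FizzleSpark follow after loop delays.
# Glow emitters stack several renderers through per-renderer size and color attributes, some adding lights.
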NiagaraSystem /Caretaker/VFX/NS_Caretaker_Death
dependencies {
 usedDataInterfaces [
  &NiagaraDataInterfaceColorCurve
  &NiagaraDataInterfaceCurve
  &NiagaraDataInterfaceArrayFloat3
  &NiagaraDataInterfaceParticleRead
  &NiagaraDataInterfaceArrayFloat
  &NiagaraDataInterfaceVectorCurve
  &NiagaraDataInterfaceSkeletalMesh
 ]
 usedDynamicInputs [
  &MakeVector4_FromLinearColor
  &NormalizeFloat
  &VectorLength
  &SetBoolByFloatComparison
 ]
 usedModules [
  &self:SystemSpawnScript.SystemScriptSource.SystemScriptGraph.NiagaraNodeAssignment_0.SetVariables_#1
  &SystemState
  &self:SystemSpawnScript.SystemScriptSource.SystemScriptGraph.NiagaraNodeAssignment_1.SetVariables_#2
  &FN_Integrator
  &FN_Timer
 ]
}
emitters {
 Electric_Arcs {
  emitterUpdateScript [
   &EmitterState(Life Cycle Mode=::NewEnumerator1, Loop Behavior=::NewEnumerator1, Loop Duration=0.5, UseLoopDelay=false)
   &SpawnBurst_Instantaneous(Spawn Count=10, Use Spawn Probability=false, UseLoopCountLimit=false)
  ]
  particleSpawnScript [
   &InitializeParticle(
    Lifetime Max=0.35
    Lifetime Min=0.25
    Lifetime Mode=::NewEnumerator1
    Mesh Scale Mode=::NewEnumerator1
    Mesh Uniform Scale Max=45
    Mesh Uniform Scale Min=30
    Position Offset=<0 0 50>
    Position Offset Coordinate Space=::Local
    UsePositionOffset=true
    Write Material Random=true
   )
   &InitialMeshOrientation(Use Rotation Vector=false)
  ]
  particleUpdateScript [
   &ParticleState()
   &SkeletalMeshLocation(
    MaskSpawnGroup=false
    Mesh Sampling Type=::NewEnumerator5
    Position Sampling=::NewEnumerator1
    Skeletal Mesh=$System.Skeletal Mesh
    Socket Sampling Mode=::NewEnumerator5
   )
   &ScaleMeshSize(Scale Factor=&VectorFromFloat(Value=&FloatFromCurve(FloatCurve=DataInterface() {propertyValues {Curve {keys [{arriveTangent=7.6735, interpMode=RCIM_Cubic, leaveTangent=7.6735, tangentMode=RCTM_User, time=0, value=0.5}, <at 1 3 RCIM_Cubic>]}}})))
   &ScaleColor(
    Scale Alpha=&FloatFromCurve(FloatCurve=DataInterface() {propertyValues {Curve {keys [
         {arriveTangent=9.0725, interpMode=RCIM_Cubic, leaveTangent=9.0725, tangentMode=RCTM_User, time=0, value=0}
         {arriveTangent=-2.5249, interpMode=RCIM_Cubic, leaveTangent=-2.5249, tangentMode=RCTM_User, time=0.1408, value=0.9847}
         {arriveTangent=-0.0405, interpMode=RCIM_Cubic, leaveTangent=-0.0405, tangentMode=RCTM_User, time=1, value=0}
        ]}}})
    Scale RGB=&MakeVectorFromLinearColorRGB(Color=&ColorFromCurve(ColorCurve=DataInterface() {propertyValues {
        AlphaCurve {keys [<at 0 1 RCIM_Cubic>,<at 1 0 RCIM_Cubic>]}
        BlueCurve {keys [<at 0 0 RCIM_Cubic>,<at 1 1 RCIM_Cubic>]}
        GreenCurve {keys [<at 0 23.0949 RCIM_Cubic>,<at 1 0.2012 RCIM_Cubic>]}
        RedCurve {keys [<at 0 25 RCIM_Cubic>,<at 1 0.7609 RCIM_Cubic>]}
       }}))
    ScaleA=true
    ScaleRGB=true
   )
   &DynamicMaterialParameters(
    Index 0 Param 1=&RandomRangeFloat(Minimum=2, Maximum=3)
    Index 0 Param 2=&FloatFromCurve(FloatCurve=DataInterface() {propertyValues {Curve {keys [<at 0 1 RCIM_Cubic>, {arriveTangent=-2.4751, interpMode=RCIM_Cubic, leaveTangent=-2.4751, tangentMode=RCTM_User, time=1, value=-1}]}}})
    Index 0 Param 3=&RandomRangeFloat(Minimum=2, Maximum=3)
    Index 0 Param 4=1
    Module.Param0WriteEnabled=true
    Module.Param1WriteEnabled=true
    Module.Param2WriteEnabled=true
    Module.Param3WriteEnabled=true
   )
  ]
  rendererClasses [&NiagaraMeshRendererProperties]
  renderers [{rendererClass=&NiagaraMeshRendererProperties, rendererIndex=0}]
 }
 Embers {
  simTarget=GPUComputeSim
  emitterUpdateScript [
   &EmitterState()
   &SpawnRate(SpawnRate=1000, Use Spawn Probability=false) {module=SpawnRate001}
  ]
  particleSpawnScript [
   &PartitionParticles(Number Of Partitions=7)
   &InitializeParticle(
    Color=&ColorFromCurve(
     ColorCurve=DataInterface() {propertyValues {
       AlphaCurve {keys [<at 0 1 RCIM_Cubic>]}
       BlueCurve {keys [<at 0 1 RCIM_Cubic>,<at 1 1 RCIM_Cubic>]}
       GreenCurve {keys [<at 0 0.2012 RCIM_Cubic>,<at 1 0 RCIM_Cubic>]}
       RedCurve {keys [<at 0 0.7609 RCIM_Cubic>,<at 1 0.1645 RCIM_Cubic>]}
      }}
     CurveIndex=&RandomRangeFloat(Minimum=0, Maximum=1)
    )
    Lifetime=&RandomRangeFloat(Minimum=1, Maximum=1.5)
    Sprite Size=&Multiply_Vector2D_ByFloat(
     Float=&FloatFromCurve(
      FloatCurve=DataInterface() {propertyValues {Curve {keys [<at 0 1>,<at 0.1 0.9>,<at 0.2 0.65>,<at 0.7 0.5>]}}}
      CurveIndex=&RandomRangeFloat(Minimum=0, Maximum=1)
      Scale Curve=12
     )
     Vector2D=<1 2>
    )
    Sprite Size Mode=::NewEnumerator4
    UsePositionOffset=false
    Write Material Random=true
   )
   &FN_BoneFlow(
    Angle:NiagaraFloat=&RandomRangeFloat(Minimum=-1, Maximum=1)
    BoneID:NiagaraInt32=$Particles.PartitionParticles.Partition
    Height:NiagaraFloat=&RandomRangeFloat(Minimum=0, Maximum=1)
    Radius:NiagaraFloat=&RandomRangeFloat(Minimum=0, Maximum=60)
    Set Material Dynamic=false
    SkeletalMesh:NiagaraDataInterfaceSkeletalMesh=$System.Skeletal Mesh
   )
  ]
  particleUpdateScript [
   &ParticleState()
   &SkeletalMeshLocation(
    MaskSpawnGroup=false
    Mesh Sampling Type=::NewEnumerator5
    Orientation Sampling=::NewEnumerator1
    Skeletal Mesh=$System.Skeletal Mesh
    Socket Sampling Mode=::NewEnumerator5
    Write To Particle Position=false
   ) {module=SkeletalMeshLocation001}
   &ScaleColor(
    Scale Alpha=&FloatFromCurve(FloatCurve=DataInterface() {propertyValues {Curve {keys [<at 0 1 RCIM_Cubic>, {arriveTangent=-2.398, interpMode=RCIM_Cubic, leaveTangent=-2.398, tangentMode=RCTM_User, time=1, value=0}]}}})
    Scale RGB=&VectorFromFloat(Value=&RandomRangeFloat(Minimum=0.5, Maximum=4))
    ScaleA=true
    ScaleRGB=true
   )
   /Niagara/Modules/Update/Velocity/VortexVelocity(
    Influence Falloff Radius:NiagaraFloat=&FloatFromCurve(
     FloatCurve=DataInterface() {propertyValues {Curve {keys [
         {arriveTangent=1.0268, interpMode=RCIM_Cubic, leaveTangent=1.0268, tangentMode=RCTM_User, time=0, value=0.25}
         {interpMode=RCIM_Cubic, tangentMode=RCTM_User, time=0.5, value=1}
        ]}}}
     CurveIndex=$System.Age
     Scale Curve=900
    )
    Module.Use Influence Falloff=true
    Velocity Amount=-700
    Vortex Axis:Vector3f=<0 1 0>
    Vortex Origin:NiagaraPosition=$Particles.SkeletalMeshLocation001.SampledPosition
   )
   &VortexForce(
    Influence Falloff Radius:NiagaraFloat=&FloatFromCurve(
     FloatCurve=DataInterface() {propertyValues {Curve {keys [
         {arriveTangent=1.0268, interpMode=RCIM_Cubic, leaveTangent=1.0268, tangentMode=RCTM_User, time=0, value=0.25}
         {interpMode=RCIM_Cubic, tangentMode=RCTM_User, time=0.5, value=1}
        ]}}}
     CurveIndex=$System.Age
     Scale Curve=2000
    )
    Origin Pull Amount=20000
    Use Influence Falloff=true
    UseOriginOffset=false
    Vortex Axis=<0 1 0>
    Vortex Force Amount=0
    Vortex Origin:NiagaraPosition=$Particles.SkeletalMeshLocation001.SampledPosition
   )
   &PointAttractionForce(
    Attraction Radius=2000
    Attraction Strength=2000
    Attractor Position:NiagaraPosition=$Particles.SkeletalMeshLocation001.SampledPosition
    Kill Within Radius=false
    Module.Use Falloff=true
    UseAttractorOffset=false
   )
   &SolveForcesAndVelocity(Clamp Velocity=false, Limit Acceleration=false)
   &KillParticlesInVolume(Sphere Radius=25, Volume Origin:NiagaraPosition=$Particles.SkeletalMeshLocation001.SampledPosition)
   &ScaleSpriteSizeBySpeed(Max Scale Factor=<1 2>, Velocity Threshold=200)
  ]
  rendererClasses [&NiagaraSpriteRendererProperties]
  renderers [{rendererClass=&NiagaraSpriteRendererProperties}]
 }
 Energy_Blast {
  emitterUpdateScript [
   &EmitterState(Life Cycle Mode=::NewEnumerator1, Loop Behavior=::NewEnumerator1, Loop Duration=0.5, UseLoopDelay=false)
   &SpawnBurst_Instantaneous(Spawn Count=2, Use Spawn Probability=false, UseLoopCountLimit=false)
  ]
  particleSpawnScript [&InitializeParticle(
    Lifetime=0.3
    Sprite Rotation Mode=::NewEnumerator2
    Sprite Size Mode=::NewEnumerator3
    Uniform Sprite Size=&RandomRangeFloat(Minimum=500, Maximum=650)
    UsePositionOffset=false
    Write Material Random=true
   )]
  particleUpdateScript [
   &ParticleState()
   &SkeletalMeshLocation(
    MaskSpawnGroup=false
    Mesh Sampling Type=::NewEnumerator5
    Orientation Sampling=::NewEnumerator1
    Position Sampling=::NewEnumerator1
    Skeletal Mesh=$System.Skeletal Mesh
    Socket Sampling Mode=::NewEnumerator5
    Use Sampled Mesh Attribute as Ribbon ID=false
   ) {module=SkeletalMeshLocation001}
   &FN_Radial_Material(
    Base Sprite Tile V:NiagaraFloat=&RandomRangeFloat(Minimum=2, Maximum=3.5)
    Radial Mask Power=3
    Sprite Extrusion Speed=1.5
   )
   &ScaleColor(
    Linear Color Curve=DataInterface() {propertyValues {
      AlphaCurve {keys [<at 0.0093 0>,<at 0.057 1 RCIM_Cubic>,<at 0.9897 0>]}
      BlueCurve {keys [<at 0.0361 20 RCIM_Cubic>,<at 0.674 3 RCIM_Cubic>]}
      GreenCurve {keys [<at 0.0361 0.4535 RCIM_Cubic>,<at 0.674 0.6035 RCIM_Cubic>]}
      RedCurve {keys [<at 0.0361 14.2364 RCIM_Cubic>,<at 0.674 2.2828 RCIM_Cubic>]}
     }}
    Scale Mode=::NewEnumerator2
   )
   &ScaleSpriteSize(Uniform Curve Sprite Scale=DataInterface() {propertyValues {Curve {keys [
        {arriveTangent=1.0327, interpMode=RCIM_Cubic, leaveTangent=1.0327, tangentMode=RCTM_User, time=0, value=0}
        {interpMode=RCIM_Cubic, tangentMode=RCTM_User, time=1, value=1}
       ]}}})
  ]
  rendererClasses [&NiagaraSpriteRendererProperties]
  renderers [{rendererClass=&NiagaraSpriteRendererProperties}]
 }
 Energy_ring {
  emitterUpdateScript [
   &EmitterState(Life Cycle Mode=::NewEnumerator1, Loop Behavior=::NewEnumerator1, Loop Delay=0.1, Loop Duration=0.5, UseLoopDelay=false)
   &SpawnBurst_Instantaneous(Spawn Count=1, Use Spawn Probability=false)
  ]
  particleSpawnScript [
   &InitializeParticle(Color=<rgba 1 1 1 2>, Lifetime=0.25, Mesh Scale Mode=::NewEnumerator3, UsePositionOffset=false, Write Material Random=true) {module=InitializeParticle001}
   &InitialMeshOrientation(Orientation Vector=<0 0 1>, Use Orientation Vector=true, Use Rotation Vector=true) {module=InitialMeshOrientation001}
  ]
  particleUpdateScript [
   &ParticleState()
   &SkeletalMeshLocation(
    MaskSpawnGroup=false
    Mesh Sampling Type=::NewEnumerator5
    Position Sampling=::NewEnumerator1
    Skeletal Mesh=$System.Skeletal Mesh
    Socket Sampling Mode=::NewEnumerator5
   )
   &ScaleColor(
    Scale Alpha=&Multiply_Float(
     A=&FloatFromCurve(FloatCurve=DataInterface() {propertyValues {Curve {keys [
          {arriveTangent=-0.9, interpMode=RCIM_Cubic, time=0, value=0}
          {arriveTangent=0.8333, interpMode=RCIM_Cubic, leaveTangent=0.8333, tangentMode=RCTM_Break, time=0.1, value=1}
          {arriveTangent=-1.6178, interpMode=RCIM_Cubic, leaveTangent=-1.6178, tangentMode=RCTM_User, time=0.6, value=0.8}
          {arriveTangent=-0.5181, interpMode=RCIM_Cubic, leaveTangent=-0.5181, tangentMode=RCTM_User, time=1, value=0}
         ]}}})
     B=&FNTimeOfDayScalar(Night Scale=1, Morning Scale=0.5, Evening Scale=0.5)
    )
    Scale RGB=&MakeVectorFromLinearColorRGB(Color=&ColorFromCurve(ColorCurve=DataInterface() {propertyValues {
        AlphaCurve {keys []}
        BlueCurve {keys [<at 0.0363 3 RCIM_Cubic>,<at 0.4322 0 RCIM_Cubic>,<at 0.9296 3>]}
        GreenCurve {keys [<at 0.0363 0.6035 RCIM_Cubic>,<at 0.4322 5.9684 RCIM_Cubic>,<at 0.9296 0.6035>]}
        RedCurve {keys [<at 0.0363 2.2828 RCIM_Cubic>,<at 0.4322 7 RCIM_Cubic>,<at 0.9296 2.2828>]}
       }}))
    ScaleA=true
    ScaleRGB=true
   )
   &ScaleMeshSize(Scale Factor=&ScaleFloatByCurve(
     InitialValue:Vector3f=<1 1 1>
     ScaleCurve:NiagaraDataInterfaceVectorCurve=DataInterface() {propertyValues {XCurve {keys [<at 0 0.25>,<at 1 1>]}, YCurve {keys [<at 0 1.25>,<at 1 2.25>]}, ZCurve {keys [<at 0 1.25>,<at 1 2.25>]}}}
     CurveIndex:NiagaraFloat=$Particles.NormalizedAge
    ))
  ]
  rendererClasses [&NiagaraMeshRendererProperties]
  renderers [{rendererClass=&NiagaraMeshRendererProperties, rendererIndex=0}]
 }
 FN_RadialSprite {
  emitterUpdateScript [
   &EmitterState()
   &SpawnRate(SpawnRate=5, Use Spawn Probability=false)
  ]
  particleSpawnScript [&InitializeParticle(
    Lifetime=0.6
    Sprite Rotation Mode=::NewEnumerator2
    Sprite Size Mode=::NewEnumerator3
    Uniform Sprite Size=1000
    UsePositionOffset=false
    Write Material Random=true
   )]
  particleUpdateScript [
   &ParticleState()
   &SkeletalMeshLocation(
    MaskSpawnGroup=false
    Mesh Sampling Type=::NewEnumerator5
    Orientation Sampling=::NewEnumerator1
    Position Sampling=::NewEnumerator1
    Skeletal Mesh=$System.Skeletal Mesh
    Socket Sampling Mode=::NewEnumerator5
    Use Sampled Mesh Attribute as Ribbon ID=false
   ) {module=SkeletalMeshLocation001}
   &FN_Radial_Material(Base Sprite Tile V=3, Radial Mask Power=3, Sprite Extrusion Speed=-1)
   &ScaleColor(
    Linear Color Curve=DataInterface() {propertyValues {
      AlphaCurve {keys [<at 0.0093 0>,<at 0.8269 1 RCIM_Cubic>,<at 0.9897 0>]}
      BlueCurve {keys [<at 0.3901 10 RCIM_Cubic>,<at 0.9639 0 RCIM_Cubic>]}
      GreenCurve {keys [<at 0.3901 0.1201 RCIM_Cubic>,<at 0.9639 9.1969 RCIM_Cubic>]}
      RedCurve {keys [<at 0.3901 0 RCIM_Cubic>,<at 0.9639 10 RCIM_Cubic>]}
     }}
    Scale Mode=::NewEnumerator2
   )
   &ScaleSpriteSize(Uniform Curve Sprite Scale=DataInterface() {propertyValues {Curve {keys [{interpMode=RCIM_Cubic, tangentMode=RCTM_User, time=0, value=1}, {interpMode=RCIM_Cubic, tangentMode=RCTM_User, time=1, value=0}]}}})
  ]
  rendererClasses [&NiagaraSpriteRendererProperties]
  renderers [{rendererClass=&NiagaraSpriteRendererProperties}]
 }
 FN_RadialSprite001 {
  emitterUpdateScript [
   &EmitterState(Life Cycle Mode=::NewEnumerator1, Loop Behavior=::NewEnumerator1, Loop Duration=4, UseLoopDelay=false)
   &SpawnBurst_Instantaneous(Spawn Count=5, Use Spawn Probability=false, UseLoopCountLimit=false)
   &SpawnRate(
    SpawnRate=&FloatFromCurve(
     FloatCurve=DataInterface() {propertyValues {Curve {keys [
         <at 0 1 RCIM_Cubic>
         {arriveTangent=-0.3333, interpMode=RCIM_Cubic, leaveTangent=-0.3333, time=0.5, value=0.15}
         {arriveTangent=-0.3, interpMode=RCIM_Cubic, leaveTangent=-0.3, time=0.6, value=0.8}
         <at 1 0 RCIM_Cubic>
        ]}}}
     CurveIndex=$Emitter.NormalizedLoopAge
     Scale Curve=15
    )
    Use Spawn Probability=false
   )
  ]
  particleSpawnScript [
   &InitializeParticle(
    Lifetime Max=0.55
    Lifetime Min=0.35
    Lifetime Mode=::NewEnumerator1
    Mesh Scale Mode=::NewEnumerator3
    Mesh Uniform Scale=&RandomRangeFloat(Minimum=300, Maximum=700)
    Sprite Rotation Mode=::NewEnumerator2
    UsePositionOffset=false
    Write Material Random=true
   )
   &InitialMeshOrientation(
    Mesh Axis To Orient=<0 0 1>
    Mesh Orientation Mode=::NewEnumerator2
    Orientation Vector=<0 1 0>
    Rotation=&RandomRangeVector(Minimum=<-0.1 -0.1 -1>, Maximum=<0.1 0.1 1>)
    Use Rotation Vector=true
   )
  ]
  particleUpdateScript [
   &ParticleState()
   &SkeletalMeshLocation(
    MaskSpawnGroup=false
    Mesh Sampling Type=::NewEnumerator5
    Position Sampling=::NewEnumerator1
    Skeletal Mesh=$System.Skeletal Mesh
    Socket Sampling Mode=::NewEnumerator5
    Use Sampled Mesh Attribute as Ribbon ID=false
   ) {module=SkeletalMeshLocation001}
   &SolveForcesAndVelocity(Clamp Velocity=false, Limit Acceleration=false)
   &FN_Radial_Material(
    Base Sprite Offset X:NiagaraFloat=&RandomRangeFloat(Minimum=0, Maximum=1)
    Base Sprite Offset Y:NiagaraFloat=&RandomRangeFloat(Minimum=0, Maximum=1)
    Base Sprite Tile V=3
    Sprite Extrusion Speed=1
   )
   &ScaleColor(
    Linear Color Curve=DataInterface() {propertyValues {
      AlphaCurve {keys [<at 0.005 1 RCIM_Cubic>,<at 0.9897 0>]}
      BlueCurve {keys [<at 0.0087 0 RCIM_Cubic>,<at 0.6914 10 RCIM_Cubic>]}
      GreenCurve {keys [<at 0.0087 44.3592 RCIM_Cubic>,<at 0.6914 0 RCIM_Cubic>]}
      RedCurve {keys [<at 0.0087 50 RCIM_Cubic>,<at 0.6914 9.5634 RCIM_Cubic>]}
     }}
    Scale Mode=::NewEnumerator2
   )
   &ScaleMeshSize(Scale Factor=&VectorFromCurve(VectorCurve=DataInterface() {propertyValues {
       XCurve {keys [{arriveTangent=1.5555, interpMode=RCIM_Cubic, leaveTangent=1.5555, tangentMode=RCTM_User, time=0, value=0}, <at 1 1 RCIM_Cubic>]}
       YCurve {keys [{arriveTangent=1.5555, interpMode=RCIM_Cubic, leaveTangent=1.5555, tangentMode=RCTM_User, time=0, value=0}, <at 1 1 RCIM_Cubic>]}
       ZCurve {keys [{arriveTangent=1.5555, interpMode=RCIM_Cubic, leaveTangent=1.5555, tangentMode=RCTM_User, time=0, value=0}, <at 1 1.25 RCIM_Cubic>]}
      }}))
  ]
  rendererClasses [&NiagaraMeshRendererProperties]
  renderers [{rendererClass=&NiagaraMeshRendererProperties, rendererIndex=0}]
 }
 FizzleGlow {
  emitterUpdateScript [
   &EmitterState(Life Cycle Mode=::NewEnumerator1, Loop Behavior=::NewEnumerator1, Loop Delay=4, Loop Duration=1, UseLoopDelay=true)
   &SpawnBurst_Instantaneous(Use Spawn Probability=false, UseLoopCountLimit=false)
  ]
  particleSpawnScript [&InitializeParticle(Color=<rgba 10 10 10 1>, Lifetime=2.3, Sprite Size Mode=::NewEnumerator3, UsePositionOffset=false, Write Material Random=true)]
  particleUpdateScript [
   &ParticleState()
   &SkeletalMeshLocation(
    MaskSpawnGroup=false
    Mesh Sampling Type=::NewEnumerator5
    Position Sampling=::NewEnumerator1
    Skeletal Mesh=$System.Skeletal Mesh
    Socket Sampling Mode=::NewEnumerator5
   ) {module=SkeletalMeshLocation001}
   &ScaleColor(
    Scale Alpha=&FloatFromCurve(
     FloatCurve=DataInterface() {propertyValues {Curve {keys [
         {arriveTangent=2.7855, interpMode=RCIM_Cubic, leaveTangent=2.7855, tangentMode=RCTM_User, time=0, value=0}
         <at 0.1 0.95>
         {arriveTangent=-1.0556, interpMode=RCIM_Cubic, leaveTangent=-1.0556, time=0.8, value=1}
         {arriveTangent=-4.5812, interpMode=RCIM_Cubic, leaveTangent=-4.5812, tangentMode=RCTM_User, time=1, value=0}
        ]}}}
     Scale Curve=&FloatFromCurve(FloatCurve=DataInterface() {propertyValues {Curve {keys [{tangentMode=RCTM_User, time=0, value=1}, <at 0.2 0.5>, <at 0.4 1>, <at 0.7 0.5>, <at 0.9 1>, {tangentMode=RCTM_User, time=1, value=0}]}}})
    )
    Scale RGB=&MakeVectorFromLinearColorRGB(Color=&ColorFromCurve(ColorCurve=DataInterface() {propertyValues {
        AlphaCurve {keys [<at 0 1 RCIM_Cubic>]}
        BlueCurve {keys [<at 0.0016 1 RCIM_Cubic>,<at 0.7893 0>,<at 1 0 RCIM_Cubic>]}
        GreenCurve {keys [<at 0.0016 0 RCIM_Cubic>,<at 0.7893 74.2738>,<at 1 0.2756 RCIM_Cubic>]}
        RedCurve {keys [<at 0.0016 0.5518 RCIM_Cubic>,<at 0.7893 100>,<at 1 1 RCIM_Cubic>]}
       }}))
    ScaleA=true
    ScaleRGB=true
   )
   &CameraOffset(Camera Offset Amount=50)
   &ScaleSpriteSize(Uniform Curve Sprite Scale=DataInterface() {propertyValues {Curve {keys [
        {tangentMode=RCTM_User, time=0, value=0}
        <at 0.2 1>
        <at 0.4 0.7>
        <at 0.6 1>
        <at 0.7 0.625>
        <at 0.9 1>
        {tangentMode=RCTM_User, time=1, value=0}
       ]}}})
   /SetVariables(Particles.Renderer2Size:Vector2f=&Multiply_Vector2D_ByFloat(Float=20, Vector2D=$Particles.SpriteSize)) {isSetParameters=true, module=SetVariables_#3}
   /SetVariables(Particles.Renderer2Color:LinearColor=&MultiplyLinearColor(
     Linear Color=&ColorFromCurve(ColorCurve=DataInterface() {propertyValues {
        AlphaCurve {keys [
          {arriveTangent=3.7993, interpMode=RCIM_Cubic, time=0, value=1}
          <at 0.1 0.5 RCIM_Cubic>
          <at 0.4 1 RCIM_Cubic>
          <at 0.7 0.5 RCIM_Cubic>
          {arriveTangent=-1.7348, interpMode=RCIM_Cubic, leaveTangent=-1.7348, time=0.9, value=1}
          {interpMode=RCIM_Cubic, leaveTangent=-10.4123, time=0.9882, value=0}
         ]}
        BlueCurve {keys [<at 0.0016 1 RCIM_Cubic>,<at 0.3956 0>,<at 0.9958 1>]}
        GreenCurve {keys [<at 0.0016 0 RCIM_Cubic>,<at 0.3956 6.212>,<at 0.9958 0>]}
        RedCurve {keys [<at 0.0016 0.5518 RCIM_Cubic>,<at 0.3956 10>,<at 0.9958 0.5518>]}
       }})
     Scale Factor=<1 1 1 0.6>
    )) {isSetParameters=true, module=SetVariables_#4}
   /SetVariables(Particles.Renderer3Size:Vector2f=&Vector2DFromFloat(Value=&FloatFromCurve(
      FloatCurve=DataInterface() {propertyValues {Curve {keys [
          {tangentMode=RCTM_User, time=0, value=0}
          <at 0.2 0.35>
          <at 0.4 0.2>
          <at 0.6 0.7>
          <at 0.7 0.25>
          <at 0.9 1>
          {tangentMode=RCTM_User, time=1, value=1.15}
         ]}}}
      Scale Curve=190
     ))) {isSetParameters=true}
   /SetVariables(Particles.Renderer3Color:LinearColor=&MultiplyLinearColor(
     Linear Color=&ColorFromCurve(ColorCurve=DataInterface() {propertyValues {
        AlphaCurve {keys [
          <at 0.0059 1>
          {arriveTangentWeight=0.0314, interpMode=RCIM_Cubic, leaveTangentWeight=0.2667, tangentMode=RCTM_User, time=0.1, value=0.5}
          {arriveTangentWeight=0.1, interpMode=RCIM_Cubic, leaveTangentWeight=0.1667, tangentMode=RCTM_User, time=0.4, value=1}
          {arriveTangentWeight=0.1, interpMode=RCIM_Cubic, leaveTangentWeight=0.0667, tangentMode=RCTM_User, time=0.7, value=0.5}
          {
           arriveTangent=-1.7348
           arriveTangentWeight=0.298
           interpMode=RCIM_Cubic
           leaveTangent=-1.7348
           leaveTangentWeight=0.0333
           tangentMode=RCTM_User
           time=0.9
           value=1
          }
          <at 1 0 RCIM_Cubic>
         ]}
        BlueCurve {keys [<at 0.0096 0>,<at 0.9912 1>]}
        GreenCurve {keys [<at 0.0096 1.1877>,<at 0.9912 0>]}
        RedCurve {keys [<at 0.0096 5>,<at 0.9912 0.5518>]}
       }})
     Scale Factor=<1 1 1 1>
    )) {isSetParameters=true}
   /SetVariables(Particles.Renderer4Size:Vector2f=&Multiply_Vector2D(A=$Particles.SpriteSize, B=<50 4>)) {isSetParameters=true}
   /SetVariables(Particles.Renderer4Color:LinearColor=&MultiplyLinearColor(
     Linear Color=&ColorFromCurve(ColorCurve=DataInterface() {propertyValues {
        AlphaCurve {keys [
          {arriveTangent=3.7993, interpMode=RCIM_Cubic, time=0, value=1}
          <at 0.1 0.5 RCIM_Cubic>
          <at 0.4 1 RCIM_Cubic>
          <at 0.7 0.5 RCIM_Cubic>
          {arriveTangent=-1.7348, interpMode=RCIM_Cubic, leaveTangent=-1.7348, time=0.9, value=1}
          {interpMode=RCIM_Cubic, leaveTangent=-10.4123, time=0.9882, value=0}
         ]}
        BlueCurve {keys [<at 0.0016 1 RCIM_Cubic>,<at 0.4048 5>,<at 0.9866 0.0886>]}
        GreenCurve {keys [<at 0.0016 0 RCIM_Cubic>,<at 0.4048 0>,<at 0.9866 0>]}
        RedCurve {keys [<at 0.0016 0.5518 RCIM_Cubic>,<at 0.4048 1.132>,<at 0.9866 1>]}
       }})
     Scale Factor=<1 1 1 1>
    )) {isSetParameters=true}
   /SetVariables(Particles.Renderer5Size:Vector2f=&Vector2DFromFloat(Value=&FloatFromCurve(
      FloatCurve=DataInterface() {propertyValues {Curve {keys [
          {tangentMode=RCTM_User, time=0, value=0}
          <at 0.2 0.35>
          <at 0.4 0.2>
          <at 0.6 0.7>
          <at 0.7 0.25>
          <at 0.9 1>
          {tangentMode=RCTM_User, time=1, value=1.15}
         ]}}}
      Scale Curve=600
     ))) {isSetParameters=true}
   /SetVariables(Particles.Renderer5Color:LinearColor=&ColorFromCurve(ColorCurve=DataInterface() {propertyValues {
       AlphaCurve {keys [
         {arriveTangent=3.7993, interpMode=RCIM_Cubic, leaveTangent=3.7993, tangentMode=RCTM_User, time=0, value=0}
         <at 0.2608 1>
         {arriveTangent=-1.3748, interpMode=RCIM_Cubic, leaveTangent=-1.3748, time=0.7907, value=1}
         {arriveTangent=-10.4123, interpMode=RCIM_Cubic, leaveTangent=-10.4123, tangentMode=RCTM_User, time=0.9882, value=0}
        ]}
       BlueCurve {keys [<at 0.2653 0.2344>]}
       GreenCurve {keys [<at 0.2653 0.0178>]}
       RedCurve {keys [<at 0.2653 0>]}
      }})) {isSetParameters=true}
  ]
  rendererClasses [&NiagaraSpriteRendererProperties]
  renderers [
   {rendererClass=&NiagaraSpriteRendererProperties}
   {rendererClass=&NiagaraSpriteRendererProperties, rendererIndex=1}
   {rendererClass=&NiagaraSpriteRendererProperties, rendererIndex=2}
   {rendererClass=&NiagaraSpriteRendererProperties, rendererIndex=3}
   {rendererClass=&NiagaraSpriteRendererProperties, rendererIndex=4}
  ]
 }
 FizzleSpark {
  emitterUpdateScript [
   &EmitterState(Life Cycle Mode=::NewEnumerator1, Loop Behavior=::NewEnumerator1, Loop Delay=6, Loop Duration=1, UseLoopDelay=true)
   &SpawnBurst_Instantaneous(Spawn Count=7, Use Spawn Probability=false, UseLoopCountLimit=false)
  ]
  particleSpawnScript [
   &InitializeParticle(
    Color=&ColorFromCurve(
     ColorCurve=DataInterface() {propertyValues {
       AlphaCurve {keys [<at 0 1 RCIM_Cubic>]}
       BlueCurve {keys [<at 0 0 RCIM_Cubic>,<at 1 10 RCIM_Cubic>]}
       GreenCurve {keys [<at 0 41.4335 RCIM_Cubic>,<at 1 2.0117 RCIM_Cubic>]}
       RedCurve {keys [<at 0 50 RCIM_Cubic>,<at 1 7.6092 RCIM_Cubic>]}
      }}
     CurveIndex=&RandomRangeFloat(Minimum=0, Maximum=1)
    )
    Lifetime Max=0.8
    Lifetime Min=0.3
    Lifetime Mode=::NewEnumerator1
    Sprite Size=&Multiply_Vector2D_ByFloat(
     Float=&FloatFromCurve(
      FloatCurve=DataInterface() {propertyValues {Curve {keys [<at 0 1>,<at 0.1 0.9>,<at 0.2 0.65>,<at 0.7 0.5>]}}}
      CurveIndex=&RandomRangeFloat(Minimum=0, Maximum=1)
      Scale Curve=30
     )
     Vector2D=<1 2>
    )
    Sprite Size Mode=::NewEnumerator4
    UsePositionOffset=false
    Write Material Random=true
   )
   &AddVelocityFromPoint(Module.Use Falloff Distance=false, Velocity Strength=&RandomRangeFloat(Minimum=150, Maximum=300))
  ]
  particleUpdateScript [
   &ParticleState()
   &SkeletalMeshLocation(
    MaskSpawnGroup=false
    Mesh Sampling Type=::NewEnumerator5
    Position Sampling=::NewEnumerator1
    Skeletal Mesh=$System.Skeletal Mesh
    Socket Sampling Mode=::NewEnumerator5
   ) {module=SkeletalMeshLocation001}
   &VectorNoiseForce(Force Amount=100, Module.Use Random Noise Time Interval=false)
   &SolveForcesAndVelocity(Clamp Velocity=false, Limit Acceleration=false)
   &ScaleSpriteSize(Uniform Curve Sprite Scale=DataInterface() {propertyValues {Curve {keys [{interpMode=RCIM_Cubic, tangentMode=RCTM_User, time=0, value=1}, {interpMode=RCIM_Cubic, tangentMode=RCTM_User, time=1, value=0}]}}})
  ]
  rendererClasses [&NiagaraSpriteRendererProperties]
  renderers [{rendererClass=&NiagaraSpriteRendererProperties}]
 }
 Glow {
  emitterUpdateScript [
   &EmitterState()
   &SpawnBurst_Instantaneous(Use Spawn Probability=false, UseLoopCountLimit=false)
  ]
  particleSpawnScript [&InitializeParticle(Lifetime=1.2, Sprite Size Mode=::NewEnumerator3, Uniform Sprite Size=50, UsePositionOffset=false, Write Material Random=true)]
  particleUpdateScript [
   &ParticleState()
   &SkeletalMeshLocation(
    MaskSpawnGroup=false
    Mesh Sampling Type=::NewEnumerator5
    Position Sampling=::NewEnumerator1
    Skeletal Mesh=$System.Skeletal Mesh
    Socket Sampling Mode=::NewEnumerator5
   ) {module=SkeletalMeshLocation001}
   &ScaleColor(
    Scale Alpha=&FloatFromCurve(FloatCurve=DataInterface() {propertyValues {Curve {keys [
         {arriveTangent=2.7855, interpMode=RCIM_Cubic, leaveTangent=2.7855, tangentMode=RCTM_User, time=0, value=0}
         <at 0.8 1 RCIM_Cubic>
         {arriveTangent=-4.5812, interpMode=RCIM_Cubic, leaveTangent=-4.5812, tangentMode=RCTM_User, time=1, value=0}
        ]}}})
    Scale RGB=&MakeVectorFromLinearColorRGB(Color=&ColorFromCurve(ColorCurve=DataInterface() {propertyValues {
        AlphaCurve {keys [<at 0 1 RCIM_Cubic>]}
        BlueCurve {keys [<at 0.0016 1 RCIM_Cubic>,<at 0.7893 0>,<at 1 0.0886 RCIM_Cubic>]}
        GreenCurve {keys [<at 0.0016 0 RCIM_Cubic>,<at 0.7893 87.3327>,<at 1 0 RCIM_Cubic>]}
        RedCurve {keys [<at 0.0016 0.5518 RCIM_Cubic>,<at 0.7893 100>,<at 1 1 RCIM_Cubic>]}
       }}))
    ScaleA=true
    ScaleRGB=true
   )
   &CameraOffset(Camera Offset Amount=50)
   &FN_Light_Attributes(
    Color=&ColorFromCurve(ColorCurve=DataInterface() {propertyValues {
       AlphaCurve {keys [
         {arriveTangent=3.7392, interpMode=RCIM_Cubic, leaveTangent=3.7392, tangentMode=RCTM_User, time=0, value=0}
         <at 0.7885 1 RCIM_Cubic>
         {arriveTangent=-7.8769, interpMode=RCIM_Cubic, leaveTangent=-7.8769, tangentMode=RCTM_User, time=0.9962, value=0}
        ]}
       BlueCurve {keys [
         <at 0.0032 0>
         {arriveTangent=1.1275, interpMode=RCIM_Cubic, leaveTangent=1.1275, time=0.2109, value=3}
         {arriveTangent=-3.8019, interpMode=RCIM_Cubic, leaveTangent=-3.8019, time=0.7893, value=0.8863}
         {arriveTangent=-36.6349, interpMode=RCIM_Cubic, leaveTangent=-36.6349, tangentMode=RCTM_User, time=1, value=0}
        ]}
       GreenCurve {keys [
         <at 0.0032 0>
         <at 0.2109 0 RCIM_Cubic>
         <at 0.7893 0>
         {arriveTangent=-34.8254, interpMode=RCIM_Cubic, leaveTangent=-34.8254, tangentMode=RCTM_User, time=1, value=0}
        ]}
       RedCurve {keys [
         <at 0.0032 0>
         {arriveTangent=12.7217, interpMode=RCIM_Cubic, leaveTangent=12.7217, time=0.2109, value=1.0808}
         {arriveTangent=-1.3697, interpMode=RCIM_Cubic, leaveTangent=-1.3697, time=0.7893, value=10}
         {arriveTangent=-32.9241, interpMode=RCIM_Cubic, leaveTangent=-32.9241, tangentMode=RCTM_User, time=1, value=0}
        ]}
      }})
    Create Color Attribute=true
    Create Exponent Attribute=true
    Create Position Attribute=true
    Create Radius Attribute=true
    Exponent=5
    Position=&Add_Vector(A=Linked() {linkedVariable {name=Particles.Position, type=NiagaraPosition}}, B=<0 50 0>)
    Radius=250
    Scale Color=25
   )
   /SetVariables(Particles.Renderer2Size:Vector2f=&Vector2DFromFloat(Value=&FloatFromCurve(
      FloatCurve=DataInterface() {propertyValues {Curve {keys [
          <at 0 1 RCIM_Cubic>
          {arriveTangent=-0.0824, interpMode=RCIM_Cubic, leaveTangent=-0.0824, tangentMode=RCTM_User, time=0.3, value=0.6}
          {arriveTangent=-0.8477, interpMode=RCIM_Cubic, leaveTangent=-0.8477, tangentMode=RCTM_User, time=0.6, value=0.95}
          {arriveTangent=-3.8832, interpMode=RCIM_Cubic, leaveTangent=-3.8832, tangentMode=RCTM_User, time=1, value=0}
         ]}}}
      Scale Curve=500
     ))) {isSetParameters=true, module=SetVariables_#3}
   /SetVariables(Particles.Renderer2Color:LinearColor=&ColorFromCurve(ColorCurve=DataInterface() {propertyValues {
       AlphaCurve {keys [
         {arriveTangent=3.7993, interpMode=RCIM_Cubic, leaveTangent=3.7993, tangentMode=RCTM_User, time=0, value=0}
         <at 0.7907 1 RCIM_Cubic>
         {arriveTangent=-10.4123, interpMode=RCIM_Cubic, leaveTangent=-10.4123, tangentMode=RCTM_User, time=0.9882, value=0}
        ]}
       BlueCurve {keys [<at 0.0016 1 RCIM_Cubic>,<at 0.7893 0>,<at 1 1 RCIM_Cubic>]}
       GreenCurve {keys [<at 0.0016 0 RCIM_Cubic>,<at 0.7893 17.4169>,<at 1 0 RCIM_Cubic>]}
       RedCurve {keys [<at 0.0016 0.5518 RCIM_Cubic>,<at 0.7893 50>,<at 1 0.8655 RCIM_Cubic>]}
      }})) {isSetParameters=true, module=SetVariables_#4}
   /SetVariables(Particles.Renderer3Color:LinearColor=&MultiplyLinearColor(
     Linear Color=&ColorFromCurve(ColorCurve=DataInterface() {propertyValues {
        AlphaCurve {keys [<at 0.005 0>, {arriveTangent=2.1524, interpMode=RCIM_Cubic, leaveTangent=2.1524, time=0.2012, value=1}, <at 0.9342 2>]}
        BlueCurve {keys [<at 0 2 RCIM_Cubic>]}
        GreenCurve {keys [<at 0 0.4514 RCIM_Cubic>]}
        RedCurve {keys [<at 0 1.2257 RCIM_Cubic>]}
       }})
     Scale Factor=<1 1 1 0.5>
    )) {isSetParameters=true, module=SetVariables_#5}
   /SetVariables(Particles.Renderer3Size:Vector2f=&Vector2DFromFloat(Value=&FloatFromCurve(
      FloatCurve=DataInterface() {propertyValues {Curve {keys [
          <at 0 1 RCIM_Cubic>
          {arriveTangent=-0.0824, interpMode=RCIM_Cubic, leaveTangent=-0.0824, tangentMode=RCTM_User, time=0.3, value=0.6}
          {arriveTangent=-0.8477, interpMode=RCIM_Cubic, leaveTangent=-0.8477, tangentMode=RCTM_User, time=0.6, value=0.95}
          {arriveTangent=-3.8832, interpMode=RCIM_Cubic, leaveTangent=-3.8832, tangentMode=RCTM_User, time=1, value=0}
         ]}}}
      Scale Curve=700
     ))) {isSetParameters=true, module=SetVariables_#6}
   /SetVariables(Particles.Renderer4Size:Vector2f=&Multiply_Vector2D(
     A=&Vector2DFromFloat(Value=&FloatFromCurve(
       FloatCurve=DataInterface() {propertyValues {Curve {keys [
           <at 0 1 RCIM_Cubic>
           {arriveTangent=-0.0824, interpMode=RCIM_Cubic, leaveTangent=-0.0824, tangentMode=RCTM_User, time=0.3, value=0.6}
           {arriveTangent=-0.8477, interpMode=RCIM_Cubic, leaveTangent=-0.8477, tangentMode=RCTM_User, time=0.6, value=0.95}
           {arriveTangent=-3.8832, interpMode=RCIM_Cubic, leaveTangent=-3.8832, tangentMode=RCTM_User, time=1, value=0}
          ]}}}
       Scale Curve=130
      ))
     B=<25 3>
    )) {isSetParameters=true, module=SetVariables_#7}
   /SetVariables(Particles.Renderer4Color:LinearColor=&ColorFromCurve(ColorCurve=DataInterface() {propertyValues {
       AlphaCurve {keys [
         {arriveTangent=3.7993, interpMode=RCIM_Cubic, leaveTangent=3.7993, tangentMode=RCTM_User, time=0, value=0}
         <at 0.7907 1 RCIM_Cubic>
         {arriveTangent=-10.4123, interpMode=RCIM_Cubic, leaveTangent=-10.4123, tangentMode=RCTM_User, time=0.9882, value=0}
        ]}
       BlueCurve {keys [<at 0.2744 1 RCIM_Cubic>,<at 0.7893 0>,<at 0.9126 1 RCIM_Cubic>]}
       GreenCurve {keys [<at 0.2744 0 RCIM_Cubic>,<at 0.7893 2.3754>,<at 0.9126 0 RCIM_Cubic>]}
       RedCurve {keys [<at 0.2744 0.5518 RCIM_Cubic>,<at 0.7893 10>,<at 0.9126 0.8655 RCIM_Cubic>]}
      }})) {isSetParameters=true, module=SetVariables_#8}
   /SetVariables(Particles.Renderer5Size:Vector2f=&Vector2DFromFloat(Value=&FloatFromCurve(
      FloatCurve=DataInterface() {propertyValues {Curve {keys [
          <at 0 1 RCIM_Cubic>
          {arriveTangent=-0.0824, interpMode=RCIM_Cubic, leaveTangent=-0.0824, tangentMode=RCTM_User, time=0.3, value=0.6}
          {arriveTangent=-0.8477, interpMode=RCIM_Cubic, leaveTangent=-0.8477, tangentMode=RCTM_User, time=0.6, value=0.95}
          {arriveTangent=-3.8832, interpMode=RCIM_Cubic, leaveTangent=-3.8832, tangentMode=RCTM_User, time=1, value=0}
         ]}}}
      Scale Curve=1500
     ))) {isSetParameters=true, module=SetVariables_#9}
   /SetVariables(Particles.Renderer5Color:LinearColor=&ColorFromCurve(ColorCurve=DataInterface() {propertyValues {
       AlphaCurve {keys [
         {arriveTangent=3.7993, interpMode=RCIM_Cubic, leaveTangent=3.7993, tangentMode=RCTM_User, time=0, value=0}
         <at 0.2608 1>
         {arriveTangent=-1.3748, interpMode=RCIM_Cubic, leaveTangent=-1.3748, time=0.7907, value=1}
         {arriveTangent=-10.4123, interpMode=RCIM_Cubic, leaveTangent=-10.4123, tangentMode=RCTM_User, time=0.9882, value=0}
        ]}
       BlueCurve {keys [<at 0.2653 0.2344>]}
       GreenCurve {keys [<at 0.2653 0.0178>]}
       RedCurve {keys [<at 0.2653 0>]}
      }})) {isSetParameters=true, module=SetVariables_#10}
   &Color(Color=<rgba 5 0 0.4431 1>) {enabled=false}
  ]
  rendererClasses [&NiagaraSpriteRendererProperties, &NiagaraLightRendererProperties]
  renderers [
   {rendererClass=&NiagaraSpriteRendererProperties}
   {rendererClass=&NiagaraLightRendererProperties, rendererIndex=1}
   {rendererClass=&NiagaraSpriteRendererProperties, rendererIndex=2}
   {rendererClass=&NiagaraSpriteRendererProperties, rendererIndex=3}
   {rendererClass=&NiagaraSpriteRendererProperties, rendererIndex=4}
   {rendererClass=&NiagaraSpriteRendererProperties, rendererIndex=5}
  ]
 }
 Glow001 {
  emitterUpdateScript [
   &EmitterState(Life Cycle Mode=::NewEnumerator1, Loop Behavior=::NewEnumerator1, Loop Duration=0.5, UseLoopDelay=false)
   &SpawnBurst_Instantaneous(Use Spawn Probability=false, UseLoopCountLimit=false)
  ]
  particleSpawnScript [&InitializeParticle(Lifetime=0.35, Sprite Size Mode=::NewEnumerator3, Uniform Sprite Size=50, UsePositionOffset=false, Write Material Random=true)]
  particleUpdateScript [
   &ParticleState()
   &SkeletalMeshLocation(
    MaskSpawnGroup=false
    Mesh Sampling Type=::NewEnumerator5
    Position Sampling=::NewEnumerator1
    Skeletal Mesh=$System.Skeletal Mesh
    Socket Sampling Mode=::NewEnumerator5
   ) {module=SkeletalMeshLocation001}
   &ScaleColor(
    Scale Alpha=&FloatFromCurve(FloatCurve=DataInterface() {propertyValues {Curve {keys [
         {arriveTangent=9.9691, interpMode=RCIM_Cubic, leaveTangent=9.9691, tangentMode=RCTM_User, time=0, value=0}
         {arriveTangent=1.9176, interpMode=RCIM_Cubic, leaveTangent=1.9176, tangentMode=RCTM_User, time=0.1, value=0.9}
         {arriveTangent=-0.8417, interpMode=RCIM_Cubic, leaveTangent=-0.8417, tangentMode=RCTM_User, time=1, value=0}
        ]}}})
    Scale RGB=&MakeVectorFromLinearColorRGB(Color=&ColorFromCurve(ColorCurve=DataInterface() {propertyValues {
        AlphaCurve {keys [<at 0 1 RCIM_Cubic>]}
        BlueCurve {keys [<at 0.0096 0>,<at 0.8505 0.0886 RCIM_Cubic>]}
        GreenCurve {keys [<at 0.0096 87.3327>,<at 0.8505 0 RCIM_Cubic>]}
        RedCurve {keys [<at 0.0096 100>,<at 0.8505 1 RCIM_Cubic>]}
       }}))
    ScaleA=true
    ScaleRGB=true
   )
   &CameraOffset(Camera Offset Amount=50)
   &FN_Light_Attributes(
    Color=&ColorFromCurve(ColorCurve=DataInterface() {propertyValues {
       AlphaCurve {keys [
         {arriveTangent=3.7392, interpMode=RCIM_Cubic, leaveTangent=3.7392, tangentMode=RCTM_User, time=0, value=0}
         <at 0.0841 1 RCIM_Cubic>
         {arriveTangent=-7.8769, interpMode=RCIM_Cubic, leaveTangent=-7.8769, tangentMode=RCTM_User, time=0.9962, value=0}
        ]}
       BlueCurve {keys [
         <at 0.0032 0>
         {arriveTangent=3.4692, interpMode=RCIM_Cubic, leaveTangent=3.4692, time=0.0473, value=0.8863}
         <at 0.868 3 RCIM_Cubic>
         {arriveTangent=-36.6349, interpMode=RCIM_Cubic, leaveTangent=-36.6349, tangentMode=RCTM_User, time=1, value=0}
        ]}
       GreenCurve {keys [
         <at 0.0032 0>
         <at 0.0473 0>
         <at 0.868 0 RCIM_Cubic>
         {arriveTangent=-34.8254, interpMode=RCIM_Cubic, leaveTangent=-34.8254, tangentMode=RCTM_User, time=1, value=0}
        ]}
       RedCurve {keys [
         <at 0.0032 0>
         {arriveTangent=1.2499, interpMode=RCIM_Cubic, leaveTangent=1.2499, time=0.0473, value=10}
         <at 0.868 1.0808 RCIM_Cubic>
         {arriveTangent=-32.9241, interpMode=RCIM_Cubic, leaveTangent=-32.9241, tangentMode=RCTM_User, time=1, value=0}
        ]}
      }})
    Create Color Attribute=true
    Create Exponent Attribute=true
    Create Position Attribute=true
    Create Radius Attribute=true
    Exponent=5
    Position=&Add_Vector(A=Linked() {linkedVariable {name=Particles.Position, type=NiagaraPosition}}, B=<0 50 0>)
    Radius=500
    Scale Color=25
   )
   /SetVariables(Particles.Renderer2Size:Vector2f=&Vector2DFromFloat(Value=&FloatFromCurve(
      FloatCurve=DataInterface() {propertyValues {Curve {keys [
          {arriveTangent=1.9233, interpMode=RCIM_Cubic, leaveTangent=1.9233, tangentMode=RCTM_User, time=0, value=0}
          {interpMode=RCIM_Cubic, tangentMode=RCTM_User, time=1, value=1}
         ]}}}
      Scale Curve=600
     ))) {isSetParameters=true, module=SetVariables_#3}
   /SetVariables(Particles.Renderer2Color:LinearColor=&ColorFromCurve(ColorCurve=DataInterface() {propertyValues {
       AlphaCurve {keys [
         <at 0.0023 1 RCIM_Cubic>
         {arriveTangent=-10.4123, interpMode=RCIM_Cubic, leaveTangent=-10.4123, tangentMode=RCTM_User, time=0.9882, value=0}
        ]}
       BlueCurve {keys [<at 0.0529 0>,<at 0.9136 4.804 RCIM_Cubic>]}
       GreenCurve {keys [<at 0.0529 4.6606>,<at 0.9136 0 RCIM_Cubic>]}
       RedCurve {keys [<at 0.0529 5>,<at 0.9136 5 RCIM_Cubic>]}
      }})) {isSetParameters=true, module=SetVariables_#4}
   /SetVariables(Particles.Renderer3Color:LinearColor=&MultiplyLinearColor(
     Linear Color=&ColorFromCurve(ColorCurve=DataInterface() {propertyValues {
        AlphaCurve {keys [<at 0.4885 2>,<at 0.9958 0>]}
        BlueCurve {keys [<at 0 0 RCIM_Cubic>,<at 0.8726 5>]}
        GreenCurve {keys [<at 0 14.3841 RCIM_Cubic>,<at 0.8726 0>]}
        RedCurve {keys [<at 0 20 RCIM_Cubic>,<at 0.8726 4.5673>]}
       }})
     Scale Factor=<1 1 1 0.5>
    )) {isSetParameters=true, module=SetVariables_#5}
   /SetVariables(Particles.Renderer3Size:Vector2f=&Vector2DFromFloat(Value=&FloatFromCurve(
      FloatCurve=DataInterface() {propertyValues {Curve {keys [
          {arriveTangent=1.5995, interpMode=RCIM_Cubic, leaveTangent=1.5995, tangentMode=RCTM_User, time=0, value=0}
          {interpMode=RCIM_Cubic, tangentMode=RCTM_User, time=1, value=1}
         ]}}}
      Scale Curve=1000
     ))) {isSetParameters=true, module=SetVariables_#6}
   /SetVariables(Particles.Renderer4Size:Vector2f=&Multiply_Vector2D(
     A=&Vector2DFromFloat(Value=&FloatFromCurve(
       FloatCurve=DataInterface() {propertyValues {Curve {keys [
           {arriveTangent=2.7377, interpMode=RCIM_Cubic, leaveTangent=2.7377, tangentMode=RCTM_User, time=0, value=0}
           {interpMode=RCIM_Cubic, tangentMode=RCTM_User, time=1, value=1}
          ]}}}
       Scale Curve=220
      ))
     B=<25 2>
    )) {isSetParameters=true, module=SetVariables_#7}
   /SetVariables(Particles.Renderer4Color:LinearColor=&ColorFromCurve(ColorCurve=DataInterface() {propertyValues {
       AlphaCurve {keys [
         <at 0.0104 1 RCIM_Cubic>
         {arriveTangent=-10.4123, interpMode=RCIM_Cubic, leaveTangent=-10.4123, tangentMode=RCTM_User, time=0.9882, value=0}
        ]}
       BlueCurve {keys [<at 0 0>,<at 0.9759 3 RCIM_Cubic>]}
       GreenCurve {keys [<at 0 2.3754>,<at 0.9759 0 RCIM_Cubic>]}
       RedCurve {keys [<at 0 10>,<at 0.9759 2.5964 RCIM_Cubic>]}
      }})) {isSetParameters=true, module=SetVariables_#8}
   /SetVariables(Particles.Renderer5Size:Vector2f=&Vector2DFromFloat(Value=&FloatFromCurve(
      FloatCurve=DataInterface() {propertyValues {Curve {keys [
          {arriveTangent=2.6672, interpMode=RCIM_Cubic, leaveTangent=2.6672, tangentMode=RCTM_User, time=0, value=0}
          {interpMode=RCIM_Cubic, tangentMode=RCTM_User, time=1, value=1}
         ]}}}
      Scale Curve=1500
     ))) {isSetParameters=true, module=SetVariables_#9}
   /SetVariables(Particles.Renderer5Color:LinearColor=&ColorFromCurve(ColorCurve=DataInterface() {propertyValues {
       AlphaCurve {keys [
         {arriveTangent=3.7993, interpMode=RCIM_Cubic, leaveTangent=3.7993, tangentMode=RCTM_User, time=0, value=0}
         <at 0.0628 1>
         <at 0.4995 1 RCIM_Cubic>
         {arriveTangent=-10.4123, interpMode=RCIM_Cubic, leaveTangent=-10.4123, tangentMode=RCTM_User, time=0.9882, value=0}
        ]}
       BlueCurve {keys [<at 0.2653 0.2344>]}
       GreenCurve {keys [<at 0.2653 0.0178>]}
       RedCurve {keys [<at 0.2653 0>]}
      }})) {isSetParameters=true, module=SetVariables_#10}
  ]
  rendererClasses [&NiagaraSpriteRendererProperties, &NiagaraLightRendererProperties]
  renderers [
   {rendererClass=&NiagaraSpriteRendererProperties}
   {rendererClass=&NiagaraLightRendererProperties, rendererIndex=1}
   {rendererClass=&NiagaraSpriteRendererProperties, rendererIndex=2}
   {rendererClass=&NiagaraSpriteRendererProperties, rendererIndex=3}
   {rendererClass=&NiagaraSpriteRendererProperties, rendererIndex=4}
   {rendererClass=&NiagaraSpriteRendererProperties, rendererIndex=5}
  ]
 }
 Init_Glow {
  emitterUpdateScript [
   &EmitterState(Life Cycle Mode=::NewEnumerator1, Loop Behavior=::NewEnumerator1, Loop Delay=0.1, Loop Duration=0.5, UseLoopDelay=false)
   &SpawnBurst_Instantaneous(Spawn Count=1, Use Spawn Probability=false)
  ]
  particleSpawnScript [
   &InitializeParticle(
    Color=<rgba 1 1 1 10>
    Lifetime=0.25
    Mesh Scale Mode=::NewEnumerator3
    Mesh Uniform Scale=0.5
    UsePositionOffset=false
    Write Material Random=true
   ) {module=InitializeParticle001}
   &InitialMeshOrientation(
    Rotation=&RandomRangeVector(Minimum=<0 0 0>, Maximum=<1 1 1>)
    Use Orientation Vector=true
    Use Rotation Vector=true
   ) {module=InitialMeshOrientation001}
  ]
  particleUpdateScript [
   &ParticleState()
   &SkeletalMeshLocation(
    MaskSpawnGroup=false
    Mesh Sampling Type=::NewEnumerator5
    Position Sampling=::NewEnumerator1
    Skeletal Mesh=$System.Skeletal Mesh
    Socket Sampling Mode=::NewEnumerator5
   )
   &DynamicMaterialParameters(
    Index 0 Param 1=&FloatFromCurve(FloatCurve=DataInterface() {propertyValues {Curve {keys [<at 0 1.5>,<at 0.4 1>,<at 1 0.968>]}}})
    Index 0 Param 2=1
    Index 0 Param 3=1
    Index 0 Param 4=1
    Module.Param0WriteEnabled=true
    Module.Param1WriteEnabled=true
    Module.Param2WriteEnabled=true
    Module.Param3WriteEnabled=true
   )
   &ScaleColor(
    Scale Alpha=&FloatFromCurve(FloatCurve=DataInterface() {propertyValues {Curve {keys [<at 0 0.5 RCIM_Cubic>,<at 1 0>]}}})
    Scale RGB=&MakeVectorFromLinearColorRGB(Color=&ColorFromCurve(ColorCurve=DataInterface() {propertyValues {
        AlphaCurve {keys [<at 0 1 RCIM_Cubic>]}
        BlueCurve {keys [<at 0 0 RCIM_Cubic>,<at 0.6829 1 RCIM_Cubic>]}
        GreenCurve {keys [<at 0 82.5692 RCIM_Cubic>,<at 0.6829 0.2012 RCIM_Cubic>]}
        RedCurve {keys [<at 0 100 RCIM_Cubic>,<at 0.6829 0.7609 RCIM_Cubic>]}
       }}))
    ScaleA=true
    ScaleRGB=true
   )
   &ScaleMeshSize(Scale Factor=&VectorFromFloat(Value=&FloatFromCurve(FloatCurve=DataInterface() {propertyValues {Curve {keys [{interpMode=RCIM_Cubic, leaveTangent=1.6767, tangentMode=RCTM_Break, time=0, value=0}, <at 1 2>]}}})))
  ]
  rendererClasses [&NiagaraMeshRendererProperties]
  renderers [{rendererClass=&NiagaraMeshRendererProperties, rendererIndex=0}]
 }
 Parent {
  emitterUpdateScript [
   &EmitterState(Life Cycle Mode=::NewEnumerator1, Loop Behavior=::NewEnumerator1, Loop Duration Mode=::NewEnumerator1, UseLoopDelay=false)
   &SpawnBurst_Instantaneous(Spawn Count=14, Use Spawn Probability=false, UseLoopCountLimit=false)
  ]
  particleSpawnScript [
   &InitializeParticle(
    Lifetime=&RandomRangeFloat(Minimum=2.5, Maximum=3.5)
    Sprite Size Mode=::NewEnumerator3
    Uniform Sprite Size=1
    UsePositionOffset=false
    Write Material Random=true
   )
   &FN_SetSpawnGroup(Number of Spawn Groups=14, Set Ribbon ID and Ribbon Link Order=true)
   &PartitionParticles(Number Of Partitions=14)
  ]
  particleUpdateScript [
   &ParticleState(Kill Particles When Lifetime Has Elapsed=false, Let Infinitely Lived Particles Die When Emitter Deactivates=true)
   &SkeletalMeshLocation(
    Bone / Socket Index=/Niagara/DynamicInputs/Divide/Divide_Int(A:NiagaraInt32=$Particles.PartitionParticles.Partition, B:NiagaraInt32=2)
    Bone Sampling Mode=::NewEnumerator4
    MaskSpawnGroup=false
    Orientation Sampling=::NewEnumerator1
    Sampled Position Offset=&SelectVectorFromArray(
     Direct Array Index=&ReturnParticleIDAttributes()
     Vector Selection Array=DataInterface() {propertyValues {FloatData [
        <-20 0 0>
        <60 1 1>
        <25 0 0>
        <-25 0 0>
        <0 0 0>
        <-50 0 0>
        <0 0 0>
        <40 0 0>
        <30 0 0>
        <-25 0 0>
        <0 -10 -35>
        <0 -10 0>
        <0 0 0>
        <0 -50 25>
       ]}}
    )
    Skeletal Mesh=$System.Skeletal Mesh
   )
  ]
  rendererClasses [&NiagaraSpriteRendererProperties]
  renderers [{rendererClass=&NiagaraSpriteRendererProperties}]
 }
 RibbonsFromParticles {
  emitterUpdateScript [
   &EmitterState()
   /SetVariables(Emitter.EmitterReader=DataInterface() {propertyValues {EmitterBinding {bindingMode=Other, emitterName=Parent}}}) {isSetParameters=true}
   &SpawnRate(
    SpawnRate=&self:Spawn_by_Emitter_01(New Particle Attribute Reader:NiagaraDataInterfaceParticleRead=$Emitter.EmitterReader, Rate:NiagaraFloat=20)
    Use Spawn Probability=false
   )
  ]
  particleSpawnScript [
   &self:Sample_Location_01(New Particle Attribute Reader:NiagaraDataInterfaceParticleRead=$Emitter.EmitterReader) {module=Sample_Location_01}
   &InitializeRibbon(
    Lifetime=&SelectFloatFromArray(
     Direct Array Index=$Particles.SampleID
     Float Selection Array=DataInterface() {propertyValues {FloatData [1,1.5,0.8,1.2,1,1.4,0.9,1.3,0.8,1,1.2,0.8,1.6,1.1]}}
    )
    Module.Write Color=true
    Module.Write Lifetime=true
    Module.Write Mass=true
    Module.Write Material Random=true
    Module.Write Position=false
    Module.Write RibbonTwist=false
    Module.Write RibbonWidth=true
    Ribbon Width=30
   )
  ]
  particleUpdateScript [
   &ParticleState()
   &SkeletalMeshLocation(
    MaskSpawnGroup=false
    Mesh Sampling Type=::NewEnumerator5
    Orientation Sampling=::NewEnumerator1
    Skeletal Mesh=$System.Skeletal Mesh
    Socket Sampling Mode=::NewEnumerator5
    Use Sampled Mesh Attribute as Ribbon ID=false
    Write To Particle Position=false
   ) {module=SkeletalMeshLocation001}
   &VortexVelocity(
    Influence Falloff Radius:NiagaraFloat=&FloatFromCurve(
     FloatCurve=DataInterface() {propertyValues {Curve {keys [
         {arriveTangent=1.0268, interpMode=RCIM_Cubic, leaveTangent=1.0268, tangentMode=RCTM_User, time=0, value=0.25}
         {interpMode=RCIM_Cubic, tangentMode=RCTM_User, time=0.5, value=1}
        ]}}}
     CurveIndex=$System.Age
     Scale Curve=1500
    )
    Module.Use Influence Falloff=true
    Velocity Amount=-1500
    Vortex Axis:Vector3f=<0 1 0>
    Vortex Origin:NiagaraPosition=$Particles.SkeletalMeshLocation001.SampledPosition
   )
   &VortexForce(
    Influence Falloff Exponent=0.25
    Influence Falloff Radius:NiagaraFloat=&FloatFromCurve(
     FloatCurve=DataInterface() {propertyValues {Curve {keys [
         {arriveTangent=1.0268, interpMode=RCIM_Cubic, leaveTangent=1.0268, tangentMode=RCTM_User, time=0, value=0.25}
         {interpMode=RCIM_Cubic, tangentMode=RCTM_User, time=0.5, value=1}
         <at 1 1>
        ]}}}
     CurveIndex=$System.Age
     Scale Curve=3000
    )
    Origin Pull Amount=30000
    Use Influence Falloff=true
    UseOriginOffset=false
    Vortex Axis=<0 1 0>
    Vortex Force Amount=0
    Vortex Origin:NiagaraPosition=$Particles.SkeletalMeshLocation001.SampledPosition
   )
   &PointAttractionForce(
    Attraction Radius=2000
    Attraction Strength=2000
    Attractor Position:NiagaraPosition=$Particles.SkeletalMeshLocation001.SampledPosition
    Kill Within Radius=false
    Module.Use Falloff=true
    UseAttractorOffset=false
   )
   &SolveForcesAndVelocity(Clamp Velocity=false, Limit Acceleration=false)
   &FN_RibbonMaterial(
    Color=&ColorFromCurve(ColorCurve=DataInterface() {propertyValues {
       AlphaCurve {keys [<at 0 1 RCIM_Cubic>]}
       BlueCurve {keys [<at 0 82.8891 RCIM_Cubic>,<at 1 10 RCIM_Cubic>]}
       GreenCurve {keys [<at 0 0 RCIM_Cubic>,<at 1 0 RCIM_Cubic>]}
       RedCurve {keys [<at 0 100 RCIM_Cubic>,<at 1 3.1505 RCIM_Cubic>]}
      }})
    Opacity Pan U Speed=0.5
    Opacity Scale U=0.5
    Scale Alpha:NiagaraFloat=&Multiply_Float(
     A=$System.FN_Timer.Value
     B=&FloatFromCurve(FloatCurve=DataInterface() {propertyValues {Curve {keys [
          {arriveTangent=6.3137, interpMode=RCIM_Cubic, leaveTangent=6.3137, tangentMode=RCTM_User, time=0, value=0}
          {arriveTangent=3.4853, interpMode=RCIM_Cubic, leaveTangent=3.4853, tangentMode=RCTM_User, time=0.1, value=0.8}
          {interpMode=RCIM_Cubic, tangentMode=RCTM_User, time=0.5, value=1}
          {arriveTangent=-4.7464, interpMode=RCIM_Cubic, leaveTangent=-4.7464, tangentMode=RCTM_User, time=1, value=0}
         ]}}})
    )
   )
   &KillParticlesInVolume(Sphere Radius=34, Volume Origin:NiagaraPosition=$Particles.SkeletalMeshLocation001.SampledPosition)
  ]
  rendererClasses [&NiagaraRibbonRendererProperties]
  renderers [{rendererClass=&NiagaraRibbonRendererProperties}]
 }
 SmokeSwirl {
  emitterUpdateScript [
   &EmitterState(Life Cycle Mode=::NewEnumerator1, Loop Behavior=::NewEnumerator1, Loop Delay=3.3, Loop Duration=0.65, UseLoopDelay=true)
   &SpawnRate(SpawnRate=3, Use Spawn Probability=false)
  ]
  particleSpawnScript [&InitializeParticle(Lifetime=0.8, Sprite Size Mode=::NewEnumerator3, Uniform Sprite Size=250, UsePositionOffset=false, Write Material Random=true)]
  particleUpdateScript [
   &ParticleState()
   &SkeletalMeshLocation(
    MaskSpawnGroup=false
    Mesh Sampling Type=::NewEnumerator5
    Position Sampling=::NewEnumerator1
    Skeletal Mesh=$System.Skeletal Mesh
    Socket Sampling Mode=::NewEnumerator5
   ) {module=SkeletalMeshLocation001}
   &ScaleColor(
    Scale Alpha=&FloatFromCurve(FloatCurve=DataInterface() {propertyValues {Curve {keys [
         {arriveTangent=2.7855, interpMode=RCIM_Cubic, leaveTangent=2.7855, tangentMode=RCTM_User, time=0, value=0}
         <at 0.8 1 RCIM_Cubic>
         {arriveTangent=-4.5812, interpMode=RCIM_Cubic, leaveTangent=-4.5812, tangentMode=RCTM_User, time=1, value=0}
        ]}}})
    Scale RGB=&MakeVectorFromLinearColorRGB(Color=&ColorFromCurve(ColorCurve=DataInterface() {propertyValues {
        AlphaCurve {keys [<at 0 1 RCIM_Cubic>]}
        BlueCurve {keys [<at 0.0016 1 RCIM_Cubic>,<at 0.7893 3.9474>,<at 1 0.0886 RCIM_Cubic>]}
        GreenCurve {keys [<at 0.0016 0 RCIM_Cubic>,<at 0.7893 1.0709>,<at 1 0 RCIM_Cubic>]}
        RedCurve {keys [<at 0.0016 0.5518 RCIM_Cubic>,<at 0.7893 5>,<at 1 1 RCIM_Cubic>]}
       }}))
    ScaleA=true
    ScaleRGB=true
   )
   &CameraOffset(Camera Offset Amount=50)
   &FN_Light_Attributes(
    Color=&ColorFromCurve(ColorCurve=DataInterface() {propertyValues {
       AlphaCurve {keys [
         {arriveTangent=3.7392, interpMode=RCIM_Cubic, leaveTangent=3.7392, tangentMode=RCTM_User, time=0, value=0}
         <at 0.7885 1 RCIM_Cubic>
         {arriveTangent=-7.8769, interpMode=RCIM_Cubic, leaveTangent=-7.8769, tangentMode=RCTM_User, time=0.9962, value=0}
        ]}
       BlueCurve {keys [
         <at 0.0032 0>
         {arriveTangent=6.3609, interpMode=RCIM_Cubic, leaveTangent=6.3609, time=0.2109, value=3}
         {arriveTangent=-3.8019, interpMode=RCIM_Cubic, leaveTangent=-3.8019, time=0.7893, value=5}
         {arriveTangent=-36.6349, interpMode=RCIM_Cubic, leaveTangent=-36.6349, tangentMode=RCTM_User, time=1, value=0}
        ]}
       GreenCurve {keys [
         <at 0.0032 0>
         <at 0.2109 0 RCIM_Cubic>
         <at 0.7893 0>
         {arriveTangent=-34.8254, interpMode=RCIM_Cubic, leaveTangent=-34.8254, tangentMode=RCTM_User, time=1, value=0}
        ]}
       RedCurve {keys [
         <at 0.0032 0>
         {arriveTangent=6.0549, interpMode=RCIM_Cubic, leaveTangent=6.0549, time=0.2109, value=1.6554}
         {arriveTangent=-2.0979, interpMode=RCIM_Cubic, leaveTangent=-2.0979, time=0.7893, value=4.7595}
         {arriveTangent=-32.9241, interpMode=RCIM_Cubic, leaveTangent=-32.9241, tangentMode=RCTM_User, time=1, value=0}
        ]}
      }})
    Create Color Attribute=true
    Create Exponent Attribute=true
    Create Position Attribute=true
    Create Radius Attribute=true
    Exponent=5
    Position=&Add_Vector(A=Linked() {linkedVariable {name=Particles.Position, type=NiagaraPosition}}, B=<0 50 0>)
    Radius=250
    Scale Color=25
   )
   &SubUVAnimation()
  ]
  rendererClasses [&NiagaraSpriteRendererProperties]
  renderers [{rendererClass=&NiagaraSpriteRendererProperties}]
 }
 Spewing_Embers {
  simTarget=GPUComputeSim
  emitterUpdateScript [
   &EmitterState(Life Cycle Mode=::NewEnumerator1, Loop Behavior=::NewEnumerator1, Loop Duration=4, UseLoopDelay=false)
   &SpawnBurst_Instantaneous(Spawn Count=250, Use Spawn Probability=false, UseLoopCountLimit=false)
   &SpawnRate(
    SpawnRate=&FloatFromCurve(
     FloatCurve=DataInterface() {propertyValues {Curve {keys [
         <at 0 1 RCIM_Cubic>
         {arriveTangent=-0.3333, interpMode=RCIM_Cubic, leaveTangent=-0.3333, time=0.5, value=0.15}
         {arriveTangent=-0.3, interpMode=RCIM_Cubic, leaveTangent=-0.3, time=0.6, value=0.8}
         <at 1 0 RCIM_Cubic>
        ]}}}
     CurveIndex=$Emitter.NormalizedLoopAge
     Scale Curve=200
    )
    Use Spawn Probability=false
   )
  ]
  particleSpawnScript [
   &InitializeParticle(
    Color=&ColorFromCurve(
     ColorCurve=DataInterface() {propertyValues {
       AlphaCurve {keys [<at 0 1 RCIM_Cubic>]}
       BlueCurve {keys [<at 0 15 RCIM_Cubic>,<at 1 1 RCIM_Cubic>]}
       GreenCurve {keys [<at 0 3.0176 RCIM_Cubic>,<at 1 0 RCIM_Cubic>]}
       RedCurve {keys [<at 0 11.4138 RCIM_Cubic>,<at 1 0.1645 RCIM_Cubic>]}
      }}
     CurveIndex=&RandomRangeFloat(Minimum=0, Maximum=1)
    )
    Lifetime Max=0.6
    Lifetime Min=0.2
    Lifetime Mode=::NewEnumerator1
    Sprite Size=&Multiply_Vector2D_ByFloat(
     Float=&FloatFromCurve(
      FloatCurve=DataInterface() {propertyValues {Curve {keys [<at 0 1>,<at 0.1 0.9>,<at 0.2 0.65>,<at 0.7 0.5>]}}}
      CurveIndex=&RandomRangeFloat(Minimum=0, Maximum=1)
      Scale Curve=12
     )
     Vector2D=<1 2>
    )
    Sprite Size Mode=::NewEnumerator4
    UsePositionOffset=false
    Write Material Random=true
   )
   &SkeletalMeshLocation(
    MaskSpawnGroup=false
    Mesh Sampling Type=::NewEnumerator5
    Orientation Sampling=::NewEnumerator1
    Position Sampling=::NewEnumerator1
    Sampled Position Offset=&RandomRangeVector(Minimum=<-10 -10 -10>, Maximum=<10 10 10>)
    Skeletal Mesh=$System.Skeletal Mesh
    Socket Sampling Mode=::NewEnumerator5
   ) {module=SkeletalMeshLocation001}
   &AddVelocityFromPoint(
    Module.Use Falloff Distance=false
    Origin Offset Coordinate Space=::World
    Velocity Origin:NiagaraPosition=$Output.SkeletalMeshLocation001.SampledPositionWithoutOffset
    Velocity Strength=&RandomRangeFloat(Minimum=250, Maximum=500)
   )
   &AddVelocity(
    Coordinate Space=::World
    Velocity=&RotateVectorByQuaternion(
     Vector To Rotate=&RandomRangeVector(Minimum=<-100 450 -100>, Maximum=<100 700 100>)
     Quaternion=$Particles.SkeletalMeshLocation001.SampledRotation
    )
   )
  ]
  particleUpdateScript [
   &ParticleState()
   &CurlNoiseForce(Noise Frequency=10, Noise Strength=2500, Pan Noise=true, Pan Noise Field=<0 0 10>)
   &ScaleColor(
    Scale Alpha=&FloatFromCurve(FloatCurve=DataInterface() {propertyValues {Curve {keys [{interpMode=RCIM_Cubic, tangentMode=RCTM_User, time=0, value=1}, {interpMode=RCIM_Cubic, tangentMode=RCTM_User, time=1, value=0}]}}})
    Scale RGB=&VectorFromFloat(Value=&RandomRangeFloat(Minimum=0.5, Maximum=4))
    ScaleA=true
    ScaleRGB=true
   )
   &SolveForcesAndVelocity(Clamp Velocity=false, Limit Acceleration=false)
   &ScaleSpriteSizeBySpeed(Max Scale Factor=<1 2>, Velocity Threshold=200)
  ]
  rendererClasses [&NiagaraSpriteRendererProperties]
  renderers [{rendererClass=&NiagaraSpriteRendererProperties}]
 }
 Spewing_Smoke {
  emitterUpdateScript [
   &EmitterState(Life Cycle Mode=::NewEnumerator1, Loop Behavior=::NewEnumerator1, Loop Duration=4, UseLoopDelay=false)
   &SpawnBurst_Instantaneous(Spawn Count=25, Use Spawn Probability=false, UseLoopCountLimit=false)
   &SpawnRate(
    SpawnRate=&FloatFromCurve(
     FloatCurve=DataInterface() {propertyValues {Curve {keys [
         <at 0 1 RCIM_Cubic>
         {arriveTangent=-0.3333, interpMode=RCIM_Cubic, leaveTangent=-0.3333, time=0.5, value=0.15}
         {arriveTangent=-0.3, interpMode=RCIM_Cubic, leaveTangent=-0.3, time=0.6, value=0.8}
         <at 1 0 RCIM_Cubic>
        ]}}}
     CurveIndex=$Emitter.NormalizedLoopAge
     Scale Curve=30
    )
    Use Spawn Probability=false
   )
  ]
  particleSpawnScript [
   &InitializeParticle(
    Lifetime Max=0.6
    Lifetime Min=0.2
    Lifetime Mode=::NewEnumerator1
    Sprite Size Max=<225 450>
    Sprite Size Min=<180 360>
    Sprite Size Mode=::NewEnumerator2
    UsePositionOffset=false
    Write Material Random=true
   )
   &SkeletalMeshLocation(
    MaskSpawnGroup=false
    Mesh Sampling Type=::NewEnumerator5
    Orientation Sampling=::NewEnumerator1
    Position Sampling=::NewEnumerator1
    Sampled Position Offset=&RandomRangeVector(Minimum=<-10 -10 -10>, Maximum=<10 10 10>)
    Skeletal Mesh=$System.Skeletal Mesh
    Socket Sampling Mode=::NewEnumerator5
   ) {module=SkeletalMeshLocation001}
   &AddVelocityFromPoint(
    Module.Use Falloff Distance=false
    Origin Offset Coordinate Space=::World
    Velocity Origin:NiagaraPosition=$Output.SkeletalMeshLocation001.SampledPositionWithoutOffset
    Velocity Strength=&RandomRangeFloat(Minimum=250, Maximum=500)
   )
   &AddVelocity(
    Coordinate Space=::World
    Velocity=&RotateVectorByQuaternion(
     Vector To Rotate=&RandomRangeVector(Minimum=<-100 450 -100>, Maximum=<100 700 100>)
     Quaternion=$Particles.SkeletalMeshLocation001.SampledRotation
    )
   )
  ]
  particleUpdateScript [
   &ParticleState()
   &ScaleColor(
    Scale Alpha=&FloatFromCurve(FloatCurve=DataInterface() {propertyValues {Curve {keys [
         {interpMode=RCIM_Cubic, tangentMode=RCTM_User, time=0, value=1}
         {arriveTangent=-2.0952, interpMode=RCIM_Cubic, leaveTangent=-2.0952, tangentMode=RCTM_User, time=1, value=0}
        ]}}})
    Scale RGB=&MakeVectorFromLinearColorRGB(Color=&ColorFromCurve(ColorCurve=DataInterface() {propertyValues {
        AlphaCurve {keys [<at 0 1 RCIM_Cubic>]}
        BlueCurve {keys [<at 0 0 RCIM_Cubic>,<at 0.5941 4.7994 RCIM_Cubic>]}
        GreenCurve {keys [<at 0 41.1656 RCIM_Cubic>,<at 0.5941 0 RCIM_Cubic>]}
        RedCurve {keys [<at 0 50 RCIM_Cubic>,<at 0.5941 5 RCIM_Cubic>]}
       }}))
    ScaleA=true
    ScaleRGB=true
   )
   &SolveForcesAndVelocity(Clamp Velocity=false, Limit Acceleration=false)
   &SubUVAnimation(Start Frame=35)
   &ScaleSpriteSize(Uniform Curve Sprite Scale=DataInterface() {propertyValues {Curve {keys [
        {arriveTangent=2.4182, interpMode=RCIM_Cubic, leaveTangent=2.4182, tangentMode=RCTM_User, time=0, value=0}
        {interpMode=RCIM_Cubic, tangentMode=RCTM_User, time=1, value=1}
       ]}}})
  ]
  rendererClasses [&NiagaraSpriteRendererProperties]
  renderers [{rendererClass=&NiagaraSpriteRendererProperties}]
 }
 WaveSlow {
  emitterUpdateScript [
   &EmitterState()
   &SpawnBurst_Instantaneous(Spawn Count=1, Use Spawn Probability=false)
   &SpawnRate(SpawnRate=3, Use Spawn Probability=false) {enabled=false}
  ]
  particleSpawnScript [
   &InitializeParticle(
    Color=<rgba 1 1 1 0.25>
    Lifetime=1.5
    Mesh Scale Mode=::NewEnumerator3
    Mesh Uniform Scale=12
    UsePositionOffset=false
    Write Material Random=true
   )
   &InitialMeshOrientation(
    Rotation=&RandomRangeVector(Minimum=<-1 0 0>, Maximum=<1 0 0>)
    Use Orientation Vector=true
    Use Rotation Vector=true
   )
  ]
  particleUpdateScript [
   &ParticleState()
   &SkeletalMeshLocation(
    MaskSpawnGroup=false
    Mesh Sampling Type=::NewEnumerator5
    Orientation Sampling=::NewEnumerator1
    Position Sampling=::NewEnumerator1
    Skeletal Mesh=$System.Skeletal Mesh
    Socket Sampling Mode=::NewEnumerator5
    Use Sampled Mesh Attribute as Ribbon ID=false
   ) {module=SkeletalMeshLocation001}
   &ScaleColor(
    Scale Alpha=&Multiply_Float(
     A=&FloatFromCurve(FloatCurve=DataInterface() {propertyValues {Curve {keys [
          {arriveTangent=-0.9, interpMode=RCIM_Cubic, time=0, value=0}
          {arriveTangent=1.7728, interpMode=RCIM_Cubic, leaveTangent=0.8241, tangentMode=RCTM_Break, time=0.3, value=0.425}
          {arriveTangent=-4.0105, interpMode=RCIM_Cubic, leaveTangent=-4.0105, tangentMode=RCTM_User, time=1, value=0.425}
         ]}}})
     B=&FNTimeOfDayScalar(Night Scale=1, Morning Scale=0.5, Evening Scale=0.5)
    )
    Scale RGB=&MakeVectorFromLinearColorRGB(Color=&ColorFromCurve(ColorCurve=DataInterface() {propertyValues {
        AlphaCurve {keys []}
        BlueCurve {keys [<at 0.0078 1 RCIM_Cubic>, {interpMode=RCIM_Cubic, leaveTangent=0.4741, time=0.9746, value=0}]}
        GreenCurve {keys [<at 0.0078 0.2012 RCIM_Cubic>, {interpMode=RCIM_Cubic, leaveTangent=-4.7123, time=0.9746, value=21.6354}]}
        RedCurve {keys [<at 0.0078 0.7609 RCIM_Cubic>, {interpMode=RCIM_Cubic, leaveTangent=-0.2609, time=0.9746, value=25}]}
       }}))
    ScaleA=true
    ScaleRGB=true
   ) {module=ScaleColor002}
   &DynamicMaterialParameters(
    Index 0 Param 2=-2
    Module.Param0WriteEnabled=false
    Module.Param1WriteEnabled=true
    Module.Param2WriteEnabled=false
    Module.Param3WriteEnabled=false
   )
   &ScaleMeshSize(Scale Factor=&VectorFromFloat(Value=&FloatFromCurve(FloatCurve=DataInterface() {propertyValues {Curve {keys [
          {arriveTangent=-0.6983, interpMode=RCIM_Cubic, leaveTangent=-0.6983, tangentMode=RCTM_User, time=0, value=1}
          {arriveTangent=-1.9793, interpMode=RCIM_Cubic, leaveTangent=-1.9793, tangentMode=RCTM_User, time=1, value=0}
         ]}}})))
   &CameraOffset(Camera Offset Amount=40)
  ]
  rendererClasses [&NiagaraMeshRendererProperties]
  renderers [{rendererClass=&NiagaraMeshRendererProperties, rendererIndex=0}]
 }
}
summary {
 systemName=NS_Caretaker_Death
 emitters [
  {emitterName=Embers, simTarget=GPUComputeSim, rendererClasses [&NiagaraSpriteRendererProperties]}
  {emitterName=Parent, rendererClasses [&NiagaraSpriteRendererProperties]}
  {emitterName=RibbonsFromParticles, rendererClasses [&NiagaraRibbonRendererProperties]}
  {emitterName=Glow, rendererClasses [&NiagaraSpriteRendererProperties, &NiagaraLightRendererProperties]}
  {emitterName=SmokeSwirl, rendererClasses [&NiagaraSpriteRendererProperties]}
  {emitterName=FizzleGlow, rendererClasses [&NiagaraSpriteRendererProperties]}
  {emitterName=FizzleSpark, rendererClasses [&NiagaraSpriteRendererProperties]}
  {emitterName=WaveSlow, rendererClasses [&NiagaraMeshRendererProperties]}
  {emitterName=FN_RadialSprite, rendererClasses [&NiagaraSpriteRendererProperties]}
  {emitterName=Glow001, rendererClasses [&NiagaraSpriteRendererProperties, &NiagaraLightRendererProperties]}
  {emitterName=Spewing_Embers, simTarget=GPUComputeSim, rendererClasses [&NiagaraSpriteRendererProperties]}
  {emitterName=Energy_Blast, rendererClasses [&NiagaraSpriteRendererProperties]}
  {emitterName=Electric_Arcs, rendererClasses [&NiagaraMeshRendererProperties]}
  {emitterName=Init_Glow, rendererClasses [&NiagaraMeshRendererProperties]}
  {emitterName=Energy_ring, rendererClasses [&NiagaraMeshRendererProperties]}
  {emitterName=Spewing_Smoke, rendererClasses [&NiagaraSpriteRendererProperties]}
  {emitterName=FN_RadialSprite001, rendererClasses [&NiagaraMeshRendererProperties]}
 ]
 userVariables [
  User.Color_Tint {type=LinearColor, description="", defaultValue {struct=LinearColor, value=<rgba 1 1 1 1>}}
  User.IsSkydiving {type=NiagaraBool}
  User.RibbonIDOffset {type=NiagaraInt32}
  User.Sideways {type=NiagaraFloat, defaultValue=1, description=""}
 ]
}
